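# Niagara charge-up effect for the tentacle attack, drawing short-lived arc sprites along a skeletal mesh.
# A single Arcs emitter spawns at a steady rate, sweeping the sampled bone/socket index with normalized loop age inside a flattened sphere scatter while spawn size shrinks.
# Sprites use SubUV and rotation, blend between blue and hot color ramps via an array-selected factor, flash bright early, then deform and animate a material parameter.
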NiagaraSystem /PistolCanary/Assets/FX/TentacleAttack/NS_TentacleAttackChargeUp
dependencies {
 usedDataInterfaces [
  &NiagaraDataInterfaceCurve
  &NiagaraDataInterfaceColorCurve
  &NiagaraDataInterfaceArrayFloat
  &NiagaraDataInterfaceStaticMesh
  &NiagaraDataInterfaceSkeletalMesh
 ]
 usedDynamicInputs []
 usedModules [
  &SystemState
  &self:SystemSpawnScript.SystemScriptSource.SystemScriptGraph.NiagaraNodeAssignment_0.SetVariables_#1
  &self:SystemSpawnScript.SystemScriptSource.SystemScriptGraph.NiagaraNodeAssignment_1.SetVariables_#2
 ]
}
emitters {Arcs {
  emitterUpdateScript [
   &EmitterState(Life Cycle Mode=::NewEnumerator1, Loop Behavior=::NewEnumerator1, Loop Duration=0.6, UseLoopDelay=false)
   &SpawnRate(SpawnRate=40, Use Spawn Probability=false)
  ]
  particleSpawnScript [
   &InitializeParticle(
    Lifetime Max=0.4
    Lifetime Min=0.6
    Lifetime Mode=::NewEnumerator1
    Sprite Rotation Mode=::NewEnumerator2
    Sprite Size Mode=::NewEnumerator3
    Uniform Sprite Size=&Multiply_Float(
     A=&FloatFromCurve(FloatCurve=DataInterface() {propertyValues {Curve {keys [<at 0 1>,<at 1 0.5>]}}}, CurveIndex=$Emitter.NormalizedLoopAge)
     B=&RandomRangeFloat(Minimum=7000, Maximum=10000)
    )
    UsePositionOffset=false
    Write Material Random=true
   )
   &SubUVAnimation(End Frame=3, SubUV Animation Mode:ENiagara_SubUVLookupMode=::NewEnumerator2)
   &SphereLocation(
    MaskSpawnGroup=false
    Non Uniform Scale=<1 1 0.25>
    Override Seed=false
    Sphere Radius=800
    Surface Only=false
    UseNonUniformScale=true
    UseOffset=false
    UseSphereOrientation=false
   )
   &SkeletalMeshLocation(
    Bone / Socket Index=&MakeIntFromFloat(Float=&FloatFromCurve(FloatCurve=DataInterface() {propertyValues {Curve {keys [<at 0 1>,<at 1 19>]}}}, CurveIndex=$Emitter.NormalizedLoopAge))
    Bone Sampling Mode=::NewEnumerator6
    MaskSpawnGroup=false
    Skeletal Mesh=$System.Skeletal Mesh
   )
  ]
  particleUpdateScript [
   &ParticleState()
   &CameraOffset(Camera Offset Amount=25)
   &ScaleColor(
    Scale RGB=&MakeVectorFromLinearColorRGB(Color=&LerpLinearColors(
      StartColor=&ColorFromCurve(ColorCurve=DataInterface() {propertyValues {
         AlphaCurve {keys [<at 0.5586 1 RCIM_Cubic>,<at 1 0 RCIM_Cubic>]}
         BlueCurve {keys [<at 0 1 RCIM_Cubic>,<at 1 1 RCIM_Cubic>]}
         GreenCurve {keys [
           {arriveTangent=-0.0536, interpMode=RCIM_Cubic, leaveTangent=-0.0536, time=0, value=0.2093}
           {arriveTangent=-0.0536, interpMode=RCIM_Cubic, leaveTangent=-0.0536, time=1, value=0.1557}
          ]}
         RedCurve {keys [
           {arriveTangent=0.0676, interpMode=RCIM_Cubic, leaveTangent=0.0676, time=0, value=0.0396}
           {arriveTangent=0.0676, interpMode=RCIM_Cubic, leaveTangent=0.0676, time=1, value=0.1073}
          ]}
        }})
      EndColor=&ColorFromCurve(ColorCurve=DataInterface() {propertyValues {
         AlphaCurve {keys [<at 0.5586 1 RCIM_Cubic>,<at 1.0091 0 RCIM_Cubic>]}
         BlueCurve {keys [<at 0 1.2408 RCIM_Cubic>,<at 0.982 0>]}
         GreenCurve {keys [<at 0 1.7901 RCIM_Cubic>,<at 0.982 0.2738>]}
         RedCurve {keys [<at 0 5 RCIM_Cubic>,<at 0.982 5>]}
        }})
      LerpFactor=&SelectFloatFromArray(Direct Array Index=0, Float Selection Array=DataInterface() {propertyValues {FloatData [0,0,0,1]}})
     ))
    ScaleA=false
    ScaleRGB=true
   )
   &ScaleColor(
    Scale RGB=&VectorFromFloat(Value=&FloatFromCurve(FloatCurve=DataInterface() {propertyValues {Curve {keys [
          {interpMode=RCIM_Cubic, tangentMode=RCTM_User, time=0, value=1}
          {arriveTangent=106.638, interpMode=RCIM_Cubic, leaveTangent=-47.9454, tangentMode=RCTM_Break, time=0.2, value=10}
          {interpMode=RCIM_Cubic, tangentMode=RCTM_User, time=0.8, value=1}
         ]}}}))
    ScaleA=false
    ScaleRGB=true
   ) {module=ScaleColor001}
   &ScaleSpriteSize(Scale Factor=&Vector2DFromFloat(Value=&FloatFromCurve(FloatCurve=DataInterface() {propertyValues {Curve {keys [
          {arriveTangent=-0.0604, interpMode=RCIM_Cubic, leaveTangent=-0.0604, tangentMode=RCTM_User, time=0, value=0.6}
          {arriveTangent=1.6386, interpMode=RCIM_Cubic, leaveTangent=1.6386, tangentMode=RCTM_User, time=1, value=1}
         ]}}})))
   &SpriteRotationRate(Rotation Rate=25)
   &ScaleSpriteSize(
    Scale Factor=&MakeVector2D(
     X=&FloatFromCurve(FloatCurve=DataInterface() {propertyValues {Curve {keys [
          {arriveTangent=-3.4808, interpMode=RCIM_Cubic, leaveTangent=-3.4808, tangentMode=RCTM_User, time=0.4, value=1}
          {arriveTangent=0.0049, interpMode=RCIM_Cubic, leaveTangent=0.0049, tangentMode=RCTM_User, time=0.8, value=0.5}
         ]}}})
     Y=&FloatFromCurve(FloatCurve=DataInterface() {propertyValues {Curve {keys [
          {arriveTangent=0.0124, interpMode=RCIM_Cubic, leaveTangent=0.0124, tangentMode=RCTM_User, time=0.4, value=0.85}
          {arriveTangent=2.3888, interpMode=RCIM_Cubic, leaveTangent=2.3888, tangentMode=RCTM_User, time=0.8, value=1.2}
         ]}}})
    )
    Scale Sprite Size Mode=::NewEnumerator2
   ) {module=ScaleSpriteSize001}
   &DynamicMaterialParameters(
    Index 0 Param 1=1
    Index 0 Param 2=1
    Index 0 Param 3=&FloatFromCurve(FloatCurve=DataInterface() {propertyValues {Curve {keys [
         {arriveTangent=0.1411, interpMode=RCIM_Cubic, leaveTangent=0.1411, tangentMode=RCTM_User, time=0.2, value=0}
         {arriveTangent=1.8523, interpMode=RCIM_Cubic, leaveTangent=1.8523, tangentMode=RCTM_User, time=1, value=0.5}
        ]}}})
    Module.Param0WriteEnabled=true
    Module.Param1WriteEnabled=true
    Module.Param2WriteEnabled=true
    Module.Param3WriteEnabled=false
   )
  ]
  rendererClasses [&NiagaraSpriteRendererProperties]
  renderers [{rendererClass=&NiagaraSpriteRendererProperties}]
 }}
summary {
 systemName=NS_TentacleAttackChargeUp
 emitters [{emitterName=Arcs, rendererClasses [&NiagaraSpriteRendererProperties]}]
 userVariables [
  User.Color {type=LinearColor, description="", defaultValue {struct=LinearColor, value=<rgba 0 0.161 0.5 1>}}
  User.ColorMult {type=NiagaraFloat, defaultValue=1, description=""}
  User.Scaling_SpawnRadius {type=NiagaraFloat, defaultValue=1, description=""}
  User.WidthMult {type=NiagaraFloat, defaultValue=1, description=""}
 ]
}
