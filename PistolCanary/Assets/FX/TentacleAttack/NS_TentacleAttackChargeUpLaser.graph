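# Niagara system for the tentacle attack's laser charge-up, placing lightning and plasma along a skeletal mesh's bones.
# Arcs spawns flickering SubUV sprites at a bone that climbs with User.ChargeProgressLightning, gated by User.LightningEnabled.
# Plasma ribbon emitters, disabled along with a second arc emitter, link rings at the base and a ChargeProgress-selected bone.
# Those ribbons use Hermite splines or lerps with waveform offsets tapered at both ends.
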NiagaraSystem /PistolCanary/Assets/FX/TentacleAttack/NS_TentacleAttackChargeUpLaser
dependencies {
 usedDataInterfaces [
  &NiagaraDataInterfaceCurve
  &NiagaraDataInterfaceColorCurve
  &NiagaraDataInterfaceArrayFloat
  &NiagaraDataInterfaceArrayFloat2
  &NiagaraDataInterfaceVector2DCurve
  &NiagaraDataInterfaceStaticMesh
  &NiagaraDataInterfaceSkeletalMesh
 ]
 usedDynamicInputs [&MakeCustomVectorFromBool]
 usedModules [
  &SystemState
  &self:SystemSpawnScript.SystemScriptSource.SystemScriptGraph.NiagaraNodeAssignment_0.SetVariables_#1
  &self:SystemSpawnScript.SystemScriptSource.SystemScriptGraph.NiagaraNodeAssignment_1.SetVariables_#2
 ]
}
emitters {
 Arcs {
  emitterUpdateScript [
   &EmitterState(Life Cycle Mode=::NewEnumerator1, Loop Behavior=::NewEnumerator1, Loop Duration=$User.LightningLoopAge, UseLoopDelay=false)
   &SpawnRate(
    SpawnRate=&FloatFromCurve(
     FloatCurve=DataInterface() {propertyValues {Curve {keys [<at 2 1>,<at 3 0.4>]}}}
     CurveIndex=$Emitter.NormalizedLoopAge
     Scale Curve=&Multiply_Float(A=80, B=$User.LightningEnabled)
    )
    Use Spawn Probability=false
   )
  ]
  particleSpawnScript [
   &InitializeParticle(
    Lifetime Max=0.2
    Lifetime Min=0.2
    Lifetime Mode=::NewEnumerator1
    Sprite Rotation Mode=::NewEnumerator2
    Sprite Size Mode=::NewEnumerator3
    Uniform Sprite Size=&Multiply_Float(
     A=&FloatFromCurve(FloatCurve=DataInterface() {propertyValues {Curve {keys [<at 0 1.5>,<at 0.7 0.3>]}}}, CurveIndex=$Emitter.NormalizedLoopAge, Scale Curve=1.5)
     B=&RandomRangeFloat(Minimum=4000, Maximum=6000)
    )
    UsePositionOffset=false
    Write Material Random=true
   )
   &SubUVAnimation(End Frame=3, SubUV Animation Mode:ENiagara_SubUVLookupMode=::NewEnumerator2)
   &SphereLocation(
    MaskSpawnGroup=false
    Non Uniform Scale=<1 1 0.25>
    Override Seed=false
    Sphere Radius=800
    Surface Only=false
    UseNonUniformScale=true
    UseOffset=false
    UseSphereOrientation=false
   )
   &SkeletalMeshLocation(
    Bone / Socket Index=&MakeIntFromFloat(Float=&FloatFromCurve(FloatCurve=DataInterface() {propertyValues {Curve {keys [<at 0 1>,<at 1 19>]}}}, CurveIndex=$User.ChargeProgressLightning))
    Bone Sampling Mode=::NewEnumerator6
    MaskSpawnGroup=false
    Skeletal Mesh=$System.Skeletal Mesh
   )
  ]
  particleUpdateScript [
   &ParticleState()
   &CameraOffset(Camera Offset Amount=&RandomRangeFloat(Minimum=-2000, Maximum=1200))
   &ScaleColor(
    Scale RGB=&MakeVectorFromLinearColorRGB(Color=&LerpLinearColors(
      StartColor=&ColorFromCurve(ColorCurve=DataInterface() {propertyValues {
         AlphaCurve {keys [<at 0.5586 1 RCIM_Cubic>,<at 1 1 RCIM_Cubic>]}
         BlueCurve {keys [<at 0 3 RCIM_Cubic>,<at 1 2 RCIM_Cubic>]}
         GreenCurve {keys [{arriveTangent=-0.0536, interpMode=RCIM_Cubic, time=0, value=0}, {interpMode=RCIM_Cubic, leaveTangent=-0.0536, time=1, value=0}]}
         RedCurve {keys [{arriveTangent=0.0676, interpMode=RCIM_Cubic, time=0, value=0.6364}, {interpMode=RCIM_Cubic, leaveTangent=0.0676, time=1, value=1.7703}]}
        }})
      EndColor=&ColorFromCurve(ColorCurve=DataInterface() {propertyValues {
         AlphaCurve {keys [<at 0.5586 1 RCIM_Cubic>,<at 1.0091 0 RCIM_Cubic>]}
         BlueCurve {keys [<at 0 1.2408 RCIM_Cubic>,<at 0.982 0>]}
         GreenCurve {keys [<at 0 1.7901 RCIM_Cubic>,<at 0.982 0.2738>]}
         RedCurve {keys [<at 0 5 RCIM_Cubic>,<at 0.982 5>]}
        }})
      LerpFactor=&SelectFloatFromArray(Direct Array Index=0, Float Selection Array=DataInterface() {propertyValues {FloatData [0,0,0,1]}})
     ))
    ScaleA=false
    ScaleRGB=true
   )
   &ScaleColor(
    Scale RGB=&VectorFromFloat(Value=&FloatFromCurve(FloatCurve=DataInterface() {propertyValues {Curve {keys [
          {interpMode=RCIM_Cubic, tangentMode=RCTM_User, time=0, value=1}
          {arriveTangent=63.022, interpMode=RCIM_Cubic, leaveTangent=-26.4845, tangentMode=RCTM_Break, time=0.0732, value=6}
          {interpMode=RCIM_Cubic, tangentMode=RCTM_User, time=0.8, value=1}
         ]}}}))
    ScaleA=false
    ScaleRGB=true
   ) {module=ScaleColor001}
   &ScaleSpriteSize(Scale Factor=&Vector2DFromFloat(Value=&FloatFromCurve(FloatCurve=DataInterface() {propertyValues {Curve {keys [
          {arriveTangent=-0.0604, interpMode=RCIM_Cubic, leaveTangent=-0.0604, tangentMode=RCTM_User, time=0, value=0.6}
          {arriveTangent=1.6386, interpMode=RCIM_Cubic, leaveTangent=1.6386, tangentMode=RCTM_User, time=1, value=1}
         ]}}})))
   &SpriteRotationRate(Rotation Rate=25)
   &ScaleSpriteSize(
    Scale Factor=&MakeVector2D(
     X=&FloatFromCurve(FloatCurve=DataInterface() {propertyValues {Curve {keys [
          {arriveTangent=-3.4808, interpMode=RCIM_Cubic, leaveTangent=-3.4808, tangentMode=RCTM_User, time=0.4, value=1}
          {arriveTangent=0.0049, interpMode=RCIM_Cubic, leaveTangent=0.0049, tangentMode=RCTM_User, time=0.8, value=0.5}
         ]}}})
     Y=&FloatFromCurve(FloatCurve=DataInterface() {propertyValues {Curve {keys [
          {arriveTangent=0.0124, interpMode=RCIM_Cubic, leaveTangent=0.0124, tangentMode=RCTM_User, time=0.4, value=0.85}
          {arriveTangent=2.3888, interpMode=RCIM_Cubic, leaveTangent=2.3888, tangentMode=RCTM_User, time=0.8, value=1.2}
         ]}}})
    )
    Scale Sprite Size Mode=::NewEnumerator2
   ) {module=ScaleSpriteSize001}
   &DynamicMaterialParameters(
    Index 0 Param 1=1
    Index 0 Param 2=1
    Index 0 Param 3=&FloatFromCurve(FloatCurve=DataInterface() {propertyValues {Curve {keys [
         {arriveTangent=0.1411, interpMode=RCIM_Cubic, leaveTangent=0.1411, tangentMode=RCTM_User, time=0.2, value=0}
         {arriveTangent=1.8523, interpMode=RCIM_Cubic, leaveTangent=1.8523, tangentMode=RCTM_User, time=1, value=0.5}
        ]}}})
    Module.Param0WriteEnabled=true
    Module.Param1WriteEnabled=true
    Module.Param2WriteEnabled=true
    Module.Param3WriteEnabled=false
   )
  ]
  rendererClasses [&NiagaraSpriteRendererProperties]
  renderers [{rendererClass=&NiagaraSpriteRendererProperties}]
 }
 Arcs001 {
  bEnabled=false
  emitterUpdateScript [
   &EmitterState(Life Cycle Mode=::NewEnumerator1, Loop Behavior=::NewEnumerator1, Loop Duration=2.5, UseLoopDelay=false)
   &SpawnRate(
    SpawnRate=&FloatFromCurve(
     FloatCurve=DataInterface() {propertyValues {Curve {keys [<at 0.5 0>,<at 0.65 1>,<at 1 0.4>]}}}
     CurveIndex=$Emitter.NormalizedLoopAge
     Scale Curve=20
    )
    Use Spawn Probability=false
   )
  ]
  particleSpawnScript [
   &InitializeParticle(
    Lifetime Max=0.4
    Lifetime Min=0.6
    Lifetime Mode=::NewEnumerator1
    Sprite Rotation Mode=::NewEnumerator2
    Sprite Size Mode=::NewEnumerator3
    Uniform Sprite Size=&Multiply_Float(
     A=&FloatFromCurve(FloatCurve=DataInterface() {propertyValues {Curve {keys [<at 0 0.8>,<at 0.7 1.2>]}}}, CurveIndex=$Emitter.NormalizedLoopAge)
     B=&RandomRangeFloat(Minimum=2000, Maximum=2500)
    )
    UsePositionOffset=false
    Write Material Random=true
   )
   &SubUVAnimation(End Frame=3, SubUV Animation Mode:ENiagara_SubUVLookupMode=::NewEnumerator2)
   &SphereLocation(
    MaskSpawnGroup=false
    Non Uniform Scale=<1 1 0.25>
    Override Seed=false
    Sphere Radius=800
    Surface Only=false
    UseNonUniformScale=true
    UseOffset=false
    UseSphereOrientation=false
   )
   &SkeletalMeshLocation(Bone / Socket Index=18, Bone Sampling Mode=::NewEnumerator6, MaskSpawnGroup=false, Skeletal Mesh=$System.Skeletal Mesh)
  ]
  particleUpdateScript [
   &ParticleState()
   &CameraOffset(Camera Offset Amount=25)
   &ScaleColor(
    Scale RGB=&MakeVectorFromLinearColorRGB(Color=&LerpLinearColors(
      StartColor=&ColorFromCurve(ColorCurve=DataInterface() {propertyValues {
         AlphaCurve {keys [<at 0.5586 1 RCIM_Cubic>,<at 1 0 RCIM_Cubic>]}
         BlueCurve {keys [<at 0 1 RCIM_Cubic>,<at 1 1 RCIM_Cubic>]}
         GreenCurve {keys [{arriveTangent=-0.0536, interpMode=RCIM_Cubic, time=0, value=0.082}, {interpMode=RCIM_Cubic, leaveTangent=-0.0536, time=1, value=0.1098}]}
         RedCurve {keys [{arriveTangent=0.0676, interpMode=RCIM_Cubic, time=0, value=0.067}, {interpMode=RCIM_Cubic, leaveTangent=0.0676, time=1, value=0.3327}]}
        }})
      EndColor=&ColorFromCurve(ColorCurve=DataInterface() {propertyValues {
         AlphaCurve {keys [<at 0.5586 1 RCIM_Cubic>,<at 1.0091 0 RCIM_Cubic>]}
         BlueCurve {keys [<at 0 1.2408 RCIM_Cubic>,<at 0.982 0>]}
         GreenCurve {keys [<at 0 1.7901 RCIM_Cubic>,<at 0.982 0.2738>]}
         RedCurve {keys [<at 0 5 RCIM_Cubic>,<at 0.982 5>]}
        }})
      LerpFactor=&SelectFloatFromArray(Direct Array Index=0, Float Selection Array=DataInterface() {propertyValues {FloatData [0,0,0,1]}})
     ))
    ScaleA=false
    ScaleRGB=true
   )
   &ScaleColor(
    Scale RGB=&VectorFromFloat(Value=&FloatFromCurve(FloatCurve=DataInterface() {propertyValues {Curve {keys [
          {interpMode=RCIM_Cubic, tangentMode=RCTM_User, time=0, value=1}
          {arriveTangent=106.638, interpMode=RCIM_Cubic, leaveTangent=-47.9454, tangentMode=RCTM_Break, time=0.2, value=10}
          {interpMode=RCIM_Cubic, tangentMode=RCTM_User, time=0.8, value=1}
         ]}}}))
    ScaleA=false
    ScaleRGB=true
   ) {module=ScaleColor001}
   &ScaleSpriteSize(Scale Factor=&Vector2DFromFloat(Value=&FloatFromCurve(FloatCurve=DataInterface() {propertyValues {Curve {keys [
          {arriveTangent=-0.0604, interpMode=RCIM_Cubic, leaveTangent=-0.0604, tangentMode=RCTM_User, time=0, value=0.6}
          {arriveTangent=1.6386, interpMode=RCIM_Cubic, leaveTangent=1.6386, tangentMode=RCTM_User, time=1, value=1}
         ]}}})))
   &SpriteRotationRate(Rotation Rate=25)
   &ScaleSpriteSize(
    Scale Factor=&MakeVector2D(
     X=&FloatFromCurve(FloatCurve=DataInterface() {propertyValues {Curve {keys [
          {arriveTangent=-3.4808, interpMode=RCIM_Cubic, leaveTangent=-3.4808, tangentMode=RCTM_User, time=0.4, value=1}
          {arriveTangent=0.0049, interpMode=RCIM_Cubic, leaveTangent=0.0049, tangentMode=RCTM_User, time=0.8, value=0.5}
         ]}}})
     Y=&FloatFromCurve(FloatCurve=DataInterface() {propertyValues {Curve {keys [
          {arriveTangent=0.0124, interpMode=RCIM_Cubic, leaveTangent=0.0124, tangentMode=RCTM_User, time=0.4, value=0.85}
          {arriveTangent=2.3888, interpMode=RCIM_Cubic, leaveTangent=2.3888, tangentMode=RCTM_User, time=0.8, value=1.2}
         ]}}})
    )
    Scale Sprite Size Mode=::NewEnumerator2
   ) {module=ScaleSpriteSize001}
   &DynamicMaterialParameters(
    Index 0 Param 1=1
    Index 0 Param 2=1
    Index 0 Param 3=&FloatFromCurve(FloatCurve=DataInterface() {propertyValues {Curve {keys [
         {arriveTangent=0.1411, interpMode=RCIM_Cubic, leaveTangent=0.1411, tangentMode=RCTM_User, time=0.2, value=0}
         {arriveTangent=1.8523, interpMode=RCIM_Cubic, leaveTangent=1.8523, tangentMode=RCTM_User, time=1, value=0.5}
        ]}}})
    Module.Param0WriteEnabled=true
    Module.Param1WriteEnabled=true
    Module.Param2WriteEnabled=true
    Module.Param3WriteEnabled=false
   )
  ]
  rendererClasses [&NiagaraSpriteRendererProperties]
  renderers [{rendererClass=&NiagaraSpriteRendererProperties}]
 }
 Blue_Outer_Plasma {
  bEnabled=false
  emitterUpdateScript [
   &EmitterState(
    Life Cycle Mode=::NewEnumerator1
    Loop Delay=&RandomRangeFloat(Minimum=0.2, Maximum=0.3)
    Loop Duration=0.4
    UseLoopDelay=false
   )
   &SpawnRate(SpawnRate=&Multiply_Float(A=$User.ChargeProgress, B=5), Use Spawn Probability=false) {enabled=false}
   &SpawnBurst_Instantaneous(Spawn Count=10, Spawn Probability=$User.PlasmaEnabled, Use Spawn Probability=true)
   &SpawnBurst_Instantaneous(Spawn Count=10, Spawn Probability=$User.PlasmaEnabled, Spawn Time=0.15, Use Spawn Probability=true) {module=SpawnBurst_Instantaneous001}
   &SpawnBurst_Instantaneous(Spawn Count=10, Spawn Probability=$User.PlasmaEnabled, Spawn Time=0.25, Use Spawn Probability=true) {module=SpawnBurst_Instantaneous002}
  ]
  particleSpawnScript [
   &FN_SetSpawnGroup(Inclusivity=true, Normalize Spawn Group=true, Particles Per Group=10, Pattern=::NewEnumerator0, Set Ribbon ID and Ribbon Link Order=true)
   &SampleSkeletalMesh(
    AdditionalYawPitchRoll=false
    Bone / Socket Index=1
    Bone Sampling Mode:ENiagaraBoneSamplingMode=::NewEnumerator4
    Skeletal Mesh:NiagaraDataInterfaceSkeletalMesh=$System.Skeletal Mesh
   )
   &ShapeLocation(
    Apply To Particle Position=::NewEnumerator1
    MaskSpawnGroup=false
    Override Seed=false
    Ring Radius=&Lerp_Float(A=1500, B=5000, Alpha=$User.ChargeProgress)
    Shape Origin=$Particles.SampleSkeletalMesh.SampledPosition
    Shape Primitive=::NewEnumerator4
    UseOwnerScale=true
   )
   &SampleSkeletalMesh(
    AdditionalYawPitchRoll=false
    Bone / Socket Index:NiagaraInt32=&MakeIntFromFloat(Float=&FloatFromCurve(FloatCurve=DataInterface() {propertyValues {Curve {keys [<at 0 1>,<at 1 18>]}}}, CurveIndex=$User.ChargeProgress))
    Bone Sampling Mode:ENiagaraBoneSamplingMode=::NewEnumerator4
    Skeletal Mesh:NiagaraDataInterfaceSkeletalMesh=$System.Skeletal Mesh
   ) {module=SampleSkeletalMesh001}
   &ShapeLocation(
    Apply To Particle Position=::NewEnumerator1
    MaskSpawnGroup=false
    Override Seed=false
    Ring Radius=&Lerp_Float(A=1500, B=100, Alpha=$User.ChargeProgress)
    Shape Origin=$Particles.SampleSkeletalMesh001.SampledPosition
    Shape Primitive=::NewEnumerator4
    UseOwnerScale=true
   ) {module=ShapeLocation001}
   &InitializeRibbon(
    Lifetime=0.4
    Module.Write Color=true
    Module.Write Lifetime=true
    Module.Write Mass=true
    Module.Write Material Random=true
    Module.Write Position=false
    Module.Write RibbonTwist=false
    Module.Write RibbonWidth=true
    Ribbon Width=200
   )
   /SetVariables(Particles.HeightAddition:NiagaraFloat=&RemapRange(Input Value To Remap=$User.ChargeProgress, Input Min=0, Input Max=1, Output Min=0, Output Max=15000)) {isSetParameters=true, module=SetVariables_#3}
  ]
  particleUpdateScript [
   &ParticleState()
   /Game/Effects/Niagara/Modules/Curves/FN_HermiteCubicSpline_v2/FN_HermiteCubicSpline(
    Position Blend:NiagaraFloat=$Particles.Normalized Spawn Group
    Position End:Vector3f=$User.HitLocation
    Position Start:Vector3f=$User.StartLocation
    Write Particles Position=true
   ) {enabled=false}
   /Game/Effects/Niagara/Modules/Curves/FN_HermiteCubicSpline_v1/FN_HermiteCubicSpline(
    Compute Direction=true
    Position Blend:NiagaraFloat=$Particles.Normalized Spawn Group
    Position End:Vector3f=&Lerp_Vector(
     A=&AddVectorToPosition(
      Position=$Particles.Initial.ShapeLocation.ShapePosition
      Vector=&MakeVector(X=0, Y=0, Z=$Particles.HeightAddition)
     )
     B=Linked() {linkedVariable {name=Particles.SampleSkeletalMesh001.SampledPosition, type=NiagaraPosition}}
     Alpha=&FloatFromCurve(FloatCurve=DataInterface() {propertyValues {Curve {keys [<at 0.99 0>,<at 1 0>]}}}, CurveIndex=$User.ChargeProgress)
    )
    Position Start:NiagaraPosition=$Particles.Initial.ShapeLocation.ShapePosition
   ) {enabled=false, module=FN_HermiteCubicSpline001}
   &FN_HermiteCubicSpline(
    Compute Direction=true
    Position Blend:NiagaraFloat=$Particles.Normalized Spawn Group
    Position End:NiagaraPosition=$Particles.Initial.ShapeLocation001.ShapePosition
    Position Start:NiagaraPosition=$Particles.Initial.ShapeLocation.ShapePosition
   ) {module=FN_HermiteCubicSpline003}
   /SetVariables(Particles.Position=&Add_Vector(
     A=$Output.FN_HermiteCubicSpline003.Position
     B=/Niagara/DynamicInputs/Vectors/RotateVector(
      Vector To Rotate:Vector3f=&MakeVector(
       X=0
       Y=&Waveform(
        Global Input Position=&Add_Float(A=$Particles.Normalized Spawn Group, B=&RandomRangeFloat(Minimum=0, Maximum=1))
        Global Amplitude Scale=4
        "Waveform [1]"=::NewEnumerator0
        "[1] Frequency"=1
        "[1] Amplitude Min/Max"=<0 3>
        "[1] Scale"=&RandomRangeFloat(Minimum=-1, Maximum=1)
        "[2] Phase"=$Emitter.Age
        "[3] Phase"=$Emitter.Age
       )
       Z=&Waveform(
        Global Input Position=&Add_Float(A=$Particles.Normalized Spawn Group, B=&RandomRangeFloat(Minimum=0, Maximum=1))
        Global Amplitude Scale=20
        "Waveform [1]"=::NewEnumerator1
        "[1] Frequency"=1
        "[1] Amplitude Min/Max"=<0 3>
        "[1] Scale"=&RandomRangeFloat(Minimum=-1, Maximum=1)
        "[2] Phase"=$Emitter.Age
        "[3] Phase"=$Emitter.Age
       )
      )
      Yaw:NiagaraFloat=&MakeFloatFromVector(Vector=$Output.FN_HermiteCubicSpline001.Direction, Channel=::NewEnumerator0)
      Pitch:NiagaraFloat=&MakeFloatFromVector(Vector=$Output.FN_HermiteCubicSpline001.Direction, Channel=::NewEnumerator1)
      Roll:NiagaraFloat=&MakeFloatFromVector(Vector=$Output.FN_HermiteCubicSpline001.Direction, Channel=::NewEnumerator2)
      Rotation Rate:NiagaraFloat=$Engine.DeltaTime
     )
    )) {isSetParameters=true, module=SetVariables_#4}
   /SetVariables(Particles.Position=&Lerp_Vector(
     A=$Output.FN_HermiteCubicSpline003.Position
     B=Linked() {linkedVariable {name=Particles.Position, type=NiagaraPosition}}
     Alpha=&FloatFromCurve(
      FloatCurve=DataInterface() {propertyValues {Curve {keys [
          {arriveTangent=6.9713, interpMode=RCIM_Cubic, leaveTangent=6.9713, tangentMode=RCTM_User, time=0, value=0}
          {interpMode=RCIM_Cubic, tangentMode=RCTM_User, time=0.5, value=1}
          {arriveTangent=-6.3689, interpMode=RCIM_Cubic, leaveTangent=-6.3689, tangentMode=RCTM_User, time=1, value=0}
         ]}}}
      CurveIndex=$Particles.Normalized Spawn Group
     )
    )) {isSetParameters=true, module=SetVariables_#5}
   &FN_RibbonMaterial(
    Color=&ColorFromCurve(
     ColorCurve=DataInterface() {propertyValues {
       AlphaCurve {keys [<at 0.0106 1 RCIM_Cubic>,<at 0.9796 1>]}
       BlueCurve {keys [
         {arriveTangent=-3.673, interpMode=RCIM_Cubic, time=0.2257, value=7}
         <at 0.2939 4>
         <at 0.4639 4>
         <at 0.5897 4>
         <at 0.898 4>
         <at 0.946 5>
         <at 0.9978 7.8247>
        ]}
       GreenCurve {keys [
         {arriveTangent=0.1965, interpMode=RCIM_Cubic, time=0.2257, value=2.7435}
         <at 0.2939 0.1534>
         <at 0.4639 0.1731>
         <at 0.5897 0>
         <at 0.898 0>
         <at 0.946 0>
         <at 0.9978 0.5035>
        ]}
       RedCurve {keys [
         {arriveTangent=0.1301, interpMode=RCIM_Cubic, time=0.2257, value=0.9963}
         <at 0.2939 0>
         <at 0.4639 0>
         <at 0.5897 2.2127>
         <at 0.898 2.2127>
         <at 0.946 2.7658>
         <at 0.9978 10>
        ]}
      }}
     CurveIndex=$Particles.Normalized Spawn Group
    )
    Opacity Pan U Offset=0.9143
    Opacity Pan U Speed=5
    Opacity Scale U=0.7
    Scale Alpha:NiagaraFloat=&FloatFromCurve(FloatCurve=DataInterface() {propertyValues {Curve {keys [
         {arriveTangent=13.8107, interpMode=RCIM_Cubic, leaveTangent=13.8107, tangentMode=RCTM_User, time=0, value=0}
         <at 0.2 1>
         {arriveTangent=-2.6535, interpMode=RCIM_Cubic, leaveTangent=-2.6535, tangentMode=RCTM_User, time=1, value=0}
        ]}}})
    Scale Color:Vector3f=<3 3 3>
   ) {enabled=false}
   &ScaleRibbonWidth(Ribbon Width Scale=&FloatFromCurve(FloatCurve=DataInterface() {propertyValues {Curve {keys [
         {interpMode=RCIM_Cubic, tangentMode=RCTM_User, time=0, value=1}
         {arriveTangent=-2.3513, interpMode=RCIM_Cubic, leaveTangent=-2.3513, tangentMode=RCTM_User, time=1, value=0}
        ]}}})) {enabled=false}
   /Game/Effects/Niagara/Modules/Curves/FN_HermiteCubicSpline_v2/FN_HermiteCubicSpline() {enabled=false, module=FN_HermiteCubicSpline002}
   &ScaleColor(
    Scale Alpha=&FloatFromCurve(
     FloatCurve=DataInterface() {propertyValues {Curve {defaultValue=1, keys [{tangentMode=RCTM_User, time=0, value=0}, {tangentMode=RCTM_User, time=1, value=1}]}}}
     CurveIndex=&ClampFloat(Float=&Subtract_Float(
       A=$Emitter.NormalizedLoopAge
       B=&Multiply_Float(A=&ReturnNormalizedExecIndex(), B=1)
      ))
    )
    ScaleA=true
    ScaleRGB=true
   ) {enabled=false}
   &DynamicMaterialParameters(
    Index 0 Param 1=0.1
    Index 0 Param 2=0.1
    Index 0 Param 4=9
    Module.Index0Float0Write=false
    Module.Index0Float1Write=false
    Module.Index0Float2Write=false
    Module.Index0Float3Write=true
   )
  ]
  rendererClasses [&NiagaraRibbonRendererProperties]
  renderers [
   {rendererClass=&NiagaraRibbonRendererProperties}
   {rendererClass=&NiagaraRibbonRendererProperties, rendererIndex=1}
  ]
 }
 Blue_Outer_Plasma001 {
  bEnabled=false
  emitterUpdateScript [
   &EmitterState(
    Life Cycle Mode=::NewEnumerator1
    Loop Delay=&RandomRangeFloat(Minimum=0.2, Maximum=0.3)
    Loop Duration=0.4
    UseLoopDelay=false
   )
   &SpawnRate(SpawnRate=&Multiply_Float(A=$User.ChargeProgress, B=5), Use Spawn Probability=false) {enabled=false}
   &SpawnBurst_Instantaneous(Spawn Count=10, Spawn Probability=$User.PlasmaEnabled, Use Spawn Probability=true)
   &SpawnBurst_Instantaneous(Spawn Count=10, Spawn Probability=$User.PlasmaEnabled, Spawn Time=0.15, Use Spawn Probability=true) {module=SpawnBurst_Instantaneous001}
   &SpawnBurst_Instantaneous(Spawn Count=10, Spawn Probability=$User.PlasmaEnabled, Spawn Time=0.25, Use Spawn Probability=true) {module=SpawnBurst_Instantaneous002}
  ]
  particleSpawnScript [
   &FN_SetSpawnGroup(Inclusivity=true, Normalize Spawn Group=true, Particles Per Group=10, Pattern=::NewEnumerator0, Set Ribbon ID and Ribbon Link Order=true)
   &SampleSkeletalMesh(
    AdditionalYawPitchRoll=false
    Bone / Socket Index=1
    Bone Sampling Mode:ENiagaraBoneSamplingMode=::NewEnumerator4
    Skeletal Mesh:NiagaraDataInterfaceSkeletalMesh=$System.Skeletal Mesh
   )
   &ShapeLocation(
    Apply To Particle Position=::NewEnumerator1
    MaskSpawnGroup=false
    Override Seed=false
    Ring Radius=&Lerp_Float(A=1500, B=5000, Alpha=$User.ChargeProgress)
    Shape Origin=$Particles.SampleSkeletalMesh.SampledPosition
    Shape Primitive=::NewEnumerator4
    UseOwnerScale=true
   )
   &SampleSkeletalMesh(
    AdditionalYawPitchRoll=false
    Bone / Socket Index:NiagaraInt32=&MakeIntFromFloat(Float=&FloatFromCurve(FloatCurve=DataInterface() {propertyValues {Curve {keys [<at 0 1>,<at 1 18>]}}}, CurveIndex=$User.ChargeProgress))
    Bone Sampling Mode:ENiagaraBoneSamplingMode=::NewEnumerator4
    Skeletal Mesh:NiagaraDataInterfaceSkeletalMesh=$System.Skeletal Mesh
   ) {module=SampleSkeletalMesh001}
   &ShapeLocation(
    Apply To Particle Position=::NewEnumerator1
    MaskSpawnGroup=false
    Override Seed=false
    Ring Radius=&Lerp_Float(A=1500, B=100, Alpha=$User.ChargeProgress)
    Shape Origin=$Particles.SampleSkeletalMesh001.SampledPosition
    Shape Primitive=::NewEnumerator4
    UseOwnerScale=true
   ) {module=ShapeLocation001}
   &InitializeRibbon(
    Lifetime=0.4
    Module.Write Color=true
    Module.Write Lifetime=true
    Module.Write Mass=true
    Module.Write Material Random=true
    Module.Write Position=false
    Module.Write RibbonTwist=false
    Module.Write RibbonWidth=true
    Ribbon Width=200
   )
   /SetVariables(Particles.HeightAddition:NiagaraFloat=&RemapRange(Input Value To Remap=$User.ChargeProgress, Input Min=0, Input Max=1, Output Min=0, Output Max=15000)) {isSetParameters=true, module=SetVariables_#3}
  ]
  particleUpdateScript [
   &ParticleState()
   &FN_HermiteCubicSpline(
    Position Blend:NiagaraFloat=$Particles.Normalized Spawn Group
    Position End:Vector3f=$User.HitLocation
    Position Start:Vector3f=$User.StartLocation
    Write Particles Position=true
   ) {enabled=false}
   /Game/Effects/Niagara/Modules/Curves/FN_HermiteCubicSpline_v1/FN_HermiteCubicSpline(
    Compute Direction=true
    Position Blend:NiagaraFloat=$Particles.Normalized Spawn Group
    Position End:Vector3f=&Lerp_Vector(
     A=&AddVectorToPosition(
      Position=$Particles.Initial.ShapeLocation.ShapePosition
      Vector=&MakeVector(X=0, Y=0, Z=$Particles.HeightAddition)
     )
     B=Linked() {linkedVariable {name=Particles.SampleSkeletalMesh001.SampledPosition, type=NiagaraPosition}}
     Alpha=&FloatFromCurve(FloatCurve=DataInterface() {propertyValues {Curve {keys [<at 0.99 0>,<at 1 0>]}}}, CurveIndex=$User.ChargeProgress)
    )
    Position Start:NiagaraPosition=$Particles.Initial.ShapeLocation.ShapePosition
   ) {enabled=false, module=FN_HermiteCubicSpline001}
   &FN_HermiteCubicSpline(
    Compute Direction=true
    Position Blend:NiagaraFloat=$Particles.Normalized Spawn Group
    Position End:NiagaraPosition=$Particles.Initial.ShapeLocation001.ShapePosition
    Position Start:NiagaraPosition=$Particles.Initial.ShapeLocation.ShapePosition
   ) {module=FN_HermiteCubicSpline003}
   /SetVariables(Particles.Position=&Add_Vector(
     A=$Output.FN_HermiteCubicSpline003.Position
     B=&RotateVector(
      Vector To Rotate:Vector3f=&MakeVector(
       X=0
       Y=&Waveform(
        Global Input Position=&Add_Float(A=$Particles.Normalized Spawn Group, B=&RandomRangeFloat(Minimum=0, Maximum=1))
        Global Amplitude Scale=4
        "Waveform [1]"=::NewEnumerator0
        "[1] Frequency"=1
        "[1] Amplitude Min/Max"=<0 3>
        "[1] Scale"=&RandomRangeFloat(Minimum=-1, Maximum=1)
        "[2] Phase"=$Emitter.Age
        "[3] Phase"=$Emitter.Age
       )
       Z=&Waveform(
        Global Input Position=&Add_Float(A=$Particles.Normalized Spawn Group, B=&RandomRangeFloat(Minimum=0, Maximum=1))
        Global Amplitude Scale=20
        "Waveform [1]"=::NewEnumerator1
        "[1] Frequency"=1
        "[1] Amplitude Min/Max"=<0 3>
        "[1] Scale"=&RandomRangeFloat(Minimum=-1, Maximum=1)
        "[2] Phase"=$Emitter.Age
        "[3] Phase"=$Emitter.Age
       )
      )
      Yaw:NiagaraFloat=&MakeFloatFromVector(Vector=$Output.FN_HermiteCubicSpline001.Direction, Channel=::NewEnumerator0)
      Pitch:NiagaraFloat=&MakeFloatFromVector(Vector=$Output.FN_HermiteCubicSpline001.Direction, Channel=::NewEnumerator1)
      Roll:NiagaraFloat=&MakeFloatFromVector(Vector=$Output.FN_HermiteCubicSpline001.Direction, Channel=::NewEnumerator2)
      Rotation Rate:NiagaraFloat=$Engine.DeltaTime
     )
    )) {isSetParameters=true, module=SetVariables_#4}
   /SetVariables(Particles.Position=&Lerp_Vector(
     A=$Output.FN_HermiteCubicSpline003.Position
     B=Linked() {linkedVariable {name=Particles.Position, type=NiagaraPosition}}
     Alpha=&FloatFromCurve(
      FloatCurve=DataInterface() {propertyValues {Curve {keys [
          {arriveTangent=6.9713, interpMode=RCIM_Cubic, leaveTangent=6.9713, tangentMode=RCTM_User, time=0, value=0}
          {interpMode=RCIM_Cubic, tangentMode=RCTM_User, time=0.5, value=1}
          {arriveTangent=-6.3689, interpMode=RCIM_Cubic, leaveTangent=-6.3689, tangentMode=RCTM_User, time=1, value=0}
         ]}}}
      CurveIndex=$Particles.Normalized Spawn Group
     )
    )) {isSetParameters=true, module=SetVariables_#5}
   &FN_RibbonMaterial(
    Color=&ColorFromCurve(
     ColorCurve=DataInterface() {propertyValues {
       AlphaCurve {keys [<at 0.0106 1 RCIM_Cubic>,<at 0.9796 1>]}
       BlueCurve {keys [
         {arriveTangent=-3.673, interpMode=RCIM_Cubic, time=0.2257, value=7}
         <at 0.2939 4>
         <at 0.4639 4>
         <at 0.5897 4>
         <at 0.898 4>
         <at 0.946 5>
         <at 0.9978 7.8247>
        ]}
       GreenCurve {keys [
         {arriveTangent=0.1965, interpMode=RCIM_Cubic, time=0.2257, value=2.7435}
         <at 0.2939 0.1534>
         <at 0.4639 0.1731>
         <at 0.5897 0>
         <at 0.898 0>
         <at 0.946 0>
         <at 0.9978 0.5035>
        ]}
       RedCurve {keys [
         {arriveTangent=0.1301, interpMode=RCIM_Cubic, time=0.2257, value=0.9963}
         <at 0.2939 0>
         <at 0.4639 0>
         <at 0.5897 2.2127>
         <at 0.898 2.2127>
         <at 0.946 2.7658>
         <at 0.9978 10>
        ]}
      }}
     CurveIndex=$Particles.Normalized Spawn Group
    )
    Opacity Pan U Offset=0.9143
    Opacity Pan U Speed=5
    Opacity Scale U=0.7
    Scale Alpha:NiagaraFloat=&FloatFromCurve(FloatCurve=DataInterface() {propertyValues {Curve {keys [
         {arriveTangent=13.8107, interpMode=RCIM_Cubic, leaveTangent=13.8107, tangentMode=RCTM_User, time=0, value=0}
         <at 0.2 1>
         {arriveTangent=-2.6535, interpMode=RCIM_Cubic, leaveTangent=-2.6535, tangentMode=RCTM_User, time=1, value=0}
        ]}}})
    Scale Color:Vector3f=<3 3 3>
   )
   &ScaleRibbonWidth(Ribbon Width Scale=&FloatFromCurve(FloatCurve=DataInterface() {propertyValues {Curve {keys [
         {interpMode=RCIM_Cubic, tangentMode=RCTM_User, time=0, value=1}
         {arriveTangent=-2.3513, interpMode=RCIM_Cubic, leaveTangent=-2.3513, tangentMode=RCTM_User, time=1, value=0}
        ]}}})) {enabled=false}
   /Game/Effects/Niagara/Modules/Curves/FN_HermiteCubicSpline_v2/FN_HermiteCubicSpline() {enabled=false, module=FN_HermiteCubicSpline002}
   &ScaleColor(
    Scale Alpha=&FloatFromCurve(
     FloatCurve=DataInterface() {propertyValues {Curve {defaultValue=1, keys [{tangentMode=RCTM_User, time=0, value=0}, {tangentMode=RCTM_User, time=1, value=1}]}}}
     CurveIndex=&ClampFloat(Float=&Subtract_Float(
       A=$Emitter.NormalizedLoopAge
       B=&Multiply_Float(A=&ReturnNormalizedExecIndex(), B=1)
      ))
    )
    ScaleA=true
    ScaleRGB=true
   ) {enabled=false}
  ]
  rendererClasses [&NiagaraRibbonRendererProperties]
  renderers [{rendererClass=&NiagaraRibbonRendererProperties}]
 }
 Bright_Ribbon_01 {
  bEnabled=false
  emitterSpawnScript [/SetVariables(
    Emitter.OffsetBeam:NiagaraFloat=&RandomRangeFloat(Minimum=0, Maximum=1)
    Emitter.ParticleLife:NiagaraFloat=&RandomRangeFloat(Minimum=0.5, Maximum=2.25)
    Emitter.ZapSpeed_High=100
    Emitter.ZapSpeed_Low=100
   ) {isSetParameters=true, module=SetVariables_#6}]
  emitterUpdateScript [
   /SetVariables(Emitter.ParticleLife:NiagaraFloat=&RandomRangeFloat(Minimum=0.1, Maximum=0.2)) {isSetParameters=true, module=SetVariables_#7}
   &EmitterState(
    Life Cycle Mode=::NewEnumerator1
    Loop Duration=0.4
    Recalculate Duration Each Loop=true
    Scalability Mode=::NewEnumerator1
    UseLoopDelay=false
   )
   &SpawnBurst_Instantaneous(Spawn Count=10, Spawn Probability=$User.PlasmaEnabled, Use Spawn Probability=true) {module=SpawnBurst_Instantaneous002}
   &SpawnBurst_Instantaneous(Spawn Count=10, Spawn Probability=$User.PlasmaEnabled, Spawn Time=0.15, Use Spawn Probability=true) {module=SpawnBurst_Instantaneous001}
   &SpawnBurst_Instantaneous(Spawn Count=10, Spawn Probability=$User.PlasmaEnabled, Spawn Time=0.25, Use Spawn Probability=true)
  ]
  particleSpawnScript [
   &InitializeParticle(
    Lifetime=0.2
    Position Mode=::NewEnumerator0
    Ribbon Width=&RandomRangeFloat(Minimum=150, Maximum=300)
    Ribbon Width Mode=::NewEnumerator1
    Write Material Random=true
   )
   &SampleSkeletalMesh(
    AdditionalYawPitchRoll=false
    Bone / Socket Index=1
    Bone Sampling Mode:ENiagaraBoneSamplingMode=::NewEnumerator4
    Skeletal Mesh:NiagaraDataInterfaceSkeletalMesh=$System.Skeletal Mesh
   )
   &ShapeLocation(
    Apply To Particle Position=::NewEnumerator1
    MaskSpawnGroup=false
    Override Seed=false
    Ring Radius=&Lerp_Float(A=1500, B=3000, Alpha=$User.ChargeProgress)
    Shape Origin=$Particles.SampleSkeletalMesh.SampledPosition
    Shape Primitive=::NewEnumerator4
    UseOwnerScale=true
   )
   &SampleSkeletalMesh(
    AdditionalYawPitchRoll=false
    Bone / Socket Index:NiagaraInt32=&MakeIntFromFloat(Float=&FloatFromCurve(FloatCurve=DataInterface() {propertyValues {Curve {keys [<at 0 1>,<at 1 18>]}}}, CurveIndex=$User.ChargeProgress))
    Bone Sampling Mode:ENiagaraBoneSamplingMode=::NewEnumerator4
    Skeletal Mesh:NiagaraDataInterfaceSkeletalMesh=$System.Skeletal Mesh
   ) {module=SampleSkeletalMesh001}
   &ShapeLocation(
    Apply To Particle Position=::NewEnumerator1
    MaskSpawnGroup=false
    Override Seed=false
    Ring Radius=&Lerp_Float(A=1500, B=100, Alpha=$User.ChargeProgress)
    Shape Origin=$Particles.SampleSkeletalMesh001.SampledPosition
    Shape Primitive=::NewEnumerator4
    UseOwnerScale=true
   ) {module=ShapeLocation001}
   /SetVariables(
    Particles.DistortionScaleX:NiagaraFloat=&RandomRangeFloat(Minimum=100, Maximum=300)
    Particles.DistortionScaleY:NiagaraFloat=&RandomRangeFloat(Minimum=100, Maximum=300)
    Particles.DistortionScaleZ:NiagaraFloat=&RandomRangeFloat(Minimum=100, Maximum=550)
    Particles.ExecIndex:NiagaraFloat=&ReturnNormalizedExecIndex()
    Particles.ZapSpeed_High:NiagaraFloat=$Emitter.ZapSpeed_High
    Particles.ZapSpeed_Low:NiagaraFloat=$Emitter.ZapSpeed_Low
   ) {isSetParameters=true, module=SetVariables_#8}
   &DynamicMaterialParameters(
    Index 0 Param 1=0.2
    Index 0 Param 2=0.1
    Index 0 Param 3=$Emitter.OffsetBeam
    Index 0 Param 4=1
    Module.Index0Float0Write=true
    Module.Index0Float1Write=true
    Module.Index0Float2Write=true
    Module.Index0Float3Write=true
   )
   &Color(Color=&Multiply_LinearColorByFloat(
     Linear Color=&ColorFromCurve(
      ColorCurve=DataInterface() {propertyValues {
        AlphaCurve {keys [<at 0 1 RCIM_Cubic>,<at 1 1 RCIM_Cubic>]}
        BlueCurve {keys [<at 0 1 RCIM_Cubic>,<at 1 1 RCIM_Cubic>]}
        GreenCurve {keys [<at 0 0.7583 RCIM_Cubic>,<at 1 0.1964 RCIM_Cubic>]}
        RedCurve {keys [<at 0 0.0659 RCIM_Cubic>,<at 1 0.8209 RCIM_Cubic>]}
       }}
      CurveIndex=$Particles.ExecIndex
     )
     Float=2
    ))
  ]
  particleUpdateScript [
   &SolveForcesAndVelocity(Clamp Velocity=false, Limit Acceleration=false)
   &ParticleState()
   /Game/Effects/Niagara/Modules/Curves/FN_HermiteCubicSpline_v1/FN_HermiteCubicSpline(
    Compute Direction=true
    Position Blend:NiagaraFloat=$Particles.ExecIndex
    Position End:NiagaraPosition=$Particles.Initial.ShapeLocation001.ShapePosition
    Position Start:NiagaraPosition=$Particles.Initial.ShapeLocation.ShapePosition
   ) {enabled=false, module=FN_HermiteCubicSpline003}
   /SetVariables(Particles.Position=&Add_Vector(
     A=$Output.FN_HermiteCubicSpline003.Position
     B=&RotateVector(
      Vector To Rotate:Vector3f=&MakeVector(
       X=0
       Y=&Waveform(
        Global Input Position=&Add_Float(A=$Particles.ExecIndex, B=&RandomRangeFloat(Minimum=0, Maximum=1))
        Global Amplitude Scale=4
        "Waveform [1]"=::NewEnumerator0
        "[1] Frequency"=1
        "[1] Amplitude Min/Max"=<0 0>
        "[1] Scale"=&RandomRangeFloat(Minimum=-1, Maximum=1)
        "[2] Phase"=$Emitter.Age
        "[3] Phase"=$Emitter.Age
       )
       Z=&Waveform(
        Global Input Position=&Add_Float(A=$Particles.ExecIndex, B=&RandomRangeFloat(Minimum=0, Maximum=1))
        Global Amplitude Scale=20
        "Waveform [1]"=::NewEnumerator1
        "[1] Frequency"=1
        "[1] Amplitude Min/Max"=<0 3>
        "[1] Scale"=&RandomRangeFloat(Minimum=-1, Maximum=1)
        "[2] Phase"=$Emitter.Age
        "[3] Phase"=$Emitter.Age
       )
      )
      Yaw:NiagaraFloat=&MakeFloatFromVector(Vector=$Output.FN_HermiteCubicSpline003.Direction, Channel=::NewEnumerator0)
      Pitch:NiagaraFloat=&MakeFloatFromVector(Vector=$Output.FN_HermiteCubicSpline003.Direction, Channel=::NewEnumerator1)
      Roll:NiagaraFloat=&MakeFloatFromVector(Vector=$Output.FN_HermiteCubicSpline003.Direction, Channel=::NewEnumerator2)
      Rotation Rate:NiagaraFloat=$Engine.DeltaTime
     )
    )) {enabled=false, isSetParameters=true, module=SetVariables_#9}
   /SetVariables(Particles.Position=&Lerp_Vector(
     A=$Output.FN_HermiteCubicSpline003.Position
     B=Linked() {linkedVariable {name=Particles.Position, type=NiagaraPosition}}
     Alpha=&FloatFromCurve(
      FloatCurve=DataInterface() {propertyValues {Curve {keys [
          {arriveTangent=6.9713, interpMode=RCIM_Cubic, leaveTangent=6.9713, tangentMode=RCTM_User, time=0, value=0}
          {interpMode=RCIM_Cubic, tangentMode=RCTM_User, time=0.5, value=1}
          {arriveTangent=-6.3689, interpMode=RCIM_Cubic, leaveTangent=-6.3689, tangentMode=RCTM_User, time=1, value=0}
         ]}}}
      CurveIndex=$Particles.ExecIndex
     )
    )) {enabled=false, isSetParameters=true, module=SetVariables_#10}
   /SetVariables(Particles.Position=&AddVectorToPosition(
     Position=&Lerp_Vector(
      A=Linked() {linkedVariable {name=Particles.Initial.ShapeLocation.ShapePosition, type=NiagaraPosition}}
      B=Linked() {linkedVariable {name=Particles.Initial.ShapeLocation001.ShapePosition, type=NiagaraPosition}}
      Alpha=$Particles.ExecIndex
     )
     Vector=&Multiply_VectorByFloat(
      Vector=&MakeVector(
       X=&Waveform(
        Waveform Count=::NewEnumerator1
        Global Phase=&Multiply_Float(A=$Emitter.Age, B=&RandomRangeFloat(Minimum=$Particles.ZapSpeed_Low, Maximum=$Particles.ZapSpeed_High))
        "Waveform [1]"=::NewEnumerator0
        "[1] Frequency"=&RandomRangeFloat(Minimum=2, Maximum=4)
        "[1] Amplitude Min/Max"=<0 1>
        "[1] Scale"=$Particles.DistortionScaleX
        "[2] Average"=false
        "[2] Input Position"=0
        "[2] Phase"=$Emitter.Age
        "[2] Phase Scale"=1
        "[2] Compound Frequency"=2
        "[2] Frequency"=2
        "[2] Amplitude Min/Max"=<0 1>
        "[2] Scale"=$Particles.DistortionScaleX
        "[2] Bias"=0
        "[2] Exponent"=1
        "[2] Pulse Duration"=0.1
        "[2] Clip Triangle Min/Max"=<0 0>
        "[2] Random Seed"=0
       )
       Y=&Waveform(
        Waveform Count=::NewEnumerator2
        Global Phase=&Multiply_Float(A=$Emitter.Age, B=&RandomRangeFloat(Minimum=$Particles.ZapSpeed_Low, Maximum=$Particles.ZapSpeed_High))
        "Waveform [1]"=::NewEnumerator0
        "[1] Frequency"=1
        "[1] Amplitude Min/Max"=<0 1>
        "[1] Scale"=$Particles.DistortionScaleY
        "[2] Average"=false
        "[2] Input Position"=0
        "[2] Phase"=$Emitter.Age
        "[2] Phase Scale"=1
        "[2] Compound Frequency"=2
        "[2] Frequency"=2
        "[2] Amplitude Min/Max"=<0 1>
        "[2] Scale"=$Particles.DistortionScaleX
        "[2] Bias"=0
        "[2] Exponent"=1
        "[2] Pulse Duration"=0.1
        "[2] Clip Triangle Min/Max"=<0 0>
        "[2] Random Seed"=0
        "[3] Average"=false
        "Waveform [3]"=::NewEnumerator1
        "[3] Input Position"=0
        "[3] Phase"=$Emitter.Age
        "[3] Phase Scale"=1
        "[3] Frequency"=4
        "[3] Compound Frequency"=2
        "[3] Amplitude Min/Max"=<0 1>
        "[3] Scale"=1
        "[3] Bias"=0
        "[3] Exponent"=1
        "[3] Pulse Duration"=0.1
        "[3] Clip Triangle Min/Max"=<0 0>
        "[3] Random Seed"=0
       )
       Z=&Waveform(
        Waveform Count=::NewEnumerator1
        Global Phase=&Multiply_Float(A=$Emitter.Age, B=&RandomRangeFloat(Minimum=$Particles.ZapSpeed_Low, Maximum=$Particles.ZapSpeed_High))
        "Waveform [1]"=::NewEnumerator0
        "[1] Frequency"=1
        "[1] Amplitude Min/Max"=<0 1>
        "[1] Scale"=$Particles.DistortionScaleZ
        "[2] Average"=false
        "[2] Input Position"=0
        "[2] Phase"=$Emitter.Age
        "[2] Phase Scale"=1
        "[2] Compound Frequency"=2
        "[2] Frequency"=&RandomRangeFloat(Minimum=0, Maximum=2)
        "[2] Amplitude Min/Max"=<-0.1 0.1>
        "[2] Scale"=$Particles.DistortionScaleZ
        "[2] Bias"=0
        "[2] Exponent"=2
        "[2] Pulse Duration"=0.1
        "[2] Clip Triangle Min/Max"=<0 0>
        "[2] Random Seed"=0
        "Waveform [3]"=::NewEnumerator1
       )
      )
      Float=&FloatFromCurve(
       FloatCurve=DataInterface() {propertyValues {Curve {
          defaultValue=1
          keys [
           {arriveTangent=1.8269, interpMode=RCIM_Cubic, leaveTangent=1.8269, tangentMode=RCTM_User, time=0, value=0}
           {interpMode=RCIM_Cubic, tangentMode=RCTM_User, time=0.5, value=1}
           {arriveTangent=-1.1725, tangentMode=RCTM_User, time=1, value=0}
          ]
         }}}
       CurveIndex=$Particles.ExecIndex
      )
     )
    )) {isSetParameters=true, module=SetVariables_#11}
   &DynamicMaterialParameters(
    Index 0 Param 4=&RandomRangeFloat(Minimum=1, Maximum=2)
    Module.Index0Float0Write=false
    Module.Index0Float1Write=false
    Module.Index0Float2Write=false
    Module.Index0Float3Write=true
   ) {module=DynamicMaterialParameters001}
  ]
  rendererClasses [&NiagaraRibbonRendererProperties]
  renderers [{rendererClass=&NiagaraRibbonRendererProperties}]
 }
 Bright_Ribbon_01001 {
  bEnabled=false
  emitterSpawnScript [/SetVariables(
    Emitter.OffsetBeam:NiagaraFloat=&RandomRangeFloat(Minimum=0, Maximum=1)
    Emitter.ParticleLife:NiagaraFloat=&RandomRangeFloat(Minimum=0.5, Maximum=2.25)
    Emitter.ZapSpeed_High=100
    Emitter.ZapSpeed_Low=100
   ) {isSetParameters=true, module=SetVariables_#6}]
  emitterUpdateScript [
   /SetVariables(Emitter.ParticleLife:NiagaraFloat=&RandomRangeFloat(Minimum=0.1, Maximum=0.2)) {isSetParameters=true, module=SetVariables_#7}
   &EmitterState(
    Life Cycle Mode=::NewEnumerator1
    Loop Duration=0.4
    Recalculate Duration Each Loop=true
    Scalability Mode=::NewEnumerator1
    UseLoopDelay=false
   )
   &SpawnBurst_Instantaneous(Spawn Count=10, Spawn Probability=$User.PlasmaEnabled, Use Spawn Probability=true) {module=SpawnBurst_Instantaneous002}
   &SpawnBurst_Instantaneous(Spawn Count=10, Spawn Probability=$User.PlasmaEnabled, Spawn Time=0.15, Use Spawn Probability=true) {module=SpawnBurst_Instantaneous001}
   &SpawnBurst_Instantaneous(Spawn Count=10, Spawn Probability=$User.PlasmaEnabled, Spawn Time=0.25, Use Spawn Probability=true)
  ]
  particleSpawnScript [
   &InitializeParticle(
    Lifetime=0.2
    Position Mode=::NewEnumerator0
    Ribbon Width=&RandomRangeFloat(Minimum=250, Maximum=400)
    Ribbon Width Mode=::NewEnumerator1
    Write Material Random=true
   )
   &SampleSkeletalMesh(
    AdditionalYawPitchRoll=false
    Bone / Socket Index=1
    Bone Sampling Mode:ENiagaraBoneSamplingMode=::NewEnumerator4
    Skeletal Mesh:NiagaraDataInterfaceSkeletalMesh=$System.Skeletal Mesh
   )
   &ShapeLocation(
    Apply To Particle Position=::NewEnumerator1
    MaskSpawnGroup=false
    Override Seed=false
    Ring Radius=&Lerp_Float(A=1500, B=3000, Alpha=$User.ChargeProgress)
    Shape Origin=$Particles.SampleSkeletalMesh.SampledPosition
    Shape Primitive=::NewEnumerator4
    UseOwnerScale=true
   )
   &SampleSkeletalMesh(
    AdditionalYawPitchRoll=false
    Bone / Socket Index:NiagaraInt32=&MakeIntFromFloat(Float=&FloatFromCurve(FloatCurve=DataInterface() {propertyValues {Curve {keys [<at 0 1>,<at 1 18>]}}}, CurveIndex=$User.ChargeProgress))
    Bone Sampling Mode:ENiagaraBoneSamplingMode=::NewEnumerator4
    Skeletal Mesh:NiagaraDataInterfaceSkeletalMesh=$System.Skeletal Mesh
   ) {module=SampleSkeletalMesh001}
   &ShapeLocation(
    Apply To Particle Position=::NewEnumerator1
    MaskSpawnGroup=false
    Override Seed=false
    Ring Radius=&Lerp_Float(A=1500, B=100, Alpha=$User.ChargeProgress)
    Shape Origin=$Particles.SampleSkeletalMesh001.SampledPosition
    Shape Primitive=::NewEnumerator4
    UseOwnerScale=true
   ) {module=ShapeLocation001}
   /SetVariables(
    Particles.DistortionScaleX:NiagaraFloat=&RandomRangeFloat(Minimum=100, Maximum=300)
    Particles.DistortionScaleY:NiagaraFloat=&RandomRangeFloat(Minimum=100, Maximum=300)
    Particles.DistortionScaleZ:NiagaraFloat=&RandomRangeFloat(Minimum=100, Maximum=550)
    Particles.ExecIndex:NiagaraFloat=&ReturnNormalizedExecIndex()
    Particles.ZapSpeed_High:NiagaraFloat=$Emitter.ZapSpeed_High
    Particles.ZapSpeed_Low:NiagaraFloat=$Emitter.ZapSpeed_Low
   ) {isSetParameters=true, module=SetVariables_#8}
   &DynamicMaterialParameters(
    Index 0 Param 1=0.2
    Index 0 Param 2=0.1
    Index 0 Param 3=$Emitter.OffsetBeam
    Index 0 Param 4=1
    Module.Index0Float0Write=true
    Module.Index0Float1Write=true
    Module.Index0Float2Write=true
    Module.Index0Float3Write=true
   )
   &Color(Color=&Multiply_LinearColorByFloat(
     Linear Color=&ColorFromCurve(
      ColorCurve=DataInterface() {propertyValues {
        AlphaCurve {keys [<at 0 1 RCIM_Cubic>,<at 1 1 RCIM_Cubic>]}
        BlueCurve {keys [<at 0 1 RCIM_Cubic>,<at 1 1 RCIM_Cubic>]}
        GreenCurve {keys [<at 0 0.7583 RCIM_Cubic>,<at 1 0.1964 RCIM_Cubic>]}
        RedCurve {keys [<at 0 0.0659 RCIM_Cubic>,<at 1 0.8209 RCIM_Cubic>]}
       }}
      CurveIndex=$Particles.ExecIndex
     )
     Float=2
    ))
  ]
  particleUpdateScript [
   &SolveForcesAndVelocity(Clamp Velocity=false, Limit Acceleration=false)
   &ParticleState()
   &FN_HermiteCubicSpline(
    Compute Direction=true
    Position Blend:NiagaraFloat=$Particles.ExecIndex
    Position End:NiagaraPosition=$Particles.Initial.ShapeLocation001.ShapePosition
    Position Start:NiagaraPosition=$Particles.Initial.ShapeLocation.ShapePosition
   ) {enabled=false, module=FN_HermiteCubicSpline003}
   /SetVariables(Particles.Position=&Add_Vector(
     A=$Output.FN_HermiteCubicSpline003.Position
     B=&RotateVector(
      Vector To Rotate:Vector3f=&MakeVector(
       X=0
       Y=&Waveform(
        Global Input Position=&Add_Float(A=$Particles.ExecIndex, B=&RandomRangeFloat(Minimum=0, Maximum=1))
        Global Amplitude Scale=4
        "Waveform [1]"=::NewEnumerator0
        "[1] Frequency"=1
        "[1] Amplitude Min/Max"=<0 0>
        "[1] Scale"=&RandomRangeFloat(Minimum=-1, Maximum=1)
        "[2] Phase"=$Emitter.Age
        "[3] Phase"=$Emitter.Age
       )
       Z=&Waveform(
        Global Input Position=&Add_Float(A=$Particles.ExecIndex, B=&RandomRangeFloat(Minimum=0, Maximum=1))
        Global Amplitude Scale=20
        "Waveform [1]"=::NewEnumerator1
        "[1] Frequency"=1
        "[1] Amplitude Min/Max"=<0 3>
        "[1] Scale"=&RandomRangeFloat(Minimum=-1, Maximum=1)
        "[2] Phase"=$Emitter.Age
        "[3] Phase"=$Emitter.Age
       )
      )
      Yaw:NiagaraFloat=&MakeFloatFromVector(Vector=$Output.FN_HermiteCubicSpline003.Direction, Channel=::NewEnumerator0)
      Pitch:NiagaraFloat=&MakeFloatFromVector(Vector=$Output.FN_HermiteCubicSpline003.Direction, Channel=::NewEnumerator1)
      Roll:NiagaraFloat=&MakeFloatFromVector(Vector=$Output.FN_HermiteCubicSpline003.Direction, Channel=::NewEnumerator2)
      Rotation Rate:NiagaraFloat=$Engine.DeltaTime
     )
    )) {enabled=false, isSetParameters=true, module=SetVariables_#9}
   /SetVariables(Particles.Position=&Lerp_Vector(
     A=$Output.FN_HermiteCubicSpline003.Position
     B=Linked() {linkedVariable {name=Particles.Position, type=NiagaraPosition}}
     Alpha=&FloatFromCurve(
      FloatCurve=DataInterface() {propertyValues {Curve {keys [
          {arriveTangent=6.9713, interpMode=RCIM_Cubic, leaveTangent=6.9713, tangentMode=RCTM_User, time=0, value=0}
          {interpMode=RCIM_Cubic, tangentMode=RCTM_User, time=0.5, value=1}
          {arriveTangent=-6.3689, interpMode=RCIM_Cubic, leaveTangent=-6.3689, tangentMode=RCTM_User, time=1, value=0}
         ]}}}
      CurveIndex=$Particles.ExecIndex
     )
    )) {enabled=false, isSetParameters=true, module=SetVariables_#10}
   /SetVariables(Particles.Position=&AddVectorToPosition(
     Position=&Lerp_Vector(
      A=Linked() {linkedVariable {name=Particles.Initial.ShapeLocation.ShapePosition, type=NiagaraPosition}}
      B=Linked() {linkedVariable {name=Particles.Initial.ShapeLocation001.ShapePosition, type=NiagaraPosition}}
      Alpha=$Particles.ExecIndex
     )
     Vector=&Multiply_VectorByFloat(
      Vector=&MakeVector(
       X=&Waveform(
        Waveform Count=::NewEnumerator1
        Global Phase=&Multiply_Float(A=$Emitter.Age, B=&RandomRangeFloat(Minimum=$Particles.ZapSpeed_Low, Maximum=$Particles.ZapSpeed_High))
        "Waveform [1]"=::NewEnumerator0
        "[1] Frequency"=&RandomRangeFloat(Minimum=2, Maximum=4)
        "[1] Amplitude Min/Max"=<0 1>
        "[1] Scale"=$Particles.DistortionScaleX
        "[2] Average"=false
        "[2] Input Position"=0
        "[2] Phase"=$Emitter.Age
        "[2] Phase Scale"=1
        "[2] Compound Frequency"=2
        "[2] Frequency"=2
        "[2] Amplitude Min/Max"=<0 1>
        "[2] Scale"=$Particles.DistortionScaleX
        "[2] Bias"=0
        "[2] Exponent"=1
        "[2] Pulse Duration"=0.1
        "[2] Clip Triangle Min/Max"=<0 0>
        "[2] Random Seed"=0
       )
       Y=&Waveform(
        Waveform Count=::NewEnumerator2
        Global Phase=&Multiply_Float(A=$Emitter.Age, B=&RandomRangeFloat(Minimum=$Particles.ZapSpeed_Low, Maximum=$Particles.ZapSpeed_High))
        "Waveform [1]"=::NewEnumerator0
        "[1] Frequency"=1
        "[1] Amplitude Min/Max"=<0 1>
        "[1] Scale"=$Particles.DistortionScaleY
        "[2] Average"=false
        "[2] Input Position"=0
        "[2] Phase"=$Emitter.Age
        "[2] Phase Scale"=1
        "[2] Compound Frequency"=2
        "[2] Frequency"=2
        "[2] Amplitude Min/Max"=<0 1>
        "[2] Scale"=$Particles.DistortionScaleX
        "[2] Bias"=0
        "[2] Exponent"=1
        "[2] Pulse Duration"=0.1
        "[2] Clip Triangle Min/Max"=<0 0>
        "[2] Random Seed"=0
        "[3] Average"=false
        "Waveform [3]"=::NewEnumerator1
        "[3] Input Position"=0
        "[3] Phase"=$Emitter.Age
        "[3] Phase Scale"=1
        "[3] Frequency"=4
        "[3] Compound Frequency"=2
        "[3] Amplitude Min/Max"=<0 1>
        "[3] Scale"=1
        "[3] Bias"=0
        "[3] Exponent"=1
        "[3] Pulse Duration"=0.1
        "[3] Clip Triangle Min/Max"=<0 0>
        "[3] Random Seed"=0
       )
       Z=&Waveform(
        Waveform Count=::NewEnumerator1
        Global Phase=&Multiply_Float(A=$Emitter.Age, B=&RandomRangeFloat(Minimum=$Particles.ZapSpeed_Low, Maximum=$Particles.ZapSpeed_High))
        "Waveform [1]"=::NewEnumerator0
        "[1] Frequency"=1
        "[1] Amplitude Min/Max"=<0 1>
        "[1] Scale"=$Particles.DistortionScaleZ
        "[2] Average"=false
        "[2] Input Position"=0
        "[2] Phase"=$Emitter.Age
        "[2] Phase Scale"=1
        "[2] Compound Frequency"=2
        "[2] Frequency"=&RandomRangeFloat(Minimum=0, Maximum=2)
        "[2] Amplitude Min/Max"=<-0.1 0.1>
        "[2] Scale"=$Particles.DistortionScaleZ
        "[2] Bias"=0
        "[2] Exponent"=2
        "[2] Pulse Duration"=0.1
        "[2] Clip Triangle Min/Max"=<0 0>
        "[2] Random Seed"=0
        "Waveform [3]"=::NewEnumerator1
       )
      )
      Float=&FloatFromCurve(
       FloatCurve=DataInterface() {propertyValues {Curve {
          defaultValue=1
          keys [
           {arriveTangent=1.8269, interpMode=RCIM_Cubic, leaveTangent=1.8269, tangentMode=RCTM_User, time=0, value=0}
           {interpMode=RCIM_Cubic, tangentMode=RCTM_User, time=0.5, value=1}
           {arriveTangent=-1.1725, tangentMode=RCTM_User, time=1, value=0}
          ]
         }}}
       CurveIndex=$Particles.ExecIndex
      )
     )
    )) {isSetParameters=true, module=SetVariables_#11}
   &DynamicMaterialParameters(
    Index 0 Param 4=&RandomRangeFloat(Minimum=1, Maximum=2)
    Module.Index0Float0Write=false
    Module.Index0Float1Write=false
    Module.Index0Float2Write=false
    Module.Index0Float3Write=true
   ) {module=DynamicMaterialParameters001}
  ]
  rendererClasses [&NiagaraRibbonRendererProperties]
  renderers [{rendererClass=&NiagaraRibbonRendererProperties}]
 }
 ChargeUp {
  bEnabled=false
  emitterUpdateScript [
   &EmitterState(Life Cycle Mode=::NewEnumerator1, Loop Behavior=::NewEnumerator1, UseLoopDelay=false)
   &SpawnBurst_Instantaneous(Use Spawn Probability=false, UseLoopCountLimit=false)
   &SpawnBurst_Instantaneous(Spawn Time=0.05, Use Spawn Probability=false, UseLoopCountLimit=false) {module=SpawnBurst_Instantaneous001}
  ]
  particleSpawnScript [
   &InitializeParticle(
    Lifetime=6
    Mesh Scale Mode=::NewEnumerator3
    Sprite Size=&Multiply_Vector2D_ByFloat(Float=&Multiply_Float(A=$User.ElementSizeScale, B=$User.PostSizeScale), Vector2D=<3948.9736 3201.8015>)
    Sprite Size Mode=::NewEnumerator4
    UsePositionOffset=false
    Write Material Random=true
   )
   &CameraOffset()
  ]
  particleUpdateScript [
   &ParticleState()
   &Timeline(
    Play Rate=&FloatFromCurve(
     FloatCurve=DataInterface() {propertyValues {Curve {keys [
         {interpMode=RCIM_Cubic, leaveTangent=-12.5334, tangentMode=RCTM_Break, time=0, value=4}
         {interpMode=RCIM_Cubic, tangentMode=RCTM_Break, time=0.7, value=1}
        ]}}}
     Scale Curve=0.692
    )
    Timeline End=1.0000000200408773e+20
   )
   &ScaleSpriteSize(Uniform Curve Sprite Scale=DataInterface() {propertyValues {Curve {keys [
        {arriveTangent=15.597, interpMode=RCIM_Cubic, leaveTangent=15.597, tangentMode=RCTM_User, time=0, value=0}
        {arriveTangent=0.3869, interpMode=RCIM_Cubic, leaveTangent=0.3869, tangentMode=RCTM_User, time=0.05, value=1}
        {arriveTangent=0.0224, interpMode=RCIM_Cubic, leaveTangent=0.0224, tangentMode=RCTM_User, time=1, value=1}
       ]}}})
   &DynamicMaterialParameters(
    Index 0 Param 1=&FloatFromCurve(
     FloatCurve=DataInterface() {propertyValues {Curve {keys [
         {interpMode=RCIM_Cubic, leaveTangent=-4.8343, tangentMode=RCTM_Break, time=0, value=2}
         {arriveTangent=-0.1577, interpMode=RCIM_Cubic, tangentMode=RCTM_Break, time=1, value=0}
        ]}}}
     Scale Curve=&Multiply_Float(A=2, B=$User.PostNoiseStrength)
    )
    Index 0 Param 2=&FloatFromCurve(
     FloatCurve=DataInterface() {propertyValues {Curve {keys [
         {interpMode=RCIM_Cubic, leaveTangent=-16.2213, tangentMode=RCTM_Break, time=0, value=4}
         {arriveTangent=-17.124, interpMode=RCIM_Cubic, tangentMode=RCTM_Break, time=0.25, value=0}
        ]}}}
     Scale Curve=&Multiply_Float(A=1, B=$User.PostStrength)
    )
    Index 0 Param 3=1
    Index 0 Param 4=1
    Index 1 Param 1=0.1
    Index 1 Param 2=1500
    Index 1 Param 3=0.5
    Index 1 Param 4=1
    Index 2 Param 1=$Particles.Timeline.Time
    Index 2 Param 2=$Particles.Timeline.Time
    Index 2 Param 3=$Particles.Timeline.Time
    Index 2 Param 4=&Multiply_Float(A=0.2, B=$User.ElementSizeScale)
    Index 3 Param 1=&FloatFromCurve(FloatCurve=DataInterface() {propertyValues {Curve {keys [<at 0.1 1>,<at 0.4 0>]}}})
    Index 3 Param 2=200
    Module.Index0Param0Write=true
    Module.Index0Param1Write=true
    Module.Index0Param2Write=true
    Module.Index0Param3Write=true
    Module.Index1Param0Write=true
    Module.Index1Param1Write=true
    Module.Index1Param2Write=true
    Module.Index1Param3Write=true
    Module.Index2Param0Write=true
    Module.Index2Param1Write=true
    Module.Index2Param2Write=true
    Module.Index2Param3Write=true
    Module.Index3Param0Write=true
    Module.Index3Param1Write=true
    Module.Index3Param2Write=true
    Module.Index3Param3Write=true
    Write Parameter Index 1=true
    Write Parameter Index 2=true
    Write Parameter Index 3=true
   )
   &ScaleColor(
    Scale Alpha=&FloatFromCurve(FloatCurve=DataInterface() {propertyValues {Curve {keys [{interpMode=RCIM_Cubic, leaveTangent=-6.3065, tangentMode=RCTM_Break, time=0.2, value=1}, <at 1 0>]}}})
    ScaleA=true
    ScaleRGB=true
   )
   /SetVariables(Particles.BlendTest:NiagaraFloat=&Subtract_Float(
     A=$User.PostProcess
     B=&FN_LOD_Distance_Remap(Start Distance=8000, End Distance=120000, Minimum Output=0, Maximum Output=1, Clamp Output=true)
    )) {isSetParameters=true}
   /SetVariables(Particles.ComponentsEnabled:NiagaraBool=&Boolean_NotOperation(Bool=&SetBoolByFloatComparison(
      A=&FN_LOD_Distance_Remap(Start Distance=4000, End Distance=6000, Minimum Output=0, Maximum Output=1, Clamp Output=true)
      B=1
     ))) {isSetParameters=true}
   &KillParticles(Kill Particles=&SetBoolByFloatComparison(
     A=&FloatFromCurve(FloatCurve=DataInterface() {propertyValues {Curve {keys [<at 0.9 1>,<at 1 -0.01>]}}})
     B=0
     Comparison Type=::NewEnumerator6
    ))
  ]
  rendererClasses [&NiagaraComponentRendererProperties]
  renderers [{rendererClass=&NiagaraComponentRendererProperties, rendererIndex=0}]
 }
 Energy_Offset {
  emitterUpdateScript [
   &EmitterState(Life Cycle Mode=::NewEnumerator1, Loop Behavior=::NewEnumerator1, Loop Delay=0.1, Loop Duration=3, UseLoopDelay=true)
   &SpawnBurst_Instantaneous(Spawn Count=1, Spawn Time=$User.TipDelay, Use Spawn Probability=false)
  ]
  particleSpawnScript [
   &InitializeParticle(Lifetime=2, Mesh Scale Mode=::NewEnumerator3, Mesh Uniform Scale=30, UsePositionOffset=false, Write Material Random=true)
   &InitialMeshOrientation(Orientation Vector=<0 1 0>, Rotation=<0 0.25 0>, Use Orientation Vector=true, Use Rotation Vector=true)
   &DynamicMaterialParameters(
    Index 0 Param 2=1
    Index 0 Param 3=1
    Index 0 Param 4=1
    Module.Param0WriteEnabled=false
    Module.Param1WriteEnabled=true
    Module.Param2WriteEnabled=true
    Module.Param3WriteEnabled=true
   )
   &InitialMeshOrientation(Orientation Vector=<0 -1 0>, Use Orientation Vector=true, Use Rotation Vector=true) {module=InitialMeshOrientation001}
   &SystemLocation(MaskSpawnGroup=false, Offset=<2 -2 -2>)
  ]
  particleUpdateScript [
   &ParticleState()
   &ScaleMeshSize(Scale Factor=&VectorFromFloat(Value=&FloatFromCurve(FloatCurve=DataInterface() {propertyValues {Curve {keys [<at 0 0.5>,<at 0.2 1.2>,<at 0.8 1>,<at 1 0.5>]}}})))
   &DynamicMaterialParameters(
    Index 0 Param 1=&FloatFromCurve(FloatCurve=DataInterface() {propertyValues {Curve {keys [{arriveTangent=2.4002, interpMode=RCIM_Cubic, leaveTangent=2.4002, tangentMode=RCTM_User, time=0, value=-0.63}, <at 1 0.7>]}}})
    Module.Param0WriteEnabled=true
    Module.Param1WriteEnabled=false
    Module.Param2WriteEnabled=false
    Module.Param3WriteEnabled=false
   ) {module=DynamicMaterialParameters001}
   &ScaleColor(
    Scale Alpha=&FloatFromCurve(
     FloatCurve=DataInterface() {propertyValues {Curve {keys [<at 0 0>,<at 0.3 1>,<at 1 0>]}}}
     Scale Curve=&FloatFromCurve(FloatCurve=DataInterface() {propertyValues {Curve {keys [<at 0 0.5>,<at 0.2 1>]}}}, Scale Curve=0.5)
    )
    Scale RGB=&VectorFromFloat(Value=&FloatFromCurve(FloatCurve=DataInterface() {propertyValues {Curve {keys [<at 0 0>,<at 0.3 1>,<at 1 0>]}}}, Scale Curve=0.15))
    ScaleA=true
    ScaleRGB=true
   )
   &ScaleColor(
    Scale Alpha=&FNTimeOfDayScalar(Night Scale=0.3, Morning Scale=0.3, Evening Scale=0.5)
    Scale RGB=&VectorFromFloat(Value=0.3)
    ScaleA=true
    ScaleRGB=true
   ) {module=ScaleColor001}
   &MeshRotationForce(Rotation=&MakeVector(
     X=&FloatFromCurve(FloatCurve=DataInterface() {propertyValues {Curve {keys [<at 0 1>,<at 1 0>]}}})
     Y=0
     Z=0
    )) {enabled=false}
   &SolveForcesAndVelocity(Clamp Velocity=false, Limit Acceleration=false)
   &SkeletalMeshLocation(Bone Sampling Mode=::NewEnumerator4, MaskSpawnGroup=false, Orientation Sampling=::NewEnumerator1, Skeletal Mesh=$System.Skeletal Mesh)
  ]
  rendererClasses [&NiagaraMeshRendererProperties]
  renderers [{rendererClass=&NiagaraMeshRendererProperties, rendererIndex=0}]
 }
 Energy_Offset001 {
  emitterUpdateScript [
   &EmitterState(Life Cycle Mode=::NewEnumerator1, Loop Behavior=::NewEnumerator1, Loop Delay=0.1, Loop Duration=3, UseLoopDelay=true)
   &SpawnBurst_Instantaneous(Spawn Count=1, Spawn Time=$User.TipDelay, Use Spawn Probability=false)
  ]
  particleSpawnScript [
   &InitializeParticle(
    Color=<rgba 2 2 2 0.1>
    Lifetime=1.8
    Mesh Scale Mode=::NewEnumerator3
    Mesh Uniform Scale=25
    UsePositionOffset=false
    Write Material Random=true
   )
   &InitialMeshOrientation(Orientation Vector=<0 -1 0>, Use Orientation Vector=true, Use Rotation Vector=true)
   &DynamicMaterialParameters(
    Index 0 Param 2=1
    Index 0 Param 3=1
    Index 0 Param 4=1
    Module.Param0WriteEnabled=false
    Module.Param1WriteEnabled=true
    Module.Param2WriteEnabled=true
    Module.Param3WriteEnabled=true
   )
   &SystemLocation(MaskSpawnGroup=false, Offset=<2 -2 -2>)
  ]
  particleUpdateScript [
   &ParticleState()
   &ScaleMeshSize(Scale Factor=&VectorFromFloat(Value=&FloatFromCurve(FloatCurve=DataInterface() {propertyValues {Curve {keys [<at 0 0.5>,<at 0.2 1.2>,<at 0.8 1>,<at 1 0.5>]}}})))
   &DynamicMaterialParameters(
    Index 0 Param 1=&FloatFromCurve(FloatCurve=DataInterface() {propertyValues {Curve {keys [{arriveTangent=-2.1669, interpMode=RCIM_Cubic, leaveTangent=-2.1669, tangentMode=RCTM_User, time=0, value=0}, <at 1 -1>]}}})
    Module.Param0WriteEnabled=true
    Module.Param1WriteEnabled=false
    Module.Param2WriteEnabled=false
    Module.Param3WriteEnabled=false
   ) {module=DynamicMaterialParameters002}
   &ScaleColor(
    Scale Alpha=&FloatFromCurve(
     FloatCurve=DataInterface() {propertyValues {Curve {keys [<at 0 0>,<at 0.3 1>,<at 1 0>]}}}
     Scale Curve=&FloatFromCurve(FloatCurve=DataInterface() {propertyValues {Curve {keys [<at 0 0.5>,<at 0.2 1>]}}}, Scale Curve=0.5)
    )
    Scale RGB=&VectorFromFloat(Value=&FloatFromCurve(FloatCurve=DataInterface() {propertyValues {Curve {keys [<at 0 1>,<at 0.3 1>,<at 1 0>]}}}, Scale Curve=0.15))
    ScaleA=true
    ScaleRGB=true
   )
   &ScaleColor(
    Scale Alpha=&FNTimeOfDayScalar(Night Scale=0.3, Morning Scale=0.3, Evening Scale=0.5)
    Scale RGB=&VectorFromFloat(Value=0.2)
    ScaleA=true
    ScaleRGB=true
   ) {module=ScaleColor001}
   &MeshRotationForce(Rotation=&MakeVector(
     X=&FloatFromCurve(FloatCurve=DataInterface() {propertyValues {Curve {keys [<at 0 1>,<at 1 0>]}}})
     Y=0
     Z=0
    )) {enabled=false}
   &SolveForcesAndVelocity(Clamp Velocity=false, Limit Acceleration=false)
   &SkeletalMeshLocation(Bone Sampling Mode=::NewEnumerator4, MaskSpawnGroup=false, Skeletal Mesh=$System.Skeletal Mesh)
  ]
  rendererClasses [&NiagaraMeshRendererProperties]
  renderers [{rendererClass=&NiagaraMeshRendererProperties, rendererIndex=0}]
 }
 Flare {
  emitterUpdateScript [
   &EmitterState(Life Cycle Mode=::NewEnumerator1, Loop Behavior=::NewEnumerator1, Loop Duration=3, UseLoopDelay=false)
   &SpawnBurst_Instantaneous(Spawn Count=2, Spawn Time=1.3, Use Spawn Probability=false)
  ]
  particleSpawnScript [
   &InitializeParticle(
    Color=<rgba 0.0832 0.4003 1 1>
    Lifetime=1.1
    Sprite Size Mode=::NewEnumerator3
    Uniform Sprite Size=7000
    UsePositionOffset=false
    Write Material Random=true
   )
   /SetVariables(Particles.FlareColor=<rgba 10 0 0.2747 1>) {isSetParameters=true}
   /SetVariables(Particles.FlareSize:Vector2f=&Vector2DFromFloat(Value=200)) {isSetParameters=true}
   &CameraOffset(Camera Offset Amount=300)
  ]
  particleUpdateScript [
   &ParticleState(Kill Particles When Lifetime Has Elapsed=false, Let Infinitely Lived Particles Die When Emitter Deactivates=true)
   &ScaleColor(
    Scale Alpha=&FNTimeOfDayScalar(Night Scale=0.5, Morning Scale=0.3, Evening Scale=0.5)
    Scale RGB=&VectorFromFloat(Value=&RandomRangeFloat(Minimum=0, Maximum=1, Evaluation Type=::NewEnumerator1))
    ScaleA=true
    ScaleRGB=true
   )
   &ScaleSpriteSize(Scale Factor=&Vector2DFromFloat(Value=&Multiply_Float(
      A=&RandomRangeFloat(Minimum=0.5, Maximum=1, Evaluation Type=::NewEnumerator1)
      B=&FloatFromCurve(FloatCurve=DataInterface() {propertyValues {Curve {keys [<at 0 0>,<at 0.15 1>,<at 1 0>]}}})
     )))
   /SetVariables(Particles.Color001:LinearColor=&MultiplyLinearColor(
     Linear Color=$Particles.FlareColor
     Scale Factor=&MakeFloat4FromVectorAndFloat(
      "Vector (RGB)"=&VectorFromFloat(Value=&RandomRangeFloat(Minimum=0, Maximum=2, Evaluation Type=::NewEnumerator1))
      "Float (Alpha)"=&FNTimeOfDayScalar(Night Scale=0.5, Morning Scale=0.3, Evening Scale=0.5)
     )
    )) {isSetParameters=true}
   /SetVariables(Particles.SpriteSize001:Vector2f=&Multiply_Vector2(
     A=$Particles.FlareSize
     B=&Vector2DFromFloat(Value=&Multiply_Float(
       A=&RandomRangeFloat(Minimum=0.5, Maximum=1, Evaluation Type=::NewEnumerator1)
       B=&FloatFromCurve(FloatCurve=DataInterface() {propertyValues {Curve {keys [<at 0 0>,<at 0.15 1>,<at 1 0>]}}})
      ))
    )) {isSetParameters=true}
   &SkeletalMeshLocation(Bone Sampling Mode=::NewEnumerator4, MaskSpawnGroup=false, Skeletal Mesh=$System.Skeletal Mesh)
  ]
  rendererClasses [&NiagaraSpriteRendererProperties]
  renderers [
   {rendererClass=&NiagaraSpriteRendererProperties}
   {rendererClass=&NiagaraSpriteRendererProperties, rendererIndex=1}
  ]
 }
 Flare_merged {
  emitterUpdateScript [
   &EmitterState(Life Cycle Mode=::NewEnumerator1, Loop Behavior=::NewEnumerator1, Loop Duration=4, UseLoopDelay=false)
   &SpawnBurst_Instantaneous(Spawn Time=$User.TipDelay, Use Spawn Probability=false, UseLoopCountLimit=false)
  ]
  particleSpawnScript [
   &InitializeParticle(
    Color=<rgba 2.2476 1.7687 7 1>
    Lifetime=3
    Sprite Rotation Mode=::NewEnumerator2
    Sprite Size=&SelectVector2DFromArray(
     Direct Array Index=&SpawnGroup()
     Vector 2D Selection Array=DataInterface() {propertyValues {FloatData [<800 800>, <1500 200>]}}
    )
    Sprite Size Mode=::NewEnumerator4
    UsePositionOffset=false
    Write Material Random=true
   )
   &CameraOffset(Camera Offset Amount=1500) {module=CameraOffset001}
  ]
  particleUpdateScript [
   &ParticleState()
   &ScaleSpriteSize(
    Uniform Curve Scale=7
    Uniform Curve Sprite Scale=DataInterface() {propertyValues {Curve {keys [
        {arriveTangent=0.3535, interpMode=RCIM_Cubic, leaveTangent=0.3535, tangentMode=RCTM_User, time=0, value=0}
        {arriveTangent=-0.0405, interpMode=RCIM_Cubic, leaveTangent=-0.0405, tangentMode=RCTM_User, time=1, value=1}
       ]}}}
   )
   &ScaleColor(
    Scale Alpha=&FloatFromCurve(
     FloatCurve=DataInterface() {propertyValues {Curve {keys [
         {arriveTangent=9.0725, interpMode=RCIM_Cubic, leaveTangent=9.0725, tangentMode=RCTM_User, time=0, value=0}
         {arriveTangent=-2.5249, interpMode=RCIM_Cubic, leaveTangent=-2.5249, tangentMode=RCTM_User, time=0.1408, value=0.9847}
         {arriveTangent=-0.0405, interpMode=RCIM_Cubic, leaveTangent=-0.0405, tangentMode=RCTM_User, time=1, value=0}
        ]}}}
     Scale Curve=0.5
    )
    Scale RGB=&VectorFromFloat(Value=1)
    ScaleA=true
    ScaleRGB=true
   ) {module=ScaleColor002}
   &SkeletalMeshLocation(Bone Sampling Mode=::NewEnumerator4, MaskSpawnGroup=false, Skeletal Mesh=$System.Skeletal Mesh)
  ]
  rendererClasses [&NiagaraSpriteRendererProperties]
  renderers [
   {rendererClass=&NiagaraSpriteRendererProperties}
   {rendererClass=&NiagaraSpriteRendererProperties, rendererIndex=1}
  ]
 }
 Lightning001 {
  bEnabled=false
  emitterUpdateScript [
   &EmitterState(
    Inactive Response=::NewEnumerator1
    Life Cycle Mode=::NewEnumerator1
    Loop Behavior=::NewEnumerator1
    Loop Delay=$System.BurstDelay
    Loop Duration=1.5
    UseLoopDelay=true
   )
   &SpawnRate(SpawnRate=&Multiply_Float(A=35, B=$System.FadeOut), Use Spawn Probability=false)
   &SpawnBurst_Instantaneous(Use Spawn Probability=false, UseLoopCountLimit=false)
  ]
  particleSpawnScript [
   /SetVariables(Particles.VisibilityTag=&SpawnGroup()) {isSetParameters=true}
   &InitializeParticle(
    Color=&Multiply_LinearColorByFloat(Linear Color=$System.ElimColor, Float=0.4)
    Lifetime Max=0.45
    Lifetime Min:NiagaraFloat=&MaskFloatBySpawnGroup(
     Masked Float=&RandomRangeFloat(Minimum=0.175, Maximum=0.4)
     Passthrough Float=&RandomRangeFloat(Minimum=0.25, Maximum=0.4)
    )
    Lifetime Mode=::NewEnumerator1
    Mesh Scale Max=&MaskVectorBySpawnGroup(Masked Vector=<1.3 1.3 2.5>, Passthrough Vector=<2 2 3>)
    Mesh Scale Min=&MaskVectorBySpawnGroup(Masked Vector=<1 1 2.5>, Passthrough Vector=<1 1 2>)
    Mesh Scale Mode=::NewEnumerator2
    Position Offset=<-3 -66 0>
    UsePositionOffset=true
    Write Material Random=true
   )
   &ShapeLocation(
    MaskSpawnGroup=false
    Override Seed=false
    Sphere Radius=&MaskFloatBySpawnGroup(Masked Float=2, Passthrough Float=1)
    UseOwnerScale=true
   )
   &InitialMeshOrientation(
    Mesh Axis To Orient=&MaskVectorBySpawnGroup(Masked Vector=<1 0 0>, Passthrough Vector=<0 0 0>)
    Mesh Orientation Mode=::NewEnumerator2
    Orientation Vector=&MaskVectorBySpawnGroup(Masked Vector=<1 0 0>, Passthrough Vector=<0 0 0>)
    Rotation=&MaskVectorBySpawnGroup(
     Masked Vector=&RandomRangeVector(Minimum=<0 0 -1>, Maximum=<0 0 1>)
     Passthrough Vector=&RandomRangeVector(Minimum=<-1 -1 -0.5>, Maximum=<1 1 0.5>)
    )
    Use Rotation Vector=true
   )
   &AddVelocity(
    Velocity=<0 0 1>
    Velocity Speed Scale=&MaskFloatBySpawnGroup(Masked Float=&RandomRangeFloat(Minimum=5, Maximum=5), Passthrough Float=1)
   )
   &CameraOffset(Camera Offset Amount=&MaskFloatBySpawnGroup(Masked Float=0, Passthrough Float=2))
  ]
  particleUpdateScript [
   &SolveForcesAndVelocity(Clamp Velocity=false, Limit Acceleration=false)
   &ParticleState()
   &ScaleMeshSize(Scale Factor=&MaskVectorBySpawnGroup(
     Masked Vector=<1 1 1>
     Passthrough Vector=&MakeVector(
      X=1
      Y=2
      Z=&FloatFromCurve(FloatCurve=DataInterface() {propertyValues {Curve {keys [
           {interpMode=RCIM_Cubic, leaveTangent=0.0966, tangentMode=RCTM_Break, time=0, value=1}
           {interpMode=RCIM_Cubic, leaveTangent=-0.1745, time=0.7665, value=1.07}
          ]}}})
     )
    )) {module=ScaleMeshSize001}
   &ScaleMeshSize(Scale Factor=&MaskVectorBySpawnGroup(
     Masked Vector=&Multiply_VectorByFloat(Vector=<1 1 1>, Float=&RandomRangeFloat(Minimum=0.45, Maximum=0.5))
     Passthrough Vector=&VectorFromFloat(Value=&RandomRangeFloat(Minimum=0.1, Maximum=0.15))
    ))
   &DynamicMaterialParameters(
    Index 0 Param 1=&MakeFloatFromVector(Vector=$Particles.Scale, Channel=::NewEnumerator0)
    Index 0 Param 2=&RandomRangeFloat(Minimum=0, Maximum=10)
    Index 0 Param 3=&FloatFromCurve(
     FloatCurve=DataInterface() {propertyValues {Curve {keys [{interpMode=RCIM_Cubic, leaveTangent=8.1085, tangentMode=RCTM_Break, time=0.8568, value=1}, {tangentMode=RCTM_User, time=1, value=0}]}}}
     Scale Curve=0.07
    )
    Index 0 Param 4=0.8
    Index 1 Param 1=&FloatFromCurve(FloatCurve=DataInterface() {propertyValues {Curve {keys [
         {interpMode=RCIM_Cubic, leaveTangent=1.1034, tangentMode=RCTM_Break, time=0, value=0}
         {arriveTangent=2.2399, tangentMode=RCTM_User, time=0.5, value=1}
        ]}}})
    Index 1 Param 2=&MaskFloatBySpawnGroup(
     Masked Float=&FloatFromCurve(FloatCurve=DataInterface() {propertyValues {Curve {keys [
          {interpMode=RCIM_Cubic, tangentMode=RCTM_User, time=0, value=0}
          {arriveTangent=2.7735, interpMode=RCIM_Cubic, leaveTangent=2.7735, tangentMode=RCTM_User, time=1, value=1}
         ]}}})
     Passthrough Float=&FloatFromCurve(FloatCurve=DataInterface() {propertyValues {Curve {keys [<at 0.5 0>, {arriveTangent=3.2053, tangentMode=RCTM_User, time=0.9, value=1}]}}})
    )
    Index 1 Param 3=&FloatFromCurve(FloatCurve=DataInterface() {propertyValues {Curve {keys [<at 0 0 RCIM_Cubic>, {arriveTangent=3.4223, interpMode=RCIM_Cubic, leaveTangent=3.4223, tangentMode=RCTM_User, time=1, value=1}]}}})
    Index 1 Param 4=&FloatFromCurve(
     FloatCurve=DataInterface() {propertyValues {Curve {keys [
         {interpMode=RCIM_Cubic, leaveTangent=0.1169, tangentMode=RCTM_Break, time=0, value=0}
         {arriveTangent=2.5708, interpMode=RCIM_Cubic, leaveTangent=2.5708, tangentMode=RCTM_User, time=1, value=1}
        ]}}}
     Scale Curve=&RandomRangeFloat(Minimum=1, Maximum=40)
    )
    Index 2 Param 1=&FloatFromCurve(FloatCurve=DataInterface() {propertyValues {Curve {keys [
         {interpMode=RCIM_Cubic, leaveTangent=-0.0025, tangentMode=RCTM_Break, time=0, value=0.5}
         {arriveTangent=-0.0316, interpMode=RCIM_Cubic, tangentMode=RCTM_Break, time=1, value=-0.2}
        ]}}})
    Index 2 Param 2=&FloatFromCurve(
     FloatCurve=DataInterface() {propertyValues {Curve {keys [
         {interpMode=RCIM_Cubic, tangentMode=RCTM_User, time=0, value=0}
         {arriveTangent=2.0522, interpMode=RCIM_Cubic, leaveTangent=2.0522, tangentMode=RCTM_User, time=1, value=1}
        ]}}}
     Scale Curve=2
    )
    Module.Index0Param0Write=true
    Module.Index0Param1Write=true
    Module.Index0Param2Write=true
    Module.Index0Param3Write=true
    Module.Index1Param0Write=true
    Module.Index1Param1Write=true
    Module.Index1Param2Write=true
    Module.Index1Param3Write=true
    Module.Index2Param0Write=true
    Module.Index2Param1Write=true
    Module.Index2Param2Write=false
    Module.Index2Param3Write=false
    Module.Index3Param0Write=false
    Module.Index3Param1Write=false
    Module.Index3Param2Write=false
    Module.Index3Param3Write=true
    Write Parameter Index 1=true
    Write Parameter Index 2=true
    Write Parameter Index 3=true
   )
   &UpdateMeshOrientation(
    Rotation Rate=&MaskFloatBySpawnGroup(Masked Float=&RandomRangeFloat(Minimum=-1, Maximum=1), Passthrough Float=0)
    Rotation Vector=<0 0 1>
   )
   &ScaleColor(
    Scale Alpha=&FloatFromCurve(FloatCurve=DataInterface() {propertyValues {Curve {keys [
         {tangentMode=RCTM_User, time=0, value=1}
         {arriveTangent=-0.0405, interpMode=RCIM_Cubic, leaveTangent=-0.0405, tangentMode=RCTM_User, time=0.7009, value=1}
         {arriveTangent=-8.0089, interpMode=RCIM_Cubic, leaveTangent=-8.0089, tangentMode=RCTM_User, time=1, value=0}
        ]}}})
    Scale RGB=&Lerp_Vector(
     A=&VectorFromFloat(Value=&FloatFromCurve(
       FloatCurve=DataInterface() {propertyValues {Curve {keys [<at 0 1>,<at 0.1 1>,<at 0.2 2>,<at 0.3 1>,<at 0.6 1>,<at 0.7 2>,<at 0.8 1>,<at 1 1>]}}}
       Scale Curve=&MaskFloatBySpawnGroup(Masked Float=10, Passthrough Float=5)
      ))
     B=&VectorFromFloat(Value=&RandomRangeFloat(Minimum=0.3, Maximum=5, Evaluation Type=::NewEnumerator1))
     Alpha=&FloatFromCurve(FloatCurve=DataInterface() {propertyValues {Curve {keys [
          {interpMode=RCIM_Cubic, leaveTangent=2.5294, tangentMode=RCTM_Break, time=0, value=0}
          {arriveTangent=-0.2166, interpMode=RCIM_Cubic, tangentMode=RCTM_Break, time=0.4, value=1}
         ]}}})
    )
    ScaleA=true
    ScaleRGB=true
   )
  ]
  rendererClasses [&NiagaraMeshRendererProperties]
  renderers [
   {rendererClass=&NiagaraMeshRendererProperties, rendererIndex=0}
   {rendererClass=&NiagaraMeshRendererProperties, rendererIndex=1}
   {rendererClass=&NiagaraMeshRendererProperties, rendererIndex=2}
   {rendererClass=&NiagaraMeshRendererProperties, rendererIndex=3}
  ]
 }
 Main_Lightning {
  bEnabled=false
  emitterUpdateScript [
   &EmitterState(
    Delay First Loop Only=true
    Life Cycle Mode=::NewEnumerator1
    Loop Behavior=::NewEnumerator2
    Loop Count=3
    Loop Delay=0.15
    Loop Duration=&FloatFromCurve(
     FloatCurve=DataInterface() {propertyValues {Curve {keys [
         <at 0 1>
         {arriveTangent=0.8333, interpMode=RCIM_Cubic, leaveTangent=0.8333, time=0.5, value=1}
         <at 0.6 1.5 RCIM_Constant>
         <at 1 1.5 RCIM_Cubic>
        ]}}}
     CurveIndex=$Engine.System.Age
     Scale Curve=0.9
    )
    Recalculate Duration Each Loop=true
    UseLoopDelay=true
   )
   /SetVariables(Emitter.ColorCoreScalar=10) {isSetParameters=true}
   &SpawnBurst_Instantaneous(Spawn Count=5, Use Spawn Probability=false, UseLoopCountLimit=false)
   &SpawnBurst_Instantaneous(Spawn Count=2, Spawn Time=0.08, Use Spawn Probability=false, UseLoopCountLimit=false) {module=SpawnBurst_Instantaneous001}
   &SpawnBurst_Instantaneous(Spawn Time=0.16, Use Spawn Probability=false, UseLoopCountLimit=false) {module=SpawnBurst_Instantaneous002}
   &SpawnBurst_Instantaneous(Spawn Count=2, Spawn Time=0.24, Use Spawn Probability=false, UseLoopCountLimit=false) {module=SpawnBurst_Instantaneous003}
  ]
  particleSpawnScript [
   &PartitionParticles(Number Of Partitions=10, Particles Per Partition=1, Partition Mode:ENiagara_PartitionMode_v2=::NewEnumerator1)
   &SampleSkeletalMesh(
    AdditionalYawPitchRoll=false
    Bone / Socket Index=1
    Bone Sampling Mode:ENiagaraBoneSamplingMode=::NewEnumerator4
    Skeletal Mesh:NiagaraDataInterfaceSkeletalMesh=$System.Skeletal Mesh
   )
   &ShapeLocation(
    Apply To Particle Position=::NewEnumerator1
    MaskSpawnGroup=false
    Override Seed=false
    Ring Radius=&Lerp_Float(A=1500, B=3000, Alpha=$User.ChargeProgress)
    Shape Origin=$Particles.SampleSkeletalMesh.SampledPosition
    Shape Primitive=::NewEnumerator4
    UseOwnerScale=true
   ) {module=ShapeLocation001}
   &SampleSkeletalMesh(
    AdditionalYawPitchRoll=false
    Bone / Socket Index:NiagaraInt32=&MakeIntFromFloat(Float=&FloatFromCurve(FloatCurve=DataInterface() {propertyValues {Curve {keys [<at 0 1>,<at 1 18>]}}}, CurveIndex=$User.ChargeProgress))
    Bone Sampling Mode:ENiagaraBoneSamplingMode=::NewEnumerator4
    Skeletal Mesh:NiagaraDataInterfaceSkeletalMesh=$System.Skeletal Mesh
   ) {module=SampleSkeletalMesh001}
   &ShapeLocation(
    Apply To Particle Position=::NewEnumerator1
    MaskSpawnGroup=false
    Override Seed=false
    Ring Radius=&Lerp_Float(A=1500, B=100, Alpha=$User.ChargeProgress)
    Shape Origin=$Particles.SampleSkeletalMesh001.SampledPosition
    Shape Primitive=::NewEnumerator4
    UseOwnerScale=true
   ) {module=ShapeLocation002}
   /SetVariables(Particles.GlobalRandom:NiagaraFloat=&RandomRangeFloat(Minimum=0.4, Maximum=1)) {isSetParameters=true}
   &InitializeParticle(
    Color=$System.PalpatineColor
    Lifetime=&Multiply_Float(
     A=&RandomRangeFloat(Minimum=0.4, Maximum=0.7)
     B=&FloatFromCurve(
      FloatCurve=DataInterface() {propertyValues {Curve {keys [<at 0 0.5>,<at 1.7 0.5>,<at 1.8 1>,<at 2 1>]}}}
      CurveIndex=$System.Age
      Scale Curve=2
     )
    )
    Mesh Scale Mode=::NewEnumerator3
    Mesh Uniform Scale=&MakeFloatFromVector(
     Vector=&MakeVector(X=1, Y=1, Z=&Multiply_Float(A=0.3, B=$Particles.GlobalRandom))
     Channel=::NewEnumerator0
    )
    Position Mode=::NewEnumerator0
    Sprite Size Mode=::NewEnumerator3
    Uniform Sprite Size=5
    Write Material Random=true
   )
   &InitialMeshOrientation(
    Mesh Axis To Orient=&MaskVector(Mask Source=1, Mask Parameter=$Particles.PartitionParticles.Partition, Mask=0, Masked Vector=<0 0 -1>, Passthrough Vector=<0 0 -1>)
    Mesh Orientation Mode=::NewEnumerator2
    Orientation Vector=<0 0 1>
    Rotation=&MaskVector(
     Mask Source=1
     Mask Parameter=$Particles.PartitionParticles.Partition
     Mask=0
     Masked Vector=<-0.0026 -0.0221 0>
     Passthrough Vector=<0.0051 -0.0008 0>
    )
    Rotation Coordinate Space=::NewEnumerator2
    Use Rotation Vector=true
   )
  ]
  particleUpdateScript [
   &ParticleState()
   &SolveForcesAndVelocity(Clamp Velocity=false, Limit Acceleration=false)
   &DynamicMaterialParameters(
    Index 0 Param 1=&FloatFromCurve(
     FloatCurve=DataInterface() {propertyValues {Curve {
        defaultValue=0
        keys [
         {arriveTangent=9.0725, interpMode=RCIM_Cubic, leaveTangent=9.0725, tangentMode=RCTM_User, time=0, value=0}
         {arriveTangent=-2.5249, interpMode=RCIM_Cubic, leaveTangent=-2.5249, tangentMode=RCTM_User, time=0.1408, value=0.9847}
         {arriveTangent=-0.0405, interpMode=RCIM_Cubic, leaveTangent=-0.0405, tangentMode=RCTM_User, time=1, value=0.2}
        ]
       }}}
     Scale Curve=&RandomRangeFloat(Minimum=1, Maximum=0.5)
    )
    Index 0 Param 2=&FloatFromCurve(
     FloatCurve=DataInterface() {propertyValues {Curve {
        defaultValue=0
        keys [<at 0 0.2 RCIM_Cubic>, {arriveTangent=2.4937, interpMode=RCIM_Cubic, leaveTangent=2.4937, tangentMode=RCTM_User, time=1, value=1}]
       }}}
     Scale Curve=&Multiply_Float(A=30, B=$Particles.GlobalRandom)
    )
    Index 0 Param 3=&RandomRangeFloat(Minimum=0.25, Maximum=1)
    Index 0 Param 4=&FloatFromCurve(FloatCurve=DataInterface() {propertyValues {Curve {
        defaultValue=0
        keys [
         {arriveTangent=4.1785, interpMode=RCIM_Cubic, leaveTangent=4.1785, tangentMode=RCTM_User, time=0, value=0}
         {arriveTangent=0.0224, interpMode=RCIM_Cubic, leaveTangent=0.0224, tangentMode=RCTM_User, time=0.5, value=1}
        ]
       }}})
    Index 1 Param 1=&FloatFromCurve(FloatCurve=DataInterface() {propertyValues {Curve {
        defaultValue=1
        keys [
         {interpMode=RCIM_Cubic, tangentMode=RCTM_User, time=0, value=0}
         {arriveTangent=2.9607, interpMode=RCIM_Cubic, leaveTangent=2.9607, tangentMode=RCTM_User, time=0.7, value=1}
        ]
       }}})
    Index 1 Param 2=&Multiply_Float(A=1, B=$Particles.GlobalRandom)
    Index 1 Param 3=&Multiply_Float(A=0.5, B=$Particles.GlobalRandom)
    Index 1 Param 4=&RandomRangeFloat(Minimum=0.5, Maximum=1)
    Module.Index0Float0Write=true
    Module.Index0Float1Write=true
    Module.Index0Float2Write=true
    Module.Index0Float3Write=true
    Module.Index1Float0Write=true
    Module.Index1Float1Write=true
    Module.Index1Float2Write=true
    Module.Index1Float3Write=true
    Write Parameter Index 1=::NewEnumerator1
   )
   &ScaleColor(
    Scale RGB=&Multiply_VectorByFloat(
     Vector=<1 1 1>
     Float=&Waveform(Global Phase=$Particles.Age, "Waveform [1]"=::NewEnumerator0, "[1] Frequency"=10, "[1] Amplitude Min/Max"=<2 0.3>)
    )
    ScaleA=true
    ScaleRGB=true
   )
   /Game/Effects/Niagara/Modules/Curves/FN_HermiteSpline()
  ]
  rendererClasses [&NiagaraMeshRendererProperties]
  renderers [{rendererClass=&NiagaraMeshRendererProperties, rendererIndex=0}]
 }
 MiddleGlow_Dark {
  emitterUpdateScript [
   &EmitterState()
   &SpawnBurst_Instantaneous(Spawn Count=5, Spawn Time=$User.TipDelay, Use Spawn Probability=false, UseLoopCountLimit=true)
  ]
  particleSpawnScript [
   &InitializeParticle(
    Color=<rgba 1.0302 0.2763 2 1>
    Lifetime=3
    Sprite Size=<8000 700>
    Sprite Size Mode=::NewEnumerator4
    UsePositionOffset=true
    Write Material Random=true
   )
   &CameraOffset(Camera Offset Amount=2000)
  ]
  particleUpdateScript [
   &ParticleState()
   /SetVariables(Particles.Size002:Vector2f=&Multiply_Vector2D_ByFloat(
     Float=&ClampFloat(Float=1, Max=2.5)
     Vector2D=&Multiply_Vector2D_ByFloat(
      Float=1
      Vector2D=&Multiply_Vector2D(
       A=<2218.9351 2218.9351>
       B=&Multiply_Vector2D_ByFloat(
        Float=&Waveform("Waveform [1]"=::NewEnumerator12, "[1] Frequency"=9, "[1] Amplitude Min/Max"=<0.9 1.1>)
        Vector2D=<4 0.5>
       )
      )
     )
    )) {enabled=false, isSetParameters=true}
   &ScaleSpriteSize(Uniform Curve Sprite Scale=DataInterface() {propertyValues {Curve {keys [
        {arriveTangent=0.7872, interpMode=RCIM_Cubic, leaveTangent=0.7872, tangentMode=RCTM_User, time=0.1073, value=0}
        {arriveTangent=-0.0405, interpMode=RCIM_Cubic, leaveTangent=-0.0405, tangentMode=RCTM_User, time=0.2073, value=1}
       ]}}})
   /SetVariables(Particles.color002:LinearColor=&MakeLinearColorFromVectorAndFloat(
     "Vector (RGB)"=&Multiply_VectorByFloat(
      Vector=&MakeVectorFromLinearColorRGB(Color=&LerpLinearColors(
        StartColor=<rgba 1 0.0639 0.3455 1>
        EndColor=<rgba 1 0.0734 0.3551 1>
        LerpFactor=&Waveform("Waveform [1]"=::NewEnumerator12, "[1] Frequency"=3, "[1] Amplitude Min/Max"=<0 1>)
       ))
      Float=3
     )
     "Float (Alpha)"=&Multiply_Float(A=0.25, B=&Waveform("Waveform [1]"=::NewEnumerator12, "[1] Frequency"=3, "[1] Amplitude Min/Max"=<0.2 1>))
    )) {isSetParameters=true}
   /SetVariables(Particles.color002:LinearColor=&MultiplyLinearColor(
     Linear Color=$Particles.color002
     Scale Factor=&MakeVector4(
      X=1
      Y=1
      Z=1
      W=&FloatFromCurve(FloatCurve=DataInterface() {propertyValues {Curve {keys [<at 0 1>,<at 0.4 1>,<at 0.7 0>]}}})
     )
    )) {isSetParameters=true}
   &SkeletalMeshLocation(Bone Sampling Mode=::NewEnumerator4, MaskSpawnGroup=false, Skeletal Mesh=$System.Skeletal Mesh)
  ]
  rendererClasses [&NiagaraSpriteRendererProperties]
  renderers [{rendererClass=&NiagaraSpriteRendererProperties}]
 }
 ParticleDecal001 {
  bEnabled=false
  emitterUpdateScript [
   &EmitterState(Life Cycle Mode=::NewEnumerator1, Loop Behavior=::NewEnumerator1, Loop Duration Mode=::NewEnumerator1, UseLoopDelay=false)
   &SpawnBurst_Instantaneous(Use Spawn Probability=false, UseLoopCountLimit=false)
  ]
  particleSpawnScript [
   &InitializeParticle(
    Color=<rgba 0.064 0.1693 1 1>
    Lifetime=4.5
    Mesh Scale Mode=::NewEnumerator3
    Mesh Uniform Scale=2
    Position Offset=<0 0 100>
    Sprite Size Mode=::NewEnumerator3
    Uniform Sprite Size=1500
    UsePositionOffset=true
    Write Material Random=true
   )
   &CameraOffset(Camera Offset Amount=10000)
   &DynamicMaterialParameters(
    Index 0 Param 1=0.1
    Index 0 Param 3=1
    Module.Index0Param0Write=true
    Module.Index0Param1Write=true
    Module.Index0Param2Write=true
    Module.Index0Param3Write=true
   )
  ]
  particleUpdateScript [
   &ParticleState()
   &ScaleSpriteSize(
    Uniform Curve Scale=$User.DecalEnabled
    Uniform Curve Sprite Scale=DataInterface() {propertyValues {Curve {keys [
        {arriveTangent=0.6846, interpMode=RCIM_Cubic, leaveTangent=0.6846, tangentMode=RCTM_User, time=0.3, value=1.25}
        {arriveTangent=-1.7857, interpMode=RCIM_Cubic, leaveTangent=-1.7857, time=0.8, value=1}
        <at 1 0 RCIM_Cubic>
       ]}}}
   )
   &SkeletalMeshLocation(
    Bone / Socket Index=&MakeIntFromFloat(Float=&FloatFromCurve(FloatCurve=DataInterface() {propertyValues {Curve {keys [<at 0 1>,<at 0.3 19>]}}}))
    Bone Sampling Mode=::NewEnumerator4
    MaskSpawnGroup=false
    Position Sampling=::NewEnumerator2
    Skeletal Mesh=$System.Skeletal Mesh
   )
   /SetVariables(Particles.Position=/Niagara/DynamicInputs/Lerp/SmoothLerpOverTime_Position(
     Convergence Rate:NiagaraFloat=&FloatFromCurve(FloatCurve=DataInterface() {propertyValues {Curve {keys [<at 0.3 1>,<at 0.4 3>]}}}, Scale Curve=4)
     Smooth Value:NiagaraPosition=$Particles.Position
     Target Value:NiagaraPosition=$Particles.SkeletalMeshLocation.SampledPosition
    )) {isSetParameters=true}
   &ScaleColor(
    Scale Alpha=&FloatFromCurve(FloatCurve=DataInterface() {propertyValues {Curve {
        defaultValue=1
        keys [
         {interpMode=RCIM_Cubic, tangentMode=RCTM_User, time=0, value=0}
         {arriveTangent=1.25, interpMode=RCIM_Cubic, leaveTangent=1.25, time=0.05, value=1}
         {interpMode=RCIM_Cubic, tangentMode=RCTM_User, time=0.8, value=1}
         {tangentMode=RCTM_User, time=1, value=0}
        ]
       }}})
    ScaleA=true
    ScaleRGB=true
   )
  ]
  rendererClasses [&NiagaraSpriteRendererProperties]
  renderers [{rendererClass=&NiagaraSpriteRendererProperties}]
 }
 Trail_Leg {
  bEnabled=false
  emitterUpdateScript [
   &EmitterState(Life Cycle Mode=::NewEnumerator1, Loop Behavior=::NewEnumerator1, Loop Duration Mode=::NewEnumerator1, UseLoopDelay=false)
   &SpawnBurst_Instantaneous(Spawn Count=40, Use Spawn Probability=false, UseLoopCountLimit=false)
  ]
  particleSpawnScript [
   &FN_SetSpawnGroup(Inclusivity=true, Normalize Spawn Group=true, Particles Per Group=18, Pattern=::NewEnumerator0, Set Ribbon ID and Ribbon Link Order=true)
   &InitializeParticle(Lifetime=5, Ribbon Width=1.5, Ribbon Width Mode=::NewEnumerator1, UsePositionOffset=false, Write Material Random=true)
  ]
  particleUpdateScript [
   &ParticleState(Kill Particles When Lifetime Has Elapsed=false, Let Infinitely Lived Particles Die When Emitter Deactivates=true)
   &FN_BoneFlow(
    Angle:NiagaraFloat=&FloatFromCurve(
     FloatCurve=DataInterface() {propertyValues {Curve {keys [
         {arriveTangent=1.8289, interpMode=RCIM_Cubic, leaveTangent=1.8289, tangentMode=RCTM_User, time=0.1, value=0.125}
         {arriveTangent=2.2917, interpMode=RCIM_Cubic, leaveTangent=2.2917, time=0.3, value=0.875}
         {arriveTangent=1.0714, interpMode=RCIM_Cubic, leaveTangent=1.0714, time=0.7, value=1.5}
         {arriveTangent=1.2828, interpMode=RCIM_Cubic, leaveTangent=1.2828, tangentMode=RCTM_User, time=1, value=1.625}
        ]}}}
     CurveIndex=$Particles.Normalized Spawn Group
     Scale Curve=0.8
    )
    Height:NiagaraFloat=&Multiply_Float(
     A=$Particles.Normalized Spawn Group
     B=&SelectFloatFromArray(Direct Array Index=$Particles.SpawnGroup, Float Selection Array=DataInterface() {propertyValues {FloatData [1,1]}})
    )
    Radius:NiagaraFloat=&FloatFromCurve(
     FloatCurve=DataInterface() {propertyValues {Curve {keys [
         {arriveTangent=-0.6799, interpMode=RCIM_Cubic, leaveTangent=-0.6799, tangentMode=RCTM_User, time=0, value=7}
         {arriveTangent=-4.2857, interpMode=RCIM_Cubic, leaveTangent=-4.2857, time=0.3571, value=5.9951}
         {arriveTangent=-8.7317, interpMode=RCIM_Cubic, leaveTangent=-8.7317, tangentMode=RCTM_User, time=0.7, value=4}
         {arriveTangent=-3.5832, interpMode=RCIM_Cubic, leaveTangent=-3.5832, tangentMode=RCTM_User, time=1, value=1}
        ]}}}
     CurveIndex=$Particles.Normalized Spawn Group
     Scale Curve=150
    )
    SideVector1:Vector3f=<1 1 1>
    SideVector2:Vector3f=<1 1 1>
    SideVector3:Vector3f=<1 1 1>
    SkeletalMesh:NiagaraDataInterfaceSkeletalMesh=$System.Skeletal Mesh
    Stabalize Twisting=true
   )
   &ScaleRibbonWidth(Ribbon Width Scale=&FloatFromCurve(
     FloatCurve=DataInterface() {propertyValues {Curve {keys [
         <at 0 0.5>
         {arriveTangent=1.8056, interpMode=RCIM_Cubic, leaveTangent=1.8056, tangentMode=RCTM_User, time=0.2, value=0.9}
         {arriveTangent=0.6908, tangentMode=RCTM_User, time=1, value=2}
        ]}}}
     CurveIndex=$Particles.RibbonLinkOrder
     Scale Curve=500
    ))
   /SetVariables(Particles.Ribbon B Width:NiagaraFloat=&Multiply_Float(A=$Particles.RibbonWidth, B=4)) {isSetParameters=true}
   &ScaleColor(
    Scale Alpha=&FloatFromCurve(
     FloatCurve=DataInterface() {propertyValues {Curve {keys [
         {arriveTangent=1.4743, interpMode=RCIM_Cubic, leaveTangent=1.4743, tangentMode=RCTM_User, time=0, value=0}
         {arriveTangent=4.159, interpMode=RCIM_Cubic, leaveTangent=-3.8038, tangentMode=RCTM_Break, time=0.3, value=1}
         {arriveTangent=-0.2647, interpMode=RCIM_Cubic, leaveTangent=-0.2647, tangentMode=RCTM_User, time=0.9, value=0}
        ]}}}
     CurveIndex=$Particles.RibbonLinkOrder
     Scale Curve=0.5
    )
    Scale RGB=&MakeVectorFromLinearColorRGB(Color=&ColorFromCurve(
      ColorCurve=DataInterface() {propertyValues {
        AlphaCurve {keys []}
        BlueCurve {keys [<at 0.1537 1 RCIM_Cubic>]}
        GreenCurve {keys [<at 0.1537 0.3088 RCIM_Cubic>]}
        RedCurve {keys [<at 0.1537 0.6162 RCIM_Cubic>]}
       }}
      CurveIndex=$StackContext.RibbonLinkOrder
     ))
    ScaleA=true
    ScaleRGB=true
   )
   &ScaleColor(Color Value To Scale=$Particles.Color, ScaleA=true, ScaleRGB=true) {module=ScaleColor001}
   &DynamicMaterialParameters(
    Index 0 Param 3=1
    Index 0 Param 4=1
    Module.Param0WriteEnabled=true
    Module.Param1WriteEnabled=true
    Module.Param2WriteEnabled=true
    Module.Param3WriteEnabled=true
   )
  ]
  rendererClasses [&NiagaraRibbonRendererProperties]
  renderers [
   {rendererClass=&NiagaraRibbonRendererProperties}
   {rendererClass=&NiagaraRibbonRendererProperties, rendererIndex=1}
  ]
 }
 WindCone {rendererClasses [], renderers []}
 WindCone001 {rendererClasses [], renderers []}
 par {
  simTarget=GPUComputeSim
  emitterUpdateScript [
   &EmitterState(Life Cycle Mode=::NewEnumerator1, Loop Behavior=::NewEnumerator1, Loop Delay=0.1, Loop Duration=3, UseLoopDelay=true)
   &SpawnBurst_Instantaneous(Spawn Count=50, Spawn Time=1.3, Use Spawn Probability=false)
  ]
  particleSpawnScript [
   &InitializeParticle(
    Color=<rgba 0.0919 0.0821 1 1>
    Lifetime Max=0.8
    Lifetime Min=0.5
    Lifetime Mode=::NewEnumerator1
    Sprite Size Max=<4 7>
    Sprite Size Min=<1.5 2>
    Sprite Size Mode=::NewEnumerator2
    UsePositionOffset=false
    Write Material Random=true
   )
   &SkeletalMeshLocation(Bone Sampling Mode=::NewEnumerator4, MaskSpawnGroup=false, Position Sampling=::NewEnumerator2, Skeletal Mesh=$System.Skeletal Mesh)
   &SphereLocation(
    MaskSpawnGroup=false
    Override Seed=false
    Sphere Origin:NiagaraPosition=$Particles.SkeletalMeshLocation.SampledPosition
    Sphere Radius=60
    Surface Expansion Mode=::NewEnumerator2
    Surface Only=true
    Surface Only Band Thickness=1
    UseNonUniformScale=false
    UseOffset=false
    UseSphereOrientation=false
   )
   &AddVelocityFromPoint(
    Module.Use Falloff Distance=false
    Velocity Origin:NiagaraPosition=$Particles.SkeletalMeshLocation.SampledPosition
    Velocity Strength=&Multiply_Float(A=&RandomRangeFloat(Minimum=7000, Maximum=5000), B=2)
   )
   &CameraOffset(Camera Offset Amount=10)
  ]
  particleUpdateScript [
   &ParticleState()
   &CurlNoiseForce(
    Noise Frequency=200
    Noise Strength=&Multiply_Float(
     A=1500
     B=&FloatFromCurve(FloatCurve=DataInterface() {propertyValues {Curve {keys [<at 0 0>,<at 0.2 1>,<at 0.3 0>]}}})
    )
    Pan Noise=false
   )
   &Drag(Drag=&RandomRangeFloat(Minimum=1, Maximum=3), Use Linear Drag=true, Use Rotational Drag=false)
   &SkeletalMeshLocation(Bone Sampling Mode=::NewEnumerator4, MaskSpawnGroup=false, Position Sampling=::NewEnumerator2, Skeletal Mesh=$System.Skeletal Mesh) {module=SkeletalMeshLocation001}
   &PointForce(
    Force Falloff Distance:NiagaraFloat=&RandomRangeFloat(Minimum=50000, Maximum=50000)
    Force Origin:NiagaraPosition=$Particles.SkeletalMeshLocation001.SampledPosition
    Force Strength=&Multiply_Float(
     A=&RandomRangeFloat(Minimum=-5000, Maximum=-7000)
     B=&FloatFromCurve(FloatCurve=DataInterface() {propertyValues {Curve {keys [<at 0 1>,<at 1 20>]}}})
    )
    Module.Use Falloff=true
    Module.UseLinearForce=false
   )
   &SolveForcesAndVelocity(Clamp Velocity=false, Limit Acceleration=false)
   &KillParticlesInVolume(Sphere Radius=15)
   &ScaleSpriteSizeBySpeed(Max Scale Factor=<0.2 1>, Min Scale Factor=<1 0.5>)
   &ScaleSpriteSize(Scale Factor=&MakeVector2D(
     X=&FloatFromCurve(FloatCurve=DataInterface() {propertyValues {Curve {keys [<at 0 0>,<at 0.1 1>,<at 1 0>]}}}, Scale Curve=50)
     Y=&FloatFromCurve(FloatCurve=DataInterface() {propertyValues {Curve {keys [<at 0 1>,<at 1 0>]}}}, Scale Curve=100)
    ))
   &ScaleColor(
    Linear Color Curve=DataInterface() {propertyValues {
      AlphaCurve {keys [<at 0 1>]}
      BlueCurve {keys [<at 0 1>,<at 0.1583 1>,<at 0.3467 0.3723>,<at 0.5211 1>,<at 0.8598 1>]}
      GreenCurve {keys [<at 0 0.1833>,<at 0.1583 0>,<at 0.3467 0.0575>,<at 0.5211 0.0833>,<at 0.8598 0.6>]}
      RedCurve {keys [<at 0 0>,<at 0.1583 0.2833>,<at 0.3467 1>,<at 0.5211 0>,<at 0.8598 0>]}
     }}
    Scale Mode=::NewEnumerator2
   ) {enabled=false}
   &ScaleColor(
    Scale Alpha=&FNTimeOfDayScalar(Night Scale=0.5, Morning Scale=0.5, Evening Scale=0.5)
    Scale RGB=&VectorFromFloat(Value=&FloatFromCurve(FloatCurve=DataInterface() {propertyValues {Curve {keys [<at 0 0.5>,<at 0.4 1>,<at 1 0.5>]}}}, Scale Curve=20))
    ScaleA=true
    ScaleRGB=true
   ) {module=ScaleColor001}
  ]
  rendererClasses [&NiagaraSpriteRendererProperties]
  renderers [{rendererClass=&NiagaraSpriteRendererProperties}]
 }
}
summary {
 systemName=NS_TentacleAttackChargeUpLaser
 emitters [
  {emitterName=Arcs, rendererClasses [&NiagaraSpriteRendererProperties]}
  {emitterName=Flare_merged, rendererClasses [&NiagaraSpriteRendererProperties]}
  {bEnabled=false, emitterName=ParticleDecal001, rendererClasses [&NiagaraSpriteRendererProperties]}
  {bEnabled=false, emitterName=Arcs001, rendererClasses [&NiagaraSpriteRendererProperties]}
  {bEnabled=false, emitterName=ChargeUp, rendererClasses [&NiagaraComponentRendererProperties]}
  {emitterName=Energy_Offset, rendererClasses [&NiagaraMeshRendererProperties]}
  {emitterName=Energy_Offset001, rendererClasses [&NiagaraMeshRendererProperties]}
  {emitterName=par, simTarget=GPUComputeSim, rendererClasses [&NiagaraSpriteRendererProperties]}
  {emitterName=Flare, rendererClasses [&NiagaraSpriteRendererProperties]}
  {emitterName=WindCone, rendererClasses []}
  {emitterName=WindCone001, rendererClasses []}
  {bEnabled=false, emitterName=Blue_Outer_Plasma001, rendererClasses [&NiagaraRibbonRendererProperties]}
  {bEnabled=false, emitterName=Blue_Outer_Plasma, rendererClasses [&NiagaraRibbonRendererProperties]}
  {bEnabled=false, emitterName=Bright_Ribbon_01, rendererClasses [&NiagaraRibbonRendererProperties]}
  {bEnabled=false, emitterName=Bright_Ribbon_01001, rendererClasses [&NiagaraRibbonRendererProperties]}
  {bEnabled=false, emitterName=Main_Lightning, rendererClasses [&NiagaraMeshRendererProperties]}
  {bEnabled=false, emitterName=Lightning001, rendererClasses [&NiagaraMeshRendererProperties]}
  {bEnabled=false, emitterName=Trail_Leg, rendererClasses [&NiagaraRibbonRendererProperties]}
  {emitterName=MiddleGlow_Dark, rendererClasses [&NiagaraSpriteRendererProperties]}
 ]
 userVariables [
  User.BeamEnd_0 {type=NiagaraPosition, description="", defaultValue {}}
  User.Charge {type=NiagaraFloat, defaultValue=3, description=""}
  User.ChargeProgress {type=NiagaraFloat, defaultValue=1, description=""}
  User.ChargeProgressLightning {type=NiagaraFloat, defaultValue=0, description=""}
  User.ChargeScale {type=NiagaraFloat, defaultValue=0, description=""}
  User.ChargeScaleColor {type=LinearColor, description="", defaultValue {struct=LinearColor, value=<rgba 1 1 1 1>}}
  User.ChargeScaleVec {type=Vector3f, description="", defaultValue {struct=Vector3f, value=<0 0 0>}}
  User.Color {type=LinearColor, description="", defaultValue {struct=LinearColor, value=<rgba 0 0.161 0.5 1>}}
  User.ColorMult {type=NiagaraFloat, defaultValue=1, description=""}
  User.DecalEnabled {type=NiagaraFloat, defaultValue=1, description=""}
  User.ElementSizeScale {type=NiagaraFloat, defaultValue=1, description=""}
  User.Fade {type=NiagaraFloat, defaultValue=1, description=""}
  User.FX_Sphere_Large_Size {type=NiagaraFloat, defaultValue=250, description=""}
  User.HitLocation {type=Vector3f, description="", defaultValue {struct=Vector3f, value=<-50 50 200>}}
  User.LightningEnabled {type=NiagaraFloat, defaultValue=1, description=""}
  User.LightningLoopAge {type=NiagaraFloat, defaultValue=1.2, description=""}
  User.PlasmaEnabled {type=NiagaraFloat, defaultValue=1, description=""}
  User.PostNoiseStrength {type=NiagaraFloat, defaultValue=1, description=""}
  User.PostProcess {type=NiagaraFloat, defaultValue=0, description=""}
  User.PostSizeScale {type=NiagaraFloat, defaultValue=1, description=""}
  User.PostStrength {type=NiagaraFloat, defaultValue=1, description=""}
  User.Scaling_SpawnRadius {type=NiagaraFloat, defaultValue=1, description=""}
  User.StartLocation {type=Vector3f, description="", defaultValue {struct=Vector3f, value=<0 0 0>}}
  User.TipDelay {type=NiagaraFloat, defaultValue=0, description=""}
  User.WidthMult {type=NiagaraFloat, defaultValue=1, description=""}
 ]
}
