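# Niagara charge-up effect for the long laser tentacle attack, attached to a skeletal mesh.
# Arc sprites travel along the skeleton by sweeping the sampled bone index over the loop while shrinking, as a second arc emitter builds up at bone 18.
# Near the one-second mark, burst flare and decal-named sprites appear on the mesh, pushed toward the camera.
# A post-process emitter with distance-faded component rendering is present but disabled.
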NiagaraSystem /PistolCanary/Assets/FX/TentacleAttack/NS_TentacleAttackChargeUpLaserLong
dependencies {
 usedDataInterfaces [
  &NiagaraDataInterfaceCurve
  &NiagaraDataInterfaceColorCurve
  &NiagaraDataInterfaceArrayFloat
  &NiagaraDataInterfaceArrayFloat2
  &NiagaraDataInterfaceVector2DCurve
  &NiagaraDataInterfaceStaticMesh
  &NiagaraDataInterfaceSkeletalMesh
 ]
 usedDynamicInputs []
 usedModules [
  &SystemState
  &self:SystemSpawnScript.SystemScriptSource.SystemScriptGraph.NiagaraNodeAssignment_0.SetVariables_#1
  &self:SystemSpawnScript.SystemScriptSource.SystemScriptGraph.NiagaraNodeAssignment_1.SetVariables_#2
 ]
}
emitters {
 Arcs {
  emitterUpdateScript [
   &EmitterState(Life Cycle Mode=::NewEnumerator1, Loop Duration=0.8, UseLoopDelay=false)
   &SpawnRate(
    SpawnRate=&FloatFromCurve(FloatCurve=DataInterface() {propertyValues {Curve {keys [<at 0 1>,<at 1 0.4>]}}}, CurveIndex=$Emitter.NormalizedLoopAge, Scale Curve=40)
    Use Spawn Probability=false
   )
  ]
  particleSpawnScript [
   &InitializeParticle(
    Lifetime Max=0.4
    Lifetime Min=0.6
    Lifetime Mode=::NewEnumerator1
    Sprite Rotation Mode=::NewEnumerator2
    Sprite Size Mode=::NewEnumerator3
    Uniform Sprite Size=&Multiply_Float(
     A=&FloatFromCurve(FloatCurve=DataInterface() {propertyValues {Curve {keys [<at 0 1.5>,<at 0.7 0.3>]}}}, CurveIndex=$Emitter.NormalizedLoopAge)
     B=&RandomRangeFloat(Minimum=7000, Maximum=10000)
    )
    UsePositionOffset=false
    Write Material Random=true
   )
   &SubUVAnimation(End Frame=3, SubUV Animation Mode:ENiagara_SubUVLookupMode=::NewEnumerator2)
   &SphereLocation(
    MaskSpawnGroup=false
    Non Uniform Scale=<1 1 0.25>
    Override Seed=false
    Sphere Radius=800
    Surface Only=false
    UseNonUniformScale=true
    UseOffset=false
    UseSphereOrientation=false
   )
   &SkeletalMeshLocation(
    Bone / Socket Index=&MakeIntFromFloat(Float=&FloatFromCurve(FloatCurve=DataInterface() {propertyValues {Curve {keys [<at 0 1>,<at 1 19>]}}}, CurveIndex=$Emitter.NormalizedLoopAge))
    Bone Sampling Mode=::NewEnumerator6
    MaskSpawnGroup=false
    Skeletal Mesh=$System.Skeletal Mesh
   )
  ]
  particleUpdateScript [
   &ParticleState()
   &CameraOffset(Camera Offset Amount=25)
   &ScaleColor(
    Scale RGB=&MakeVectorFromLinearColorRGB(Color=&LerpLinearColors(
      StartColor=&ColorFromCurve(ColorCurve=DataInterface() {propertyValues {
         AlphaCurve {keys [<at 0.5586 1 RCIM_Cubic>,<at 1 0 RCIM_Cubic>]}
         BlueCurve {keys [<at 0 1 RCIM_Cubic>,<at 1 1 RCIM_Cubic>]}
         GreenCurve {keys [
           {arriveTangent=-0.0536, interpMode=RCIM_Cubic, leaveTangent=-0.0536, time=0, value=0.2093}
           {arriveTangent=-0.0536, interpMode=RCIM_Cubic, leaveTangent=-0.0536, time=1, value=0.1557}
          ]}
         RedCurve {keys [
           {arriveTangent=0.0676, interpMode=RCIM_Cubic, leaveTangent=0.0676, time=0, value=0.0396}
           {arriveTangent=0.0676, interpMode=RCIM_Cubic, leaveTangent=0.0676, time=1, value=0.1073}
          ]}
        }})
      EndColor=&ColorFromCurve(ColorCurve=DataInterface() {propertyValues {
         AlphaCurve {keys [<at 0.5586 1 RCIM_Cubic>,<at 1.0091 0 RCIM_Cubic>]}
         BlueCurve {keys [<at 0 1.2408 RCIM_Cubic>,<at 0.982 0>]}
         GreenCurve {keys [<at 0 1.7901 RCIM_Cubic>,<at 0.982 0.2738>]}
         RedCurve {keys [<at 0 5 RCIM_Cubic>,<at 0.982 5>]}
        }})
      LerpFactor=&SelectFloatFromArray(Direct Array Index=0, Float Selection Array=DataInterface() {propertyValues {FloatData [0,0,0,1]}})
     ))
    ScaleA=false
    ScaleRGB=true
   )
   &ScaleColor(
    Scale RGB=&VectorFromFloat(Value=&FloatFromCurve(FloatCurve=DataInterface() {propertyValues {Curve {keys [
          {interpMode=RCIM_Cubic, tangentMode=RCTM_User, time=0, value=1}
          {arriveTangent=106.638, interpMode=RCIM_Cubic, leaveTangent=-47.9454, tangentMode=RCTM_Break, time=0.2, value=10}
          {interpMode=RCIM_Cubic, tangentMode=RCTM_User, time=0.8, value=1}
         ]}}}))
    ScaleA=false
    ScaleRGB=true
   ) {module=ScaleColor001}
   &ScaleSpriteSize(Scale Factor=&Vector2DFromFloat(Value=&FloatFromCurve(
      FloatCurve=DataInterface() {propertyValues {Curve {keys [
          {arriveTangent=-0.0604, interpMode=RCIM_Cubic, leaveTangent=-0.0604, tangentMode=RCTM_User, time=0, value=0.6}
          {arriveTangent=1.6386, interpMode=RCIM_Cubic, leaveTangent=1.6386, tangentMode=RCTM_User, time=1, value=1}
         ]}}}
      Scale Curve=1.5
     )))
   &SpriteRotationRate(Rotation Rate=25)
   &ScaleSpriteSize(
    Scale Factor=&MakeVector2D(
     X=&FloatFromCurve(FloatCurve=DataInterface() {propertyValues {Curve {keys [
          {arriveTangent=-3.4808, interpMode=RCIM_Cubic, leaveTangent=-3.4808, tangentMode=RCTM_User, time=0.4, value=1}
          {arriveTangent=0.0049, interpMode=RCIM_Cubic, leaveTangent=0.0049, tangentMode=RCTM_User, time=0.8, value=0.5}
         ]}}})
     Y=&FloatFromCurve(FloatCurve=DataInterface() {propertyValues {Curve {keys [
          {arriveTangent=0.0124, interpMode=RCIM_Cubic, leaveTangent=0.0124, tangentMode=RCTM_User, time=0.4, value=0.85}
          {arriveTangent=2.3888, interpMode=RCIM_Cubic, leaveTangent=2.3888, tangentMode=RCTM_User, time=0.8, value=1.2}
         ]}}})
    )
    Scale Sprite Size Mode=::NewEnumerator2
   ) {module=ScaleSpriteSize001}
   &DynamicMaterialParameters(
    Index 0 Param 1=1
    Index 0 Param 2=1
    Index 0 Param 3=&FloatFromCurve(FloatCurve=DataInterface() {propertyValues {Curve {keys [
         {arriveTangent=0.1411, interpMode=RCIM_Cubic, leaveTangent=0.1411, tangentMode=RCTM_User, time=0.2, value=0}
         {arriveTangent=1.8523, interpMode=RCIM_Cubic, leaveTangent=1.8523, tangentMode=RCTM_User, time=1, value=0.5}
        ]}}})
    Module.Param0WriteEnabled=true
    Module.Param1WriteEnabled=true
    Module.Param2WriteEnabled=true
    Module.Param3WriteEnabled=false
   )
  ]
  rendererClasses [&NiagaraSpriteRendererProperties]
  renderers [{rendererClass=&NiagaraSpriteRendererProperties}]
 }
 Arcs001 {
  emitterUpdateScript [
   &EmitterState(Life Cycle Mode=::NewEnumerator1, Loop Duration=1, UseLoopDelay=false)
   &SpawnRate(
    SpawnRate=&FloatFromCurve(
     FloatCurve=DataInterface() {propertyValues {Curve {keys [<at 0.5 0>,<at 0.65 1>,<at 1 0.4>]}}}
     CurveIndex=$Emitter.NormalizedLoopAge
     Scale Curve=20
    )
    Use Spawn Probability=false
   )
  ]
  particleSpawnScript [
   &InitializeParticle(
    Lifetime Max=0.4
    Lifetime Min=0.6
    Lifetime Mode=::NewEnumerator1
    Sprite Rotation Mode=::NewEnumerator2
    Sprite Size Mode=::NewEnumerator3
    Uniform Sprite Size=&Multiply_Float(
     A=&FloatFromCurve(FloatCurve=DataInterface() {propertyValues {Curve {keys [<at 0 0.8>,<at 0.7 1.2>]}}}, CurveIndex=$Emitter.NormalizedLoopAge)
     B=&RandomRangeFloat(Minimum=2000, Maximum=2500)
    )
    UsePositionOffset=false
    Write Material Random=true
   )
   &SubUVAnimation(End Frame=3, SubUV Animation Mode:ENiagara_SubUVLookupMode=::NewEnumerator2)
   &SphereLocation(
    MaskSpawnGroup=false
    Non Uniform Scale=<1 1 0.25>
    Override Seed=false
    Sphere Radius=800
    Surface Only=false
    UseNonUniformScale=true
    UseOffset=false
    UseSphereOrientation=false
   )
   &SkeletalMeshLocation(Bone / Socket Index=18, Bone Sampling Mode=::NewEnumerator6, MaskSpawnGroup=false, Skeletal Mesh=$System.Skeletal Mesh)
  ]
  particleUpdateScript [
   &ParticleState()
   &CameraOffset(Camera Offset Amount=25)
   &ScaleColor(
    Scale RGB=&MakeVectorFromLinearColorRGB(Color=&LerpLinearColors(
      StartColor=&ColorFromCurve(ColorCurve=DataInterface() {propertyValues {
         AlphaCurve {keys [<at 0.5586 1 RCIM_Cubic>,<at 1 0 RCIM_Cubic>]}
         BlueCurve {keys [<at 0 1 RCIM_Cubic>,<at 1 1 RCIM_Cubic>]}
         GreenCurve {keys [
           {arriveTangent=-0.0536, interpMode=RCIM_Cubic, leaveTangent=-0.0536, time=0, value=0.2093}
           {arriveTangent=-0.0536, interpMode=RCIM_Cubic, leaveTangent=-0.0536, time=1, value=0.1557}
          ]}
         RedCurve {keys [
           {arriveTangent=0.0676, interpMode=RCIM_Cubic, leaveTangent=0.0676, time=0, value=0.0396}
           {arriveTangent=0.0676, interpMode=RCIM_Cubic, leaveTangent=0.0676, time=1, value=0.1073}
          ]}
        }})
      EndColor=&ColorFromCurve(ColorCurve=DataInterface() {propertyValues {
         AlphaCurve {keys [<at 0.5586 1 RCIM_Cubic>,<at 1.0091 0 RCIM_Cubic>]}
         BlueCurve {keys [<at 0 1.2408 RCIM_Cubic>,<at 0.982 0>]}
         GreenCurve {keys [<at 0 1.7901 RCIM_Cubic>,<at 0.982 0.2738>]}
         RedCurve {keys [<at 0 5 RCIM_Cubic>,<at 0.982 5>]}
        }})
      LerpFactor=&SelectFloatFromArray(Direct Array Index=0, Float Selection Array=DataInterface() {propertyValues {FloatData [0,0,0,1]}})
     ))
    ScaleA=false
    ScaleRGB=true
   )
   &ScaleColor(
    Scale RGB=&VectorFromFloat(Value=&FloatFromCurve(FloatCurve=DataInterface() {propertyValues {Curve {keys [
          {interpMode=RCIM_Cubic, tangentMode=RCTM_User, time=0, value=1}
          {arriveTangent=106.638, interpMode=RCIM_Cubic, leaveTangent=-47.9454, tangentMode=RCTM_Break, time=0.2, value=10}
          {interpMode=RCIM_Cubic, tangentMode=RCTM_User, time=0.8, value=1}
         ]}}}))
    ScaleA=false
    ScaleRGB=true
   ) {module=ScaleColor001}
   &ScaleSpriteSize(Scale Factor=&Vector2DFromFloat(Value=&FloatFromCurve(
      FloatCurve=DataInterface() {propertyValues {Curve {keys [
          {arriveTangent=-0.0604, interpMode=RCIM_Cubic, leaveTangent=-0.0604, tangentMode=RCTM_User, time=0, value=0.6}
          {arriveTangent=1.6386, interpMode=RCIM_Cubic, leaveTangent=1.6386, tangentMode=RCTM_User, time=1, value=1}
         ]}}}
      Scale Curve=1.5
     )))
   &SpriteRotationRate(Rotation Rate=25)
   &ScaleSpriteSize(
    Scale Factor=&MakeVector2D(
     X=&FloatFromCurve(FloatCurve=DataInterface() {propertyValues {Curve {keys [
          {arriveTangent=-3.4808, interpMode=RCIM_Cubic, leaveTangent=-3.4808, tangentMode=RCTM_User, time=0.4, value=1}
          {arriveTangent=0.0049, interpMode=RCIM_Cubic, leaveTangent=0.0049, tangentMode=RCTM_User, time=0.8, value=0.5}
         ]}}})
     Y=&FloatFromCurve(FloatCurve=DataInterface() {propertyValues {Curve {keys [
          {arriveTangent=0.0124, interpMode=RCIM_Cubic, leaveTangent=0.0124, tangentMode=RCTM_User, time=0.4, value=0.85}
          {arriveTangent=2.3888, interpMode=RCIM_Cubic, leaveTangent=2.3888, tangentMode=RCTM_User, time=0.8, value=1.2}
         ]}}})
    )
    Scale Sprite Size Mode=::NewEnumerator2
   ) {module=ScaleSpriteSize001}
   &DynamicMaterialParameters(
    Index 0 Param 1=1
    Index 0 Param 2=1
    Index 0 Param 3=&FloatFromCurve(FloatCurve=DataInterface() {propertyValues {Curve {keys [
         {arriveTangent=0.1411, interpMode=RCIM_Cubic, leaveTangent=0.1411, tangentMode=RCTM_User, time=0.2, value=0}
         {arriveTangent=1.8523, interpMode=RCIM_Cubic, leaveTangent=1.8523, tangentMode=RCTM_User, time=1, value=0.5}
        ]}}})
    Module.Param0WriteEnabled=true
    Module.Param1WriteEnabled=true
    Module.Param2WriteEnabled=true
    Module.Param3WriteEnabled=false
   )
  ]
  rendererClasses [&NiagaraSpriteRendererProperties]
  renderers [{rendererClass=&NiagaraSpriteRendererProperties}]
 }
 Flare_merged {
  emitterUpdateScript [
   &EmitterState(Life Cycle Mode=::NewEnumerator1, Loop Duration=4, UseLoopDelay=false)
   &SpawnBurst_Instantaneous(Spawn Time=1, Use Spawn Probability=false, UseLoopCountLimit=false)
  ]
  particleSpawnScript [
   &InitializeParticle(
    Color=<rgba 0.1531 0.6763 1 1>
    Lifetime=3
    Sprite Rotation Mode=::NewEnumerator2
    Sprite Size=&SelectVector2DFromArray(
     Direct Array Index=&SpawnGroup()
     Vector 2D Selection Array=DataInterface() {propertyValues {FloatData [<800 800>, <1500 200>]}}
    )
    Sprite Size Mode=::NewEnumerator4
    UsePositionOffset=false
    Write Material Random=true
   )
   &CameraOffset(Camera Offset Amount=1500) {module=CameraOffset001}
  ]
  particleUpdateScript [
   &ParticleState()
   &ScaleSpriteSize(
    Uniform Curve Scale=15
    Uniform Curve Sprite Scale=DataInterface() {propertyValues {Curve {keys [
        {arriveTangent=0.3535, interpMode=RCIM_Cubic, leaveTangent=0.3535, tangentMode=RCTM_User, time=0, value=0}
        {arriveTangent=-0.0405, interpMode=RCIM_Cubic, leaveTangent=-0.0405, tangentMode=RCTM_User, time=1, value=1}
       ]}}}
   )
   &ScaleColor(
    Scale Alpha=&FloatFromCurve(
     FloatCurve=DataInterface() {propertyValues {Curve {keys [
         {arriveTangent=9.0725, interpMode=RCIM_Cubic, leaveTangent=9.0725, tangentMode=RCTM_User, time=0, value=0}
         {arriveTangent=-2.5249, interpMode=RCIM_Cubic, leaveTangent=-2.5249, tangentMode=RCTM_User, time=0.1408, value=0.9847}
         {arriveTangent=-0.0405, interpMode=RCIM_Cubic, leaveTangent=-0.0405, tangentMode=RCTM_User, time=1, value=0}
        ]}}}
     Scale Curve=0.5
    )
    Scale RGB=&VectorFromFloat(Value=1)
    ScaleA=true
    ScaleRGB=true
   ) {module=ScaleColor002}
   &SkeletalMeshLocation(Bone Sampling Mode=::NewEnumerator4, MaskSpawnGroup=false, Skeletal Mesh=$System.Skeletal Mesh)
  ]
  rendererClasses [&NiagaraSpriteRendererProperties]
  renderers [
   {rendererClass=&NiagaraSpriteRendererProperties}
   {rendererClass=&NiagaraSpriteRendererProperties, rendererIndex=1}
  ]
 }
 ParticleDecal001 {
  emitterUpdateScript [
   &EmitterState(Life Cycle Mode=::NewEnumerator1, Loop Behavior=::NewEnumerator1, Loop Duration Mode=::NewEnumerator1, UseLoopDelay=false)
   &SpawnBurst_Instantaneous(Spawn Time=0.9, Use Spawn Probability=false, UseLoopCountLimit=false)
  ]
  particleSpawnScript [
   &InitializeParticle(
    Color=<rgba 0.064 0.1693 1 1>
    Lifetime=3
    Mesh Scale Mode=::NewEnumerator3
    Mesh Uniform Scale=2
    Position Offset=<0 0 100>
    Sprite Size Mode=::NewEnumerator3
    Uniform Sprite Size=1500
    UsePositionOffset=true
    Write Material Random=true
   )
   &CameraOffset(Camera Offset Amount=10000)
   &DynamicMaterialParameters(
    Index 0 Param 1=0.1
    Index 0 Param 3=1
    Module.Index0Param0Write=true
    Module.Index0Param1Write=true
    Module.Index0Param2Write=true
    Module.Index0Param3Write=true
   )
  ]
  particleUpdateScript [
   &ParticleState()
   &ScaleSpriteSize(Uniform Curve Sprite Scale=DataInterface() {propertyValues {Curve {keys [
        <at 0 0 RCIM_Cubic>
        {arriveTangent=4.0654, interpMode=RCIM_Cubic, leaveTangent=4.0654, tangentMode=RCTM_User, time=0.3, value=1}
        {arriveTangent=-1.4286, interpMode=RCIM_Cubic, leaveTangent=-1.4286, time=0.8, value=1}
        <at 1 0 RCIM_Cubic>
       ]}}})
   &SkeletalMeshLocation(Bone Sampling Mode=::NewEnumerator4, MaskSpawnGroup=false, Skeletal Mesh=$System.Skeletal Mesh)
  ]
  rendererClasses [&NiagaraSpriteRendererProperties]
  renderers [{rendererClass=&NiagaraSpriteRendererProperties}]
 }
 PostProcess {
  bEnabled=false
  emitterUpdateScript [
   &EmitterState(Life Cycle Mode=::NewEnumerator1, Loop Behavior=::NewEnumerator1, UseLoopDelay=false)
   &SpawnBurst_Instantaneous(Spawn Time=0.05, Use Spawn Probability=false, UseLoopCountLimit=false)
   &SpawnBurst_Instantaneous(Spawn Group=1, Use Spawn Probability=false, UseLoopCountLimit=false) {module=SpawnBurst_Instantaneous001}
  ]
  particleSpawnScript [
   &InitializeParticle(
    Lifetime=5
    Mesh Scale Mode=::NewEnumerator3
    Sprite Size=&Multiply_Vector2D_ByFloat(Float=&Multiply_Float(A=$User.ElementSizeScale, B=$User.PostSizeScale), Vector2D=<3948.9736 3201.8015>)
    Sprite Size Mode=::NewEnumerator4
    UsePositionOffset=false
    Write Material Random=true
   )
   &CameraOffset()
   /SetVariables(Particles.VisibilityTag=&SpawnGroup()) {isSetParameters=true}
  ]
  particleUpdateScript [
   &ParticleState()
   &Timeline(
    Play Rate=&FloatFromCurve(
     FloatCurve=DataInterface() {propertyValues {Curve {keys [
         {interpMode=RCIM_Cubic, leaveTangent=-12.5334, tangentMode=RCTM_Break, time=0, value=4}
         {interpMode=RCIM_Cubic, tangentMode=RCTM_Break, time=0.7, value=1}
        ]}}}
     Scale Curve=0.692
    )
    Timeline End=1.0000000200408773e+20
   )
   &ScaleSpriteSize(Uniform Curve Sprite Scale=DataInterface() {propertyValues {Curve {keys [
        {arriveTangent=15.597, interpMode=RCIM_Cubic, leaveTangent=15.597, tangentMode=RCTM_User, time=0, value=0}
        {arriveTangent=0.3869, interpMode=RCIM_Cubic, leaveTangent=0.3869, tangentMode=RCTM_User, time=0.05, value=1}
        {arriveTangent=0.0224, interpMode=RCIM_Cubic, leaveTangent=0.0224, tangentMode=RCTM_User, time=1, value=1}
       ]}}})
   &DynamicMaterialParameters(
    Index 0 Param 1=&FloatFromCurve(
     FloatCurve=DataInterface() {propertyValues {Curve {keys [
         {interpMode=RCIM_Cubic, leaveTangent=-4.8343, tangentMode=RCTM_Break, time=0, value=2}
         {arriveTangent=-0.1577, interpMode=RCIM_Cubic, tangentMode=RCTM_Break, time=1, value=0}
        ]}}}
     Scale Curve=&Multiply_Float(A=2, B=$User.PostNoiseStrength)
    )
    Index 0 Param 2=&FloatFromCurve(
     FloatCurve=DataInterface() {propertyValues {Curve {keys [
         {interpMode=RCIM_Cubic, leaveTangent=-16.2213, tangentMode=RCTM_Break, time=0, value=4}
         {arriveTangent=-17.124, interpMode=RCIM_Cubic, tangentMode=RCTM_Break, time=0.25, value=0}
        ]}}}
     Scale Curve=&Multiply_Float(A=1, B=$User.PostStrength)
    )
    Index 0 Param 3=1
    Index 0 Param 4=1
    Index 1 Param 1=0.1
    Index 1 Param 2=1500
    Index 1 Param 3=0.5
    Index 1 Param 4=1
    Index 2 Param 1=$Particles.Timeline.Time
    Index 2 Param 2=$Particles.Timeline.Time
    Index 2 Param 3=$Particles.Timeline.Time
    Index 2 Param 4=&Multiply_Float(A=0.2, B=$User.ElementSizeScale)
    Index 3 Param 1=&FloatFromCurve(FloatCurve=DataInterface() {propertyValues {Curve {keys [<at 0.1 1>,<at 0.4 0>]}}})
    Index 3 Param 2=200
    Module.Index0Param0Write=true
    Module.Index0Param1Write=true
    Module.Index0Param2Write=true
    Module.Index0Param3Write=true
    Module.Index1Param0Write=true
    Module.Index1Param1Write=true
    Module.Index1Param2Write=true
    Module.Index1Param3Write=true
    Module.Index2Param0Write=true
    Module.Index2Param1Write=true
    Module.Index2Param2Write=true
    Module.Index2Param3Write=true
    Module.Index3Param0Write=true
    Module.Index3Param1Write=true
    Module.Index3Param2Write=true
    Module.Index3Param3Write=true
    Write Parameter Index 1=true
    Write Parameter Index 2=true
    Write Parameter Index 3=true
   )
   &ScaleColor(
    Scale Alpha=&FloatFromCurve(FloatCurve=DataInterface() {propertyValues {Curve {keys [{interpMode=RCIM_Cubic, leaveTangent=-6.3065, tangentMode=RCTM_Break, time=0.2, value=1}, <at 1 0>]}}})
    ScaleA=true
    ScaleRGB=true
   )
   /SetVariables(Particles.BlendTest:NiagaraFloat=&FloatFromCurve(
     FloatCurve=DataInterface() {propertyValues {Curve {keys [
         {arriveTangent=1.3123, interpMode=RCIM_Cubic, leaveTangent=1.3123, tangentMode=RCTM_User, time=0, value=0}
         {arriveTangent=0.0074, interpMode=RCIM_Cubic, leaveTangent=0.0074, tangentMode=RCTM_User, time=0.5, value=0.9847}
         {arriveTangent=-1.4231, interpMode=RCIM_Cubic, leaveTangent=-1.4231, tangentMode=RCTM_User, time=1, value=0}
        ]}}}
     Scale Curve=&Subtract_Float(
      A=1
      B=&FN_LOD_Distance_Remap(Start Distance=8000, End Distance=120000, Minimum Output=0, Maximum Output=1, Clamp Output=true)
     )
    )) {isSetParameters=true}
   /SetVariables(Particles.ComponentsEnabled:NiagaraBool=&Boolean_NotOperation(Bool=&SetBoolByFloatComparison(
      A=&FN_LOD_Distance_Remap(Start Distance=4000, End Distance=6000, Minimum Output=0, Maximum Output=1, Clamp Output=true)
      B=1
     ))) {isSetParameters=true}
   &KillParticles(Kill Particles=&SetBoolByFloatComparison(
     A=&FloatFromCurve(FloatCurve=DataInterface() {propertyValues {Curve {keys [<at 0.9 1>,<at 1 -0.01>]}}})
     B=0
     Comparison Type=::NewEnumerator6
    ))
  ]
  rendererClasses [&NiagaraSpriteRendererProperties, &NiagaraComponentRendererProperties]
  renderers [
   {rendererClass=&NiagaraSpriteRendererProperties}
   {rendererClass=&NiagaraComponentRendererProperties, rendererIndex=1}
  ]
 }
}
summary {
 systemName=NS_TentacleAttackChargeUpLaserLong
 emitters [
  {emitterName=Arcs, rendererClasses [&NiagaraSpriteRendererProperties]}
  {emitterName=Flare_merged, rendererClasses [&NiagaraSpriteRendererProperties]}
  {emitterName=ParticleDecal001, rendererClasses [&NiagaraSpriteRendererProperties]}
  {emitterName=Arcs001, rendererClasses [&NiagaraSpriteRendererProperties]}
  {
   bEnabled=false
   emitterName=PostProcess
   rendererClasses [&NiagaraSpriteRendererProperties, &NiagaraComponentRendererProperties]
  }
 ]
 userVariables [
  User.Color {type=LinearColor, description="", defaultValue {struct=LinearColor, value=<rgba 0 0.161 0.5 1>}}
  User.ColorMult {type=NiagaraFloat, defaultValue=1, description=""}
  User.ElementSizeScale {type=NiagaraFloat, defaultValue=1, description=""}
  User.PostNoiseStrength {type=NiagaraFloat, defaultValue=1, description=""}
  User.PostSizeScale {type=NiagaraFloat, defaultValue=1, description=""}
  User.PostStrength {type=NiagaraFloat, defaultValue=1, description=""}
  User.Scaling_SpawnRadius {type=NiagaraFloat, defaultValue=1, description=""}
  User.WidthMult {type=NiagaraFloat, defaultValue=1, description=""}
 ]
}
