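# Niagara system for a tentacle attack impact: an electric blue-purple blast with kicked-up dirt.
# Mostly burst emitters layer bolt and impact meshes, GPU dirt and shimmers, radial plasma and a light flash, alongside lightning that spawns across a widening radius.
# Camera-distance remaps scale several layers, owner scale sizes the dirt, and User parameters tune lightning radius, width and brightness.
# The post-process blast and ground shockwave emitters are disabled.
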NiagaraSystem /PistolCanary/Assets/FX/TentacleAttack/NS_TentacleAttackImpact
dependencies {
 usedDataInterfaces [
  &NiagaraDataInterfaceArrayColor
  &NiagaraDataInterfaceCurve
  &NiagaraDataInterfaceArrayFloat
  &NiagaraDataInterfaceColorCurve
  &NiagaraDataInterfaceVectorCurve
  &NiagaraDataInterfaceSkeletalMesh
 ]
 usedDynamicInputs []
 usedModules [
  &SystemState
  &self:SystemSpawnScript.SystemScriptSource.SystemScriptGraph.NiagaraNodeAssignment_1.SetVariables_#1
 ]
}
emitters {
 BlastPostProcess {
  bEnabled=false
  emitterUpdateScript [
   &EmitterState(Life Cycle Mode=::NewEnumerator1, Loop Behavior=::NewEnumerator1, UseLoopDelay=false)
   &SpawnBurst_Instantaneous(Spawn Group=1, Use Spawn Probability=false, UseLoopCountLimit=false) {module=SpawnBurst_Instantaneous001}
  ]
  particleSpawnScript [
   &InitializeParticle(
    Lifetime=2
    Mesh Scale Mode=::NewEnumerator3
    Sprite Size=&Multiply_Vector2D_ByFloat(Float=&Multiply_Float(A=$User.ElementSizeScale, B=$User.PostSizeScale), Vector2D=<3948.9736 3201.8015>)
    Sprite Size Mode=::NewEnumerator4
    UsePositionOffset=false
    Write Material Random=true
   )
   &CameraOffset()
  ]
  particleUpdateScript [
   &ParticleState()
   &Timeline(
    Play Rate=&FloatFromCurve(
     FloatCurve=DataInterface() {propertyValues {Curve {keys [
         {interpMode=RCIM_Cubic, leaveTangent=-12.5334, tangentMode=RCTM_Break, time=0, value=4}
         {interpMode=RCIM_Cubic, tangentMode=RCTM_Break, time=0.7, value=1}
        ]}}}
     Scale Curve=0.692
    )
    Timeline End=1.0000000200408773e+20
   )
   &ScaleSpriteSize(Uniform Curve Sprite Scale=DataInterface() {propertyValues {Curve {keys [
        {arriveTangent=15.597, interpMode=RCIM_Cubic, leaveTangent=15.597, tangentMode=RCTM_User, time=0, value=0}
        {arriveTangent=0.3869, interpMode=RCIM_Cubic, leaveTangent=0.3869, tangentMode=RCTM_User, time=0.05, value=1}
        {arriveTangent=0.0224, interpMode=RCIM_Cubic, leaveTangent=0.0224, tangentMode=RCTM_User, time=1, value=1}
       ]}}})
   &DynamicMaterialParameters(
    Index 0 Param 1=&FloatFromCurve(
     FloatCurve=DataInterface() {propertyValues {Curve {keys [
         {interpMode=RCIM_Cubic, leaveTangent=-4.8343, tangentMode=RCTM_Break, time=0, value=2}
         {arriveTangent=-0.1577, interpMode=RCIM_Cubic, tangentMode=RCTM_Break, time=1, value=0}
        ]}}}
     Scale Curve=&Multiply_Float(A=2, B=$User.PostNoiseStrength)
    )
    Index 0 Param 2=&FloatFromCurve(
     FloatCurve=DataInterface() {propertyValues {Curve {keys [
         {interpMode=RCIM_Cubic, leaveTangent=-16.2213, tangentMode=RCTM_Break, time=0, value=4}
         {arriveTangent=-17.124, interpMode=RCIM_Cubic, tangentMode=RCTM_Break, time=0.25, value=0}
        ]}}}
     Scale Curve=&Multiply_Float(A=1, B=$User.PostStrength)
    )
    Index 0 Param 3=1
    Index 0 Param 4=1
    Index 1 Param 1=0.1
    Index 1 Param 2=1500
    Index 1 Param 3=0.5
    Index 1 Param 4=1
    Index 2 Param 1=$Particles.Timeline.Time
    Index 2 Param 2=$Particles.Timeline.Time
    Index 2 Param 3=$Particles.Timeline.Time
    Index 2 Param 4=&Multiply_Float(A=0.2, B=$User.ElementSizeScale)
    Index 3 Param 1=&FloatFromCurve(FloatCurve=DataInterface() {propertyValues {Curve {keys [<at 0.1 1>,<at 0.4 0>]}}})
    Index 3 Param 2=200
    Module.Index0Param0Write=true
    Module.Index0Param1Write=true
    Module.Index0Param2Write=true
    Module.Index0Param3Write=true
    Module.Index1Param0Write=true
    Module.Index1Param1Write=true
    Module.Index1Param2Write=true
    Module.Index1Param3Write=true
    Module.Index2Param0Write=true
    Module.Index2Param1Write=true
    Module.Index2Param2Write=true
    Module.Index2Param3Write=true
    Module.Index3Param0Write=true
    Module.Index3Param1Write=true
    Module.Index3Param2Write=true
    Module.Index3Param3Write=true
    Write Parameter Index 1=true
    Write Parameter Index 2=true
    Write Parameter Index 3=true
   )
   &ScaleColor(
    Scale Alpha=&FloatFromCurve(FloatCurve=DataInterface() {propertyValues {Curve {keys [{interpMode=RCIM_Cubic, leaveTangent=-6.3065, tangentMode=RCTM_Break, time=0.2, value=1}, <at 1 0>]}}})
    ScaleA=true
    ScaleRGB=true
   )
   /SetVariables(Particles.BlendTest:NiagaraFloat=&FloatFromCurve(
     FloatCurve=DataInterface() {propertyValues {Curve {keys [
         {arriveTangent=1.3123, interpMode=RCIM_Cubic, leaveTangent=1.3123, tangentMode=RCTM_User, time=0, value=0}
         {arriveTangent=0.0074, interpMode=RCIM_Cubic, leaveTangent=0.0074, tangentMode=RCTM_User, time=0.0704, value=0.9847}
         {arriveTangent=-0.6466, interpMode=RCIM_Cubic, leaveTangent=-0.6466, tangentMode=RCTM_User, time=1, value=0}
        ]}}}
     Scale Curve=&Subtract_Float(
      A=1
      B=&FN_LOD_Distance_Remap(Start Distance=8000, End Distance=120000, Minimum Output=0, Maximum Output=1, Clamp Output=true)
     )
    )) {isSetParameters=true}
   /SetVariables(Particles.ComponentsEnabled:NiagaraBool=&Boolean_NotOperation(Bool=&SetBoolByFloatComparison(
      A=&FN_LOD_Distance_Remap(Start Distance=4000, End Distance=6000, Minimum Output=0, Maximum Output=1, Clamp Output=true)
      B=1
     ))) {isSetParameters=true}
   &KillParticles(Kill Particles=&SetBoolByFloatComparison(
     A=&FloatFromCurve(FloatCurve=DataInterface() {propertyValues {Curve {keys [<at 0.9 1>,<at 1 -0.01>]}}})
     B=0
     Comparison Type=::NewEnumerator6
    ))
  ]
  rendererClasses [&NiagaraComponentRendererProperties]
  renderers [{rendererClass=&NiagaraComponentRendererProperties, rendererIndex=0}]
 }
 Bolt {
  emitterUpdateScript [
   &EmitterState()
   &SpawnBurst_Instantaneous(Spawn Count=10, Use Spawn Probability=false, UseLoopCountLimit=false)
  ]
  particleSpawnScript [
   /SetVariables(Particles.MeshIndex=&RandomRangeInt(Minimum=0, Maximum=3)) {isSetParameters=true}
   &InitializeParticle(
    Color=&SelectLinearColorFromArray(
     Array Sampling Mode=::NewEnumerator1
     Direct Array Index=&RandomRangeInt(Minimum=0, Maximum=1)
     Color Selection Array=DataInterface() {propertyValues {ColorData [<rgba 0 0.514 5 1>, <rgba 1.123 0 5 1>]}}
     Random Seed=?
    )
    Lifetime=&RandomRangeFloat(Minimum=0.6, Maximum=0.5)
    Mesh Scale=<0 0 0>
    Mesh Scale Mode=::NewEnumerator4
    Sprite Size Mode=::NewEnumerator3
    Uniform Sprite Size=800
    UsePositionOffset=false
    Write Material Random=true
   )
   &SystemLocation(MaskSpawnGroup=false, Offset=&RandomRangeVector(Minimum=<0 0 25>, Maximum=<0 0 25>))
   &InitialMeshOrientation(Rotation=&MakeVector(X=0, Y=0, Z=1), Use Rotation Vector=true)
   &AddVelocityInCone(
    Cone Angle=90
    Cone Axis=<0 0 0.25>
    Module.Use Velocity Falloff On Cone Axis=false
    Velocity Distribution Along Cone Axis=0.25
    Velocity Strength=-25
   ) {enabled=false}
   &AddVelocity(Velocity=<0 0 1>) {enabled=false}
   /SetVariables(Particles.Scale=&Multiply_VectorByFloat(
     Vector=<2 2 2>
     Float=&RemapRange(
      Input Value To Remap=&FN_LOD_Distance_Remap(Start Distance=5000, End Distance=15000, Minimum Output=0.2, Maximum Output=1)
      Input Min=0
      Input Max=1
      Output Min=0
      Output Max=1
     )
    )) {isSetParameters=true}
   &CameraOffset(Camera Offset Amount=&Lerp_Float(
     A=-200
     B=0
     Alpha=&FN_LOD_Distance_Remap(Start Distance=0, End Distance=1500, Minimum Output=0, Maximum Output=1)
    )) {enabled=false}
  ]
  particleUpdateScript [
   &SolveForcesAndVelocity(Clamp Velocity=false, Limit Acceleration=false)
   &ParticleState()
   &ScaleColor(
    Scale Alpha=&FloatFromCurve(FloatCurve=DataInterface() {propertyValues {Curve {keys [
         {interpMode=RCIM_Cubic, tangentMode=RCTM_User, time=0, value=0}
         {interpMode=RCIM_Cubic, tangentMode=RCTM_User, time=0.5, value=1}
         {tangentMode=RCTM_User, time=0.6, value=0}
        ]}}})
    Scale RGB=&VectorFromFloat(Value=&FloatFromCurve(
      FloatCurve=DataInterface() {propertyValues {Curve {keys [
          {interpMode=RCIM_Cubic, tangentMode=RCTM_User, time=0, value=0}
          {arriveTangent=-3.0443, interpMode=RCIM_Cubic, leaveTangent=-3.0443, tangentMode=RCTM_User, time=0.1, value=0.95}
          {interpMode=RCIM_Cubic, tangentMode=RCTM_User, time=0.4, value=0.55}
          {tangentMode=RCTM_User, time=1, value=0}
         ]}}}
      Scale Curve=15
     ))
    ScaleA=true
    ScaleRGB=true
   )
   &DynamicMaterialParameters(
    Index 0 Param 1=60
    Index 0 Param 2=&FloatFromCurve(FloatCurve=DataInterface() {propertyValues {Curve {keys [
         {interpMode=RCIM_Cubic, tangentMode=RCTM_User, time=0, value=0}
         {arriveTangent=-2.9636, interpMode=RCIM_Cubic, leaveTangent=-2.9636, tangentMode=RCTM_User, time=0.4, value=0.95}
         {tangentMode=RCTM_User, time=0.9, value=0}
        ]}}})
    Index 0 Param 3=&FloatFromCurve(
     FloatCurve=DataInterface() {propertyValues {Curve {keys [
         {interpMode=RCIM_Cubic, tangentMode=RCTM_User, time=0, value=0}
         {tangentMode=RCTM_User, time=1.5, value=0.85}
         {interpMode=RCIM_Cubic, tangentMode=RCTM_User, time=2.5, value=1}
        ]}}}
     Scale Curve=2
    )
    Index 0 Param 4=1
    Index 1 Param 1=&FloatFromCurve(FloatCurve=DataInterface() {propertyValues {Curve {keys [{interpMode=RCIM_Cubic, tangentMode=RCTM_User, time=0, value=1}, {interpMode=RCIM_Cubic, tangentMode=RCTM_User, time=1, value=0}]}}})
    Index 1 Param 2=50
    Index 1 Param 3=1
    Index 1 Param 4=1
    Index 2 Param 1=&FloatFromCurve(FloatCurve=DataInterface() {propertyValues {Curve {keys [{interpMode=RCIM_Cubic, tangentMode=RCTM_User, time=0, value=1}, {tangentMode=RCTM_User, time=1, value=0.5}]}}})
    Module.Param0WriteEnabled=true
    Module.Param0WriteEnabled1=true
    Module.Param0WriteEnabled2=true
    Module.Param1WriteEnabled=true
    Module.Param1WriteEnabled1=true
    Module.Param1WriteEnabled2=true
    Module.Param2WriteEnabled=true
    Module.Param2WriteEnabled1=true
    Module.Param2WriteEnabled2=true
    Module.Param3WriteEnabled=true
    Module.Param3WriteEnabled1=true
    Module.Param3WriteEnabled2=true
    Write Parameter Index 1=true
    Write Parameter Index 2=true
   )
   &ScaleMeshSize(Scale Factor=&MakeVector(
     X=&FloatFromCurve(
      FloatCurve=DataInterface() {propertyValues {Curve {keys [
          {arriveTangent=-5.4159, interpMode=RCIM_Cubic, leaveTangent=-5.4159, tangentMode=RCTM_User, time=0, value=1.5}
          {interpMode=RCIM_Cubic, tangentMode=RCTM_User, time=1, value=1.5}
         ]}}}
      Scale Curve=3
     )
     Y=&FloatFromCurve(FloatCurve=DataInterface() {propertyValues {Curve {keys [
          {arriveTangent=73.8235, interpMode=RCIM_Cubic, leaveTangent=73.8235, tangentMode=RCTM_User, time=0, value=0}
          {tangentMode=RCTM_User, time=0.05, value=1}
         ]}}})
     Z=&FloatFromCurve(
      FloatCurve=DataInterface() {propertyValues {Curve {keys [
          {arriveTangent=9.6743, interpMode=RCIM_Cubic, leaveTangent=9.6743, tangentMode=RCTM_User, time=0, value=0}
          {interpMode=RCIM_Cubic, tangentMode=RCTM_User, time=0.1588, value=1.0402}
          {interpMode=RCIM_Cubic, tangentMode=RCTM_User, time=0.9122, value=1.3662}
          <at 1 1>
         ]}}}
      Scale Curve=2
     )
    ))
  ]
  rendererClasses [&NiagaraMeshRendererProperties]
  renderers [{rendererClass=&NiagaraMeshRendererProperties, rendererIndex=0}]
 }
 Comet_Impact {
  emitterUpdateScript [
   &EmitterState()
   &SpawnBurst_Instantaneous(Use Spawn Probability=false, UseLoopCountLimit=false)
  ]
  particleSpawnScript [
   &InitializeParticle(
    Color=<rgba 0.2591 0.7361 1 1>
    Lifetime=0.35
    Mesh Scale=<0.7 3 3>
    Mesh Scale Mode=::NewEnumerator4
    Position Offset=&AddVectorToPosition(Position=<0 0 -200>, Vector=$System.FXPosition)
    UsePositionOffset=true
    Write Material Random=true
   )
   &InitialMeshOrientation(
    Mesh Orientation Mode=::NewEnumerator2
    Orientation Coordinate Space=::World
    Orientation Vector=<0 0 1>
    Rotation=<0 0.5 0>
    Use Rotation Vector=true
   )
   &CameraOffset(Camera Offset Amount=250)
  ]
  particleUpdateScript [
   &ParticleState()
   &ScaleColor(
    Scale Alpha=&FloatFromCurve(
     FloatCurve=DataInterface() {propertyValues {Curve {keys [
         {arriveTangent=6.4765, interpMode=RCIM_Cubic, leaveTangent=6.4765, tangentMode=RCTM_User, time=0, value=0}
         <at 0.4 1 RCIM_Cubic>
         <at 1 0 RCIM_Cubic>
        ]}}}
     Scale Curve=2
    )
    Scale RGB=&MakeVectorFromLinearColorRGB(Color=<rgba 1 1 1 1>)
    ScaleA=true
    ScaleRGB=true
   )
   &ScaleMeshSize(Scale Factor=&VectorFromCurve(
     VectorCurve=DataInterface() {propertyValues {XCurve {keys [<at 0 1>,<at 1 1.5>]}, YCurve {keys [<at 0 1.1>,<at 1 0.9>]}, ZCurve {keys [<at 0 1.1>,<at 1 0.9>]}}}
     Scale Curve=<1 1.7 1.7>
    ))
   &ScaleMeshSize(Scale Factor=&VectorFromFloat(Value=4)) {module=ScaleMeshSize001}
  ]
  rendererClasses [&NiagaraMeshRendererProperties]
  renderers [{rendererClass=&NiagaraMeshRendererProperties, rendererIndex=0}]
 }
 Dirt {
  simTarget=GPUComputeSim
  emitterUpdateScript [
   &EmitterState()
   &SpawnBurst_Instantaneous(
    Spawn Count=&Multiply_Int(
     A=20
     B=&MakeIntFromFloat(Float=&MakeFloatFromVector(Vector=$Engine.Owner.Scale, Channel=::NewEnumerator0))
    )
    Use Spawn Probability=false
    UseLoopCountLimit=false
   )
  ]
  particleSpawnScript [
   &InitializeParticle(
    Color Maximum=<rgba 0.1 0.0906 0.0823 1>
    Color Minimum=<rgba 0.15 0.1091 0.0975 1>
    Color Mode=::NewEnumerator2
    Lifetime Max=3.5
    Lifetime Min=2
    Lifetime Mode=::NewEnumerator1
    Sprite Rotation Mode=::NewEnumerator2
    Sprite Size Mode=::NewEnumerator3
    Uniform Sprite Size=&Multiply_Float(
     A=&MakeFloatFromVector(Vector=$Engine.Owner.Scale, Channel=::NewEnumerator0)
     B=&RandomRangeFloat(Minimum=100, Maximum=300)
    )
    UsePositionOffset=false
    Write Material Random=true
   )
   &ShapeLocation(
    Handle Radius=0
    Large Radius=500
    MaskSpawnGroup=false
    Non Uniform Scale=<3 2 2>
    Offset=<0 0 10>
    Offset Mode=::NewEnumerator0
    Override Seed=false
    Shape Primitive=::NewEnumerator3
    UseOwnerScale=true
   )
   &SubUVAnimation(End Frame=3, SubUV Animation Mode:ENiagara_SubUVLookupMode=::NewEnumerator2)
   &AddVelocity(
    Velocity=&RandomRangeVector(Minimum=<-1200 -1200 1200>, Maximum=<1200 1200 1200>)
    Velocity Speed Scale=&RandomRangeFloat(Minimum=2, Maximum=5)
   ) {enabled=false}
   &CameraOffset(Camera Offset Amount=&RandomRangeFloat(Minimum=10, Maximum=20))
   &AddVelocity(
    Module.Use Falloff Distance=false
    Origin Offset=<0 0 -5000>
    Override Seed=false
    Velocity Mode=::NewEnumerator1
    Velocity Speed=&RandomRangeFloat(Minimum=2500, Maximum=5000)
   ) {module=AddVelocity001}
  ]
  particleUpdateScript [
   &ParticleState()
   &GravityForce(Gravity=<0 0 -3000>)
   &Drag(Drag=&RandomRangeFloat(Minimum=1, Maximum=1.5), Use Linear Drag=true, Use Rotational Drag=false)
   &SolveForcesAndVelocity(Clamp Velocity=false, Limit Acceleration=false)
   &ScaleSpriteSize(
    Uniform Curve Scale=7
    Uniform Curve Sprite Scale=DataInterface() {propertyValues {Curve {keys [
        {arriveTangent=2.6899, interpMode=RCIM_Cubic, leaveTangent=2.6899, tangentMode=RCTM_User, time=0, value=0.3}
        {arriveTangent=0.7, interpMode=RCIM_Cubic, leaveTangent=0.7, time=0.2, value=0.8}
        {arriveTangentWeight=0.1472, interpMode=RCIM_Cubic, leaveTangentWeight=0.5889, tangentMode=RCTM_User, time=1, value=1}
       ]}}}
   ) {module=ScaleSpriteSize001}
   &ScaleColor(
    Scale Alpha=&FloatFromCurve(
     FloatCurve=DataInterface() {propertyValues {Curve {keys [
         {interpMode=RCIM_Cubic, leaveTangentWeight=0.3668, tangentMode=RCTM_User, time=0, value=1}
         {arriveTangent=-0.3442, interpMode=RCIM_Cubic, leaveTangent=-0.3442, tangentMode=RCTM_User, time=1.0011, value=0.002}
        ]}}}
     Scale Curve=0.6
    )
    Scale RGB=&MakeVectorFromLinearColorRGB(Color=&ColorFromCurve(
      ColorCurve=DataInterface() {propertyValues {
        AlphaCurve {keys [<at 0 1 RCIM_Cubic>,<at 1 1 RCIM_Cubic>]}
        BlueCurve {keys [<at 0 14 RCIM_Cubic>,<at 0.2247 1>,<at 1 1 RCIM_Cubic>]}
        GreenCurve {keys [<at 0 6.3689 RCIM_Cubic>,<at 0.2247 1>,<at 1 1 RCIM_Cubic>]}
        RedCurve {keys [<at 0 0 RCIM_Cubic>,<at 0.2247 1>,<at 1 1 RCIM_Cubic>]}
       }}
      CurveIndex=$Emitter.Age
     ))
    ScaleA=true
    ScaleRGB=true
   )
   &SpriteRotationRate(Rotation Rate=&RandomRangeFloat(Minimum=-25, Maximum=25))
  ]
  rendererClasses [&NiagaraSpriteRendererProperties]
  renderers [{rendererClass=&NiagaraSpriteRendererProperties}]
 }
 Empty003 {
  emitterUpdateScript [
   /SetVariables(Emitter.saturation=-2) {isSetParameters=true}
   &EmitterState(Life Cycle Mode=::NewEnumerator1, Loop Behavior=::NewEnumerator1, Loop Duration=1, UseLoopDelay=true)
   &SpawnBurst_Instantaneous(Spawn Count=1, Use Spawn Probability=false)
  ]
  particleSpawnScript [
   &InitializeParticle(
    Color=<rgba 1 1 1 0.7>
    Lifetime=0.5
    Mesh Scale Mode=::NewEnumerator3
    Mesh Uniform Scale=2.5
    UsePositionOffset=false
    Write Material Random=true
   )
   &InitialMeshOrientation(
    Rotation=&RandomRangeVector(Minimum=<0 0 0>, Maximum=<1 1 1>)
    Use Orientation Vector=true
    Use Rotation Vector=true
   )
   /SetVariables(Particles.Scale=&Multiply_VectorByFloat(
     Vector=<1 1 1>
     Float=&RemapRange(
      Input Value To Remap=&FN_LOD_Distance_Remap(Start Distance=5000, End Distance=15000, Minimum Output=0.35, Maximum Output=1)
      Input Min=0
      Input Max=1
      Output Min=0
      Output Max=3
     )
    )) {isSetParameters=true}
  ]
  particleUpdateScript [
   &ParticleState()
   &ScaleMeshSize(Scale Factor=&VectorFromFloat(Value=&FloatFromCurve(
      FloatCurve=DataInterface() {propertyValues {Curve {keys [{arriveTangent=0.7482, interpMode=RCIM_Cubic, leaveTangent=0.7482, tangentMode=RCTM_User, time=0, value=0}, <at 1 1>]}}}
      Scale Curve=4
     )))
   &ScaleColor(
    Scale Alpha=&FloatFromCurve(FloatCurve=DataInterface() {propertyValues {Curve {keys [<at 0 1>,<at 1 0>]}}})
    Scale RGB=&MakeVectorFromLinearColorRGB(Color=&SelectLinearColorFromArray(
      Array Sampling Mode=::NewEnumerator1
      Direct Array Index=&RandomRangeInt(Minimum=0, Maximum=0)
      Color Selection Array=DataInterface() {propertyValues {ColorData [<rgba 20.4006 1.0202 100 1>, <rgba 2.3063 1.068 10 1>]}}
      Random Seed=?
     ))
    ScaleA=true
    ScaleRGB=true
   )
  ]
  rendererClasses [&NiagaraMeshRendererProperties]
  renderers [{rendererClass=&NiagaraMeshRendererProperties, rendererIndex=0}]
 }
 Light001 {
  emitterUpdateScript [
   &EmitterState()
   &SpawnBurst_Instantaneous(Spawn Count=1, Use Spawn Probability=false)
  ]
  particleSpawnScript [
   &InitializeParticle(
    Color=<rgba 4.0101 1.0075 5 1>
    Lifetime Max=0.35
    Lifetime Min=0.3
    Lifetime Mode=::NewEnumerator1
    Sprite Size Mode=::NewEnumerator1
    Uniform Sprite Size Max=5000
    Uniform Sprite Size Min=5000
    UsePositionOffset=false
    Write Material Random=true
   )
   /SetVariables(Particles.SpriteSize=&Multiply_Vector2D_ByFloat(
     Float=&RemapRange(
      Input Value To Remap=&FN_LOD_Distance_Remap(Start Distance=5000, End Distance=15000, Minimum Output=1, Maximum Output=3)
      Input Min=0
      Input Max=1
      Output Min=0
      Output Max=2
     )
     Vector2D=<5000 5000>
    )) {isSetParameters=true}
   &SystemLocation(MaskSpawnGroup=false)
   /SetVariables(Particles.LightPos:Vector3f=<0 0 500>) {isSetParameters=true}
  ]
  particleUpdateScript [
   &ParticleState()
   &CameraOffset(Camera Offset Amount=100)
   &ScaleColor(
    Scale Mode=::NewEnumerator1
    Scale RGBA=&MakeVector4_FromLinearColor(LinearColor=&LerpLinearColors(StartColor=<rgba 0.0957 0.5612 1 1>, EndColor=<rgba 1.2268 0.3389 2 1>, LerpFactor=1))
    ScaleRGBA=true
   )
   &ScaleSpriteSize(Uniform Curve Sprite Scale=DataInterface() {propertyValues {Curve {keys [
        {arriveTangent=1.1934, interpMode=RCIM_Cubic, leaveTangent=1.1934, tangentMode=RCTM_User, time=0, value=0}
        {arriveTangent=0.4403, interpMode=RCIM_Cubic, leaveTangent=0.4403, tangentMode=RCTM_User, time=0.8, value=0.85}
        {arriveTangent=-0.0405, interpMode=RCIM_Cubic, leaveTangent=-0.0405, tangentMode=RCTM_User, time=1.3, value=1}
       ]}}})
   &DynamicMaterialParameters(
    Index 0 Param 1=0.5
    Index 0 Param 2=1
    Index 1 Param 1=500
    Module.Param0WriteEnabled=true
    Module.Param0WriteEnabled1=true
    Module.Param1WriteEnabled=true
    Module.Param1WriteEnabled1=true
    Module.Param2WriteEnabled=true
    Module.Param2WriteEnabled1=true
    Module.Param3WriteEnabled=true
    Module.Param3WriteEnabled1=true
    Write Parameter Index 1=true
   )
   /SetVariables(
    Particles.LightColor=&ColorFromCurve(ColorCurve=DataInterface() {propertyValues {
       AlphaCurve {keys [<at 0 1 RCIM_Cubic>,<at 1 0 RCIM_Cubic>]}
       BlueCurve {keys [{arriveTangent=-499, interpMode=RCIM_Cubic, time=0, value=20}, {interpMode=RCIM_Cubic, leaveTangent=-499, time=1, value=1}]}
       GreenCurve {keys [{arriveTangent=-6.2674, interpMode=RCIM_Cubic, time=0, value=1.2918}, {interpMode=RCIM_Cubic, leaveTangent=-6.2674, time=1, value=0.2897}]}
       RedCurve {keys [{arriveTangent=1, interpMode=RCIM_Cubic, time=0, value=0.6868}, {interpMode=RCIM_Cubic, leaveTangent=1, time=1, value=0.6258}]}
      }})
    Particles.LightIntensity_Component:NiagaraFloat=&FloatFromCurve(
     FloatCurve=DataInterface() {propertyValues {Curve {keys [
         {interpMode=RCIM_Cubic, tangentMode=RCTM_User, time=0, value=0}
         {interpMode=RCIM_Cubic, tangentMode=RCTM_User, time=0.1, value=0.7}
         <at 1 0>
        ]}}}
     Scale Curve=1000
    )
    Particles.LightRadius=&FloatFromCurve(
     FloatCurve=DataInterface() {propertyValues {Curve {keys [{interpMode=RCIM_Cubic, tangentMode=RCTM_User, time=0, value=0}, {interpMode=RCIM_Cubic, tangentMode=RCTM_User, time=0.1, value=1}]}}}
     Scale Curve=400
    )
   ) {isSetParameters=true}
  ]
  rendererClasses [
   &NiagaraSpriteRendererProperties
   &NiagaraLightRendererProperties
   &NiagaraComponentRendererProperties
  ]
  renderers [
   {rendererClass=&NiagaraSpriteRendererProperties}
   {rendererClass=&NiagaraLightRendererProperties, rendererIndex=1}
   {rendererClass=&NiagaraComponentRendererProperties, rendererIndex=2}
  ]
 }
 Lightning_Primary001 {
  emitterUpdateScript [
   &EmitterState(Life Cycle Mode=::NewEnumerator1, Loop Behavior=::NewEnumerator1, Loop Duration=0.8, UseLoopDelay=false)
   &SpawnRate(SpawnRate=40, Use Spawn Probability=false) {module=SpawnRate001}
  ]
  particleSpawnScript [
   /SetVariables(Particles.isShort:NiagaraFloat=&FloatFromCurve(
     FloatCurve=DataInterface() {propertyValues {Curve {keys [<at 0 0 RCIM_Constant>,<at 0.8 1 RCIM_Constant>]}}}
     CurveIndex=&RandomRangeFloat(Minimum=0, Maximum=1)
    )) {isSetParameters=true}
   &InitializeParticle(
    Color=&SelectLinearColorFromArray(
     Array Sampling Mode=::NewEnumerator0
     Direct Array Index=?
     Color Selection Array=DataInterface() {propertyValues {ColorData [<rgba 0.0136 0.0832 0.3 1>, <rgba 0.1221 0.0138 0.4 1>]}}
     Random Seed=0
    )
    Lifetime=&Lerp_Float(
     A=&RandomRangeFloat(Minimum=0.5, Maximum=0.8)
     B=&RandomRangeFloat(Minimum=0.3, Maximum=0.4)
     Alpha=$Particles.isShort
    )
    Mesh Scale Mode=::NewEnumerator3
    Mesh Uniform Scale=&MakeFloatFromInt(Int=&Multiply_Int(A=&RandomRangeInt(Minimum=2, Maximum=3), B=4))
    Position Mode=::NewEnumerator1
    Sprite Size Mode=::NewEnumerator4
    UsePositionOffset=false
    Write Material Random=true
   ) {module=InitializeParticle001}
   &DynamicMaterialParameters(
    Index 0 Param 1=1
    Index 0 Param 2=0.15
    Index 0 Param 3=1
    Index 0 Param 4=1
    Module.Param0WriteEnabled=true
    Module.Param1WriteEnabled=true
    Module.Param2WriteEnabled=true
    Module.Param3WriteEnabled=true
   )
   &ShapeLocation(
    MaskSpawnGroup=false
    Non Uniform Scale=<1 1 0>
    Offset Mode=::NewEnumerator0
    Override Seed=false
    Sphere Radius=&FloatFromCurve(
     FloatCurve=DataInterface() {propertyValues {Curve {defaultValue=1, keys [{tangentMode=RCTM_User, time=0, value=0}, {tangentMode=RCTM_User, time=1, value=1}]}}}
     CurveIndex=$Emitter.NormalizedLoopAge
     Scale Curve=$User.LightningRadius
    )
    Sphere Surface Distribution=1
    UseOwnerScale=true
   )
   &AddVelocity(Module.Use Falloff Distance=false, Override Seed=false, Velocity Mode=::NewEnumerator1, Velocity Speed=0.25)
   &InitialMeshOrientation(
    Mesh Orientation Mode=::NewEnumerator7
    Rotation=&RandomRangeVector(Minimum=<0 0 0>, Maximum=<0 0 0>)
    Use Rotation Vector=true
   ) {module=InitialMeshOrientation002}
  ]
  particleUpdateScript [
   &SolveForcesAndVelocity(Clamp Velocity=false, Limit Acceleration=false)
   &ParticleState()
   &ScaleMeshSize(Scale Factor=&Multiply_VectorByFloat(
     Vector=&MakeVector(
      X=&Multiply_Float(
       A=&FloatFromCurve(FloatCurve=DataInterface() {propertyValues {Curve {keys [{interpMode=RCIM_Cubic, tangentMode=RCTM_User, time=0, value=1}, {interpMode=RCIM_Cubic, tangentMode=RCTM_User, time=1, value=1.5}]}}})
       B=&RandomRangeFloat(Minimum=0.5, Maximum=0.8)
      )
      Y=1
      Z=1
     )
     Float=&MakeFloatFromVector(
      Vector=&VectorFromFloat(Value=&Multiply_Float(
        A=&RandomRangeFloat(Minimum=0.8, Maximum=1.2)
        B=&Multiply_Float(
         A=0.1
         B=&FloatFromCurve(FloatCurve=DataInterface() {propertyValues {Curve {keys [
              {arriveTangent=0.2102, interpMode=RCIM_Cubic, leaveTangent=0.2102, tangentMode=RCTM_User, time=0, value=0.9}
              {interpMode=RCIM_Cubic, tangentMode=RCTM_User, time=1, value=1}
             ]}}})
        )
       ))
      Channel=::NewEnumerator0
     )
    ))
   &DynamicMaterialParameters(
    Index 0 Param 1=&Lerp_Float(
     A=&FloatFromCurve(FloatCurve=DataInterface() {propertyValues {Curve {keys [{tangentMode=RCTM_User, time=0, value=1}, <at 0.2 0.5>, <at 0.8 0.5>, {tangentMode=RCTM_User, time=1, value=0}]}}})
     B=0.4
     Alpha=$Particles.isShort
    )
    Index 0 Param 2=&FloatFromCurve(
     FloatCurve=DataInterface() {propertyValues {Curve {keys [
         {interpMode=RCIM_Cubic, leaveTangent=-1.1405, tangentMode=RCTM_Break, time=0, value=0.5}
         {arriveTangent=4.4936, interpMode=RCIM_Cubic, leaveTangent=-5.3647, tangentMode=RCTM_Break, time=0.4833, value=1}
         <at 1 0>
        ]}}}
     Scale Curve=&Multiply_Float(A=14, B=$User.WidthMult)
    )
    Index 0 Param 3=100
    Index 1 Param 1=&RandomRangeFloat(Minimum=0, Maximum=1)
    Index 1 Param 2=1
    Index 1 Param 3=1
    Index 1 Param 4=-350
    Module.Param0WriteEnabled=true
    Module.Param0WriteEnabled1=true
    Module.Param1WriteEnabled=true
    Module.Param1WriteEnabled1=true
    Module.Param2WriteEnabled=true
    Module.Param2WriteEnabled1=true
    Module.Param3WriteEnabled=false
    Module.Param3WriteEnabled1=true
    Write Parameter Index 1=true
   ) {module=DynamicMaterialParameters001}
   &ScaleColor(
    Scale Alpha=&FloatFromCurve(FloatCurve=DataInterface() {propertyValues {Curve {keys [
         {interpMode=RCIM_Cubic, tangentMode=RCTM_User, time=0, value=0}
         {interpMode=RCIM_Cubic, tangentMode=RCTM_User, time=0.1258, value=1}
         <at 0.341 0>
         {interpMode=RCIM_Cubic, tangentMode=RCTM_User, time=0.496, value=0.5}
         <at 1 0>
        ]}}})
    Scale RGB=&VectorFromFloat(Value=&FloatFromCurve(
      FloatCurve=DataInterface() {propertyValues {Curve {keys [
          {arriveTangent=0.1143, interpMode=RCIM_Cubic, leaveTangent=0.1143, tangentMode=RCTM_User, time=0.018, value=4.9326}
          {arriveTangent=-6.0507, interpMode=RCIM_Cubic, leaveTangent=-6.0507, time=0.1778, value=0.8567}
          {arriveTangent=-0.5645, interpMode=RCIM_Cubic, leaveTangent=-0.5645, tangentMode=RCTM_User, time=0.2493, value=3.533}
          {arriveTangent=-22.9399, interpMode=RCIM_Cubic, leaveTangent=-22.9399, time=0.3224, value=1.7136}
          {arriveTangent=2.8763, interpMode=RCIM_Cubic, leaveTangent=2.8763, time=0.3773, value=0.5981}
          {arriveTangent=-1.8312, interpMode=RCIM_Cubic, leaveTangent=-1.8312, time=0.4693, value=2.1361}
          {arriveTangent=-2.2307, interpMode=RCIM_Cubic, leaveTangent=-2.2307, time=0.5983, value=0.1933}
          {arriveTangent=1.5587, interpMode=RCIM_Cubic, leaveTangent=1.5587, time=0.7391, value=1.5342}
          {arriveTangent=-2.0478, interpMode=RCIM_Cubic, leaveTangent=-2.0478, time=0.8629, value=0.6057}
          {arriveTangent=-0.0659, interpMode=RCIM_Cubic, leaveTangent=-0.0659, tangentMode=RCTM_User, time=1, value=1}
         ]}}}
      Scale Curve=&Multiply_Float(
       A=$User.ColorMult
       B=&RemapRange(
        Input Value To Remap=&FN_LOD_Distance_Remap(Start Distance=1000, End Distance=10000, Minimum Output=0, Maximum Output=1)
        Input Min=0
        Input Max=1
        Output Min=0
        Output Max=3
       )
      )
     ))
    ScaleA=true
    ScaleRGB=true
   ) {module=ScaleColor001}
  ]
  rendererClasses [&NiagaraMeshRendererProperties]
  renderers [{rendererClass=&NiagaraMeshRendererProperties, rendererIndex=0}]
 }
 Radial_Plasma_Energy {
  emitterUpdateScript [
   &EmitterState()
   &SpawnBurst_Instantaneous(Spawn Count=1, Use Spawn Probability=false)
  ]
  particleSpawnScript [
   &InitializeParticle(
    Color Mode=::NewEnumerator0
    Lifetime=0.35
    Sprite Size Mode=::NewEnumerator3
    Uniform Sprite Size=2000
    UsePositionOffset=false
    Write Material Random=true
   )
   /SetVariables(Particles.SpriteSize=&Multiply_Vector2D_ByFloat(
     Float=&FN_LOD_Distance_Remap(Start Distance=5000, End Distance=15000, Minimum Output=0.35, Maximum Output=1)
     Vector2D=<10000 10000>
    )) {isSetParameters=true}
  ]
  particleUpdateScript [
   &ParticleState()
   &FN_Radial_Material(Base Sprite Tile U=0.5, Base Sprite Tile V=3, Depth Fade=2500, Radial Mask Offset X=0.1, Sprite Extrusion Speed=1.25)
   &Color(
    Color=&ColorFromCurve(ColorCurve=DataInterface() {propertyValues {
       AlphaCurve {keys [<at 0 1 RCIM_Cubic>,<at 1 1 RCIM_Cubic>]}
       BlueCurve {keys [<at 0.4032 1 RCIM_Cubic>,<at 1 1 RCIM_Cubic>]}
       GreenCurve {keys [
         {arriveTangent=1.1139, interpMode=RCIM_Cubic, leaveTangent=1.1139, time=0.4032, value=0.0874}
         {arriveTangent=1.1139, interpMode=RCIM_Cubic, leaveTangent=1.1139, time=1, value=0.7522}
        ]}
       RedCurve {keys [<at 0.4032 0 RCIM_Cubic>,<at 1 0 RCIM_Cubic>]}
      }})
    Scale Alpha=0.7
    Scale Color=&VectorFromFloat(Value=&FloatFromCurve(
      FloatCurve=DataInterface() {propertyValues {Curve {keys [
          {arriveTangent=-2.6449, interpMode=RCIM_Cubic, leaveTangent=-2.6449, tangentMode=RCTM_User, time=0, value=1}
          {interpMode=RCIM_Cubic, tangentMode=RCTM_User, time=1, value=0.01}
         ]}}}
      Scale Curve=50
     ))
   )
   &ScaleSpriteSize(Scale Factor=&Vector2DFromFloat(Value=&FloatFromCurve(FloatCurve=DataInterface() {propertyValues {Curve {keys [
          {arriveTangent=2.2779, interpMode=RCIM_Cubic, leaveTangent=2.2779, tangentMode=RCTM_User, time=0, value=0}
          {interpMode=RCIM_Cubic, tangentMode=RCTM_User, time=1, value=1}
         ]}}})))
  ]
  rendererClasses [&NiagaraSpriteRendererProperties]
  renderers [{rendererClass=&NiagaraSpriteRendererProperties}]
 }
 Shimmers001 {
  simTarget=GPUComputeSim
  emitterUpdateScript [
   &EmitterState(
    Enable Distance Culling=true
    MaxDistance=25000
    Min Distance Response=::SleepAndLetParticlesFinish
    Module.bUseMaxDistance=true
    Module.bUseMinDistance=true
    Scalability Mode=::NewEnumerator1
   )
   &SpawnBurst_Instantaneous(Spawn Count=10, Use Spawn Probability=false, UseLoopCountLimit=false)
  ]
  particleSpawnScript [
   /SetVariables(
    Particles.ColorEnd:LinearColor=&RandomRangeLinearColor(Minimum=<rgba 0.25 0 0.138 1>, Maximum=<rgba 0.0466 0 0.25 1>)
    Particles.ColorStart:LinearColor=&RandomRangeLinearColor(Minimum=<rgba 0.36 0.3 0.45 1>, Maximum=<rgba 0.6 0.68 0.74 1>)
   ) {isSetParameters=true}
   &InitializeParticle(
    Color=<rgba 1 1 1 5>
    Lifetime Max=0.3
    Lifetime Min=0.2
    Lifetime Mode=::NewEnumerator1
    Mesh Scale Mode=::NewEnumerator1
    Mesh Uniform Scale Max=35
    Mesh Uniform Scale Min=15
    Position Mode=::NewEnumerator1
    Sprite Rotation Mode=::NewEnumerator2
    Sprite Size=&RandomRangeVector2D(Minimum=<400 400>, Maximum=<800 800>, Evaluation Type=::NewEnumerator1)
    Sprite Size Mode=::NewEnumerator4
    UsePositionOffset=false
    Write Material Random=true
   )
   &AddVelocity(Coordinate Space=::World, Velocity=&RandomRangeVector(Minimum=<-25 -2 -2>, Maximum=<-15 2 2>))
   &DynamicMaterialParameters(
    Index 0 Param 1=&RandomRangeFloat(Minimum=0, Maximum=1, Evaluation Type=::NewEnumerator1)
    Index 0 Param 2=&RandomRangeFloat(Minimum=0.4, Maximum=0.6, Evaluation Type=::NewEnumerator1)
    Index 0 Param 3=0.2
    Module.Param0WriteEnabled=true
    Module.Param1WriteEnabled=true
    Module.Param2WriteEnabled=true
    Module.Param3WriteEnabled=true
   )
   &SphereLocation(
    MaskSpawnGroup=false
    Non Uniform Scale=<1 1 0.2>
    Override Seed=false
    Sphere Radius=25
    Surface Only=false
    UseNonUniformScale=true
    UseOffset=false
    UseSphereOrientation=false
   )
   /SetVariables(Particles.Scale=&Multiply_VectorByFloat(
     Vector=$Particles.Scale
     Float=&RemapRange(
      Input Value To Remap=&FN_LOD_Distance_Remap(Start Distance=5000, End Distance=15000, Minimum Output=0.35, Maximum Output=1)
      Input Min=0
      Input Max=1
      Output Min=0
      Output Max=1
     )
    )) {isSetParameters=true}
   &InitialMeshOrientation(
    Mesh Orientation Mode=::NewEnumerator2
    Orientation Vector=<1 0 1>
    Rotation=&RandomRangeVector(Minimum=<0 0 0>, Maximum=<1 1 1>)
    Use Rotation Vector=true
   )
  ]
  particleUpdateScript [
   &ParticleState()
   &SolveForcesAndVelocity(Clamp Velocity=false, Limit Acceleration=false)
   &SpriteRotationRate(Rotation Rate=&RandomRangeFloat(Minimum=-20, Maximum=20))
   &ScaleColor(
    Scale Alpha=&FloatFromCurve(
     FloatCurve=DataInterface() {propertyValues {Curve {keys [{arriveTangent=1.0833, interpMode=RCIM_Cubic, leaveTangent=1.0833, time=0, value=0}, <at 0.6 0.65 RCIM_Cubic>, <at 1 0>]}}}
     Scale Curve=&RandomRangeFloat(Minimum=0.8, Maximum=1)
    )
    Scale RGB=&MakeVectorFromLinearColorRGB(Color=&SelectLinearColorFromArray(
      Array Sampling Mode=::NewEnumerator1
      Direct Array Index=&RandomRangeInt(Minimum=0, Maximum=1)
      Color Selection Array=DataInterface() {propertyValues {ColorData [<rgba 0.3333 0.6666 50 1>, <rgba 35.2793 9.058 50 1>]}}
      Random Seed=?
     ))
    ScaleA=true
    ScaleRGB=true
   ) {module=ScaleColor001}
   &DynamicMaterialParameters(
    Index 0 Param 1=3
    Index 0 Param 2=&FloatFromCurve(FloatCurve=DataInterface() {propertyValues {Curve {keys [
         {interpMode=RCIM_Cubic, tangentMode=RCTM_User, time=0, value=-1}
         {interpMode=RCIM_Cubic, tangentMode=RCTM_User, time=0.5, value=1}
         {tangentMode=RCTM_User, time=1, value=-1}
        ]}}})
    Index 0 Param 3=&RandomRangeFloat(Minimum=1, Maximum=5)
    Index 0 Param 4=1
    Module.Param0WriteEnabled=true
    Module.Param1WriteEnabled=true
    Module.Param2WriteEnabled=true
    Module.Param3WriteEnabled=true
   ) {module=DynamicMaterialParameters001}
   &ScaleMeshSize(Scale Factor=&VectorFromFloat(Value=&FloatFromCurve(
      FloatCurve=DataInterface() {propertyValues {Curve {keys [
          {arriveTangent=1.8697, interpMode=RCIM_Cubic, leaveTangent=1.8697, tangentMode=RCTM_User, time=0, value=0.625}
          {arriveTangent=0.7841, interpMode=RCIM_Cubic, leaveTangent=0.7841, tangentMode=RCTM_User, time=0.4, value=1}
          {arriveTangent=-0.0405, interpMode=RCIM_Cubic, leaveTangent=-0.0405, tangentMode=RCTM_User, time=1, value=1.125}
         ]}}}
      Scale Curve=5
     )))
  ]
  rendererClasses [&NiagaraMeshRendererProperties]
  renderers [{rendererClass=&NiagaraMeshRendererProperties, rendererIndex=0}]
 }
 Shockwave_Ground {
  bEnabled=false
  emitterUpdateScript [
   &EmitterState(Scalability Mode=::NewEnumerator1)
   &SpawnBurst_Instantaneous(Spawn Count=3, Spawn Time=0.05, Use Spawn Probability=false, UseLoopCountLimit=false)
  ]
  particleSpawnScript [
   &PartitionParticles(Number Of Partitions=3)
   /SetVariables(Particles.SpawnGroup=$Particles.PartitionParticles.Partition) {isSetParameters=true}
   &InitializeParticle(
    Color=$User.PrimaryColor
    Lifetime=&MergeHelper_FloatSwitch(0=0.4, 1=0.5, 2=0.4, 3=?, Index=$Particles.SpawnGroup, Num=::NewEnumerator1)
    Mesh Scale Mode=::NewEnumerator3
    Mesh Uniform Scale=&MergeHelper_FloatSwitch(0=12, 1=24, 2=5, 3=?, Index=$Particles.SpawnGroup, Num=::NewEnumerator1)
    Position Offset=&AddVectorToPosition(Position=<0 0 -100>, Vector=$System.FXPosition)
    UsePositionOffset=true
    Write Material Random=true
   )
   &FN_SetSpawnGroup() {enabled=false}
   &InitialMeshOrientation(Rotation=<0 0 0.5>, Use Orientation Vector=true, Use Rotation Vector=true)
  ]
  particleUpdateScript [
   &ParticleState()
   &DynamicMaterialParameters(
    Index 0 Param 1=&FloatFromCurve(FloatCurve=DataInterface() {propertyValues {Curve {keys [<at 0 2 RCIM_Cubic>, {arriveTangent=-3.5767, interpMode=RCIM_Cubic, leaveTangent=-3.5767, tangentMode=RCTM_User, time=1, value=-0.5}]}}})
    Index 0 Param 2=0.3
    Index 0 Param 3=0.5
    Index 0 Param 4=1
    Module.Param0WriteEnabled=true
    Module.Param1WriteEnabled=true
    Module.Param2WriteEnabled=true
    Module.Param3WriteEnabled=true
   )
   &ScaleColor(
    Scale Alpha=&FloatFromCurve(
     FloatCurve=DataInterface() {propertyValues {Curve {keys [
         {arriveTangent=-3.3102, interpMode=RCIM_Cubic, time=0, value=1}
         {arriveTangent=-0.4829, interpMode=RCIM_Cubic, leaveTangent=-0.4829, tangentMode=RCTM_User, time=0.1, value=0.15}
         <at 1 0 RCIM_Cubic>
        ]}}}
     Scale Curve=10
    )
    Scale RGB=&Multiply_VectorByFloat(
     Vector=&MakeVectorFromLinearColorRGB(Color=<rgba 1 1 1 1>)
     Float=&FNTimeOfDayScalar(Night Scale=0.8, Morning Scale=0.9, Evening Scale=0.9)
    )
    ScaleA=true
    ScaleRGB=true
   )
   /SetVariables(Particles.ShockwaveBubbleColor:LinearColor=&MultiplyLinearColor(Linear Color=$StackContext.Color, Scale Factor=<2 2 2 15>)) {isSetParameters=true}
   &ScaleMeshSize(Scale Factor=&VectorFromFloat(Value=&MergeHelper_FloatSwitch(
      0=&FloatFromCurve(
       FloatCurve=DataInterface() {propertyValues {Curve {keys [{arriveTangent=1.6867, interpMode=RCIM_Cubic, leaveTangent=1.6867, tangentMode=RCTM_User, time=0, value=0}, <at 1 1 RCIM_Cubic>]}}}
       Scale Curve=0.2
      )
      1=&FloatFromCurve(
       FloatCurve=DataInterface() {propertyValues {Curve {keys [
           {arriveTangent=2.4274, interpMode=RCIM_Cubic, leaveTangent=2.4274, tangentMode=RCTM_User, time=0, value=0}
           {arriveTangent=0.0224, interpMode=RCIM_Cubic, leaveTangent=0.0224, tangentMode=RCTM_User, time=1, value=1}
          ]}}}
       Scale Curve=&Multiply_Float(A=1.5, B=1)
      )
      2=?
      3=?
      Index=$Emitter.SpawnGroup
      Num=::NewEnumerator0
     )))
   /SetVariables(Particles.SphereColor:LinearColor=<rgba 1 0.5567 0.0205 1>) {enabled=false, isSetParameters=true}
   /SetVariables(Particles.ShockwaveBubbleParams:Vector4f=&MakeVector4(
     X=&FloatFromCurve(
      FloatCurve=DataInterface() {propertyValues {Curve {keys [{arriveTangent=5.3858, interpMode=RCIM_Cubic, time=0, value=0}, <at 0.974 1 RCIM_Cubic>]}}}
      Scale Curve=5
     )
     Y=1
     Z=&FloatFromCurve(FloatCurve=DataInterface() {propertyValues {Curve {keys [<at 0 0.4>,<at 1 0.4>]}}}, Scale Curve=0.25)
     W=&FloatFromCurve(FloatCurve=DataInterface() {propertyValues {Curve {keys [<at 0 0.5>,<at 1 12>]}}})
    )) {isSetParameters=true}
   /SetVariables(Particles.SpikeParams:Vector4f=&MakeVector4(
     X=&FloatFromCurve(FloatCurve=DataInterface() {propertyValues {Curve {keys [<at 0 0 RCIM_Cubic>,<at 1 0.28 RCIM_Cubic>]}}})
     Y=&RandomRangeFloat(Minimum=0, Maximum=1)
     Z=100
     W=&FloatFromCurve(FloatCurve=DataInterface() {propertyValues {Curve {keys [
          {interpMode=RCIM_Cubic, leaveTangent=0.0377, tangentMode=RCTM_Break, time=0, value=0}
          {arriveTangent=-3.0374, interpMode=RCIM_Cubic, leaveTangent=2.0371, tangentMode=RCTM_Break, time=0.2, value=-1.5}
          {interpMode=RCIM_Cubic, tangentMode=RCTM_Break, time=1, value=0}
         ]}}})
    )) {isSetParameters=true}
   &FN_Mesh_Attributes(
    Color=&MultiplyLinearColor(
     Linear Color=<rgba 0.25 0.224 0.24 1>
     Scale Factor=&MakeVector4_FromLinearColor(LinearColor=&MakeLinearColorFromVectorAndFloat(
       "Vector (RGB)"=&VectorFromFloat(Value=1)
       "Float (Alpha)"=&FloatFromCurve(
        FloatCurve=DataInterface() {propertyValues {Curve {keys [
            <at 0.0045 0.9397 RCIM_Cubic>
            {arriveTangent=-0.9408, interpMode=RCIM_Cubic, leaveTangent=-0.9408, time=0.5, value=0.125}
            {interpMode=RCIM_Cubic, leaveTangent=-2.4196, time=1.0011, value=0.002}
           ]}}}
        Scale Curve=0.4
       )
      ))
    )
    Create Color Attribute=true
    Create Scale Attribute=true
    Scale=&VectorFromFloat(Value=&Multiply_Float(
      A=1.5
      B=&FloatFromCurve(
       FloatCurve=DataInterface() {propertyValues {Curve {keys [
           {arriveTangent=4.1785, interpMode=RCIM_Cubic, leaveTangent=4.1785, tangentMode=RCTM_User, time=0, value=0}
           {arriveTangent=1.1179, interpMode=RCIM_Cubic, leaveTangent=1.1179, tangentMode=RCTM_User, time=0.1559, value=0.7516}
           {arriveTangent=0.0224, interpMode=RCIM_Cubic, leaveTangent=0.0224, tangentMode=RCTM_User, time=1, value=1}
          ]}}}
       Scale Curve=1.3
      )
     ))
    Write to Module Namespace=true
   )
   /SetVariables(Particles.ShockwaveDynParams:Vector4f=&MakeVector4(
     X=&Lerp_Float(
      A=1
      B=0.9
      Alpha=&FloatFromCurve(FloatCurve=DataInterface() {propertyValues {Curve {keys [
           {arriveTangent=9.0725, interpMode=RCIM_Cubic, leaveTangent=9.0725, tangentMode=RCTM_User, time=0, value=0}
           {arriveTangent=-2.5249, interpMode=RCIM_Cubic, leaveTangent=-2.5249, tangentMode=RCTM_User, time=0.1408, value=0.9847}
           {arriveTangent=-0.0405, interpMode=RCIM_Cubic, leaveTangent=-0.0405, tangentMode=RCTM_User, time=1, value=0}
          ]}}})
     )
     Y=0
     Z=0
     W=0
    )) {isSetParameters=true}
   &FN_Mesh_Attributes(
    Create Scale Attribute=true
    Scale=&VectorFromFloat(Value=&FloatFromCurve(
      FloatCurve=DataInterface() {propertyValues {Curve {keys [
          {arriveTangent=4.1785, interpMode=RCIM_Cubic, leaveTangent=4.1785, tangentMode=RCTM_User, time=0, value=0.8}
          {arriveTangent=0.0224, interpMode=RCIM_Cubic, leaveTangent=0.0224, tangentMode=RCTM_User, time=1, value=2}
         ]}}}
      Scale Curve=0.5
     ))
    Write to Module Namespace=true
   ) {module=FN_Mesh_Attributes001}
  ]
  rendererClasses [&NiagaraMeshRendererProperties]
  renderers [
   {rendererClass=&NiagaraMeshRendererProperties, rendererIndex=0}
   {rendererClass=&NiagaraMeshRendererProperties, rendererIndex=1}
   {rendererClass=&NiagaraMeshRendererProperties, rendererIndex=2}
   {rendererClass=&NiagaraMeshRendererProperties, rendererIndex=3}
  ]
 }
 Smoke_Sphere {
  emitterUpdateScript [
   &EmitterState(Life Cycle Mode=::NewEnumerator1, Loop Behavior=::NewEnumerator1, Loop Delay=0.05, Loop Duration=0.1, UseLoopDelay=true)
   &SpawnBurst_Instantaneous(Spawn Count=12, Use Spawn Probability=false)
  ]
  particleSpawnScript [
   &InitializeParticle(
    Color=<rgba 0.224 0.224 0.224 0.5>
    Lifetime=&RandomRangeFloat(Minimum=1, Maximum=2)
    Sprite Size Mode=::NewEnumerator1
    Uniform Sprite Size Max=800
    Uniform Sprite Size Min=400
    UsePositionOffset=false
    Write Material Random=true
   )
   /SetVariables(Particles.SpriteSize=&Multiply_Vector2D_ByFloat(
     Float=&FN_LOD_Distance_Remap(Start Distance=5000, End Distance=25000, Minimum Output=1, Maximum Output=3)
     Vector2D=&RandomRangeVector2D(
      Minimum=&Vector2DFromFloat(Value=600)
      Maximum=&Vector2DFromFloat(Value=1000)
     )
    )) {isSetParameters=true}
   /SetVariables(Particles.InitSpriteRotRate=&UniformRangedFloat(Minimum=-3, Maximum=3)) {isSetParameters=true}
   &SphereLocation(
    MaskSpawnGroup=false
    Non Uniform Scale=<1 1 0.25>
    Override Seed=true
    Random Seed=20
    Sphere Radius=500
    Surface Only=false
    UseNonUniformScale=true
    UseOffset=false
    UseSphereOrientation=false
   )
   /SetVariables(Particles.Position=&MaskVectorBySpawnGroup(
     Masked Vector=&Add_Vector(A=Linked() {linkedVariable {name=Engine.Owner.Position, type=NiagaraPosition}}, B=<0 0 10>)
     Passthrough Vector=Linked() {linkedVariable {name=Particles.Position, type=NiagaraPosition}}
     Spawn Group Mask=1
    )) {isSetParameters=true}
   &AddVelocityFromPoint(Module.Use Falloff Distance=false, Velocity Strength=&RandomRangeFloat(Minimum=800, Maximum=50000))
  ]
  particleUpdateScript [
   &ParticleState()
   &SpriteRotationRate(Rotation Rate=&Multiply_Float(
     A=$Particles.InitSpriteRotRate
     B=&FloatFromCurve(
      FloatCurve=DataInterface() {propertyValues {Curve {keys [
          {arriveTangent=-4.0107, interpMode=RCIM_Cubic, leaveTangent=-4.0107, tangentMode=RCTM_User, time=-0.0046, value=1.0113}
          {arriveTangent=-0.2496, interpMode=RCIM_Cubic, leaveTangent=-0.2496, tangentMode=RCTM_User, time=0.5, value=0.16}
         ]}}}
      CurveIndex=$Emitter.Age
      Scale Curve=150
     )
    ))
   &ScaleColor(
    Scale Alpha=&FloatFromCurve(FloatCurve=DataInterface() {propertyValues {Curve {keys [
         <at 0.1 0.75>
         {interpMode=RCIM_Cubic, tangentMode=RCTM_User, time=0.2, value=2.25}
         {interpMode=RCIM_Cubic, tangentMode=RCTM_User, time=0.7612, value=1.5}
         <at 1 0>
        ]}}})
    Scale RGB=&VectorFromFloat(Value=&FloatFromCurve(FloatCurve=DataInterface() {propertyValues {Curve {keys [{arriveTangent=1.8145, interpMode=RCIM_Cubic, leaveTangent=0.4403, tangentMode=RCTM_Break, time=0, value=0.4}, <at 0.4 0.625>, <at 1 1>]}}}))
    ScaleA=true
    ScaleRGB=true
   )
   /Niagara/Modules/Update/Forces/CurlNoiseForce(Noise Frequency=0.0012, Noise Strength=800, Pan Noise=true, Pan Noise Field=<0 0 1>)
   &Drag(Drag=20, Use Linear Drag=true, Use Rotational Drag=false)
   &SolveForcesAndVelocity(Clamp Velocity=false, Limit Acceleration=false)
   &ScaleSpriteSize(Scale Factor=&Vector2DFromFloat(Value=&FloatFromCurve(
      FloatCurve=DataInterface() {propertyValues {Curve {keys [
          {arriveTangent=6.2202, interpMode=RCIM_Cubic, leaveTangent=6.2202, tangentMode=RCTM_User, time=0, value=0.15}
          {arriveTangent=1.0025, interpMode=RCIM_Cubic, leaveTangent=1.0025, tangentMode=RCTM_User, time=0.2, value=0.8}
          {arriveTangent=0.0143, interpMode=RCIM_Cubic, leaveTangent=0.0143, tangentMode=RCTM_User, time=1.0463, value=1}
         ]}}}
      Scale Curve=3
     )))
   &DynamicMaterialParameters(
    Index 0 Param 1=&Add_Float(
     A=$Particles.RandomFloat
     B=&FloatFromCurve(
      FloatCurve=DataInterface() {propertyValues {Curve {keys [
          {arriveTangent=2.7533, interpMode=RCIM_Cubic, leaveTangent=1.8958, tangentMode=RCTM_User, time=0, value=1}
          {arriveTangent=0.8292, interpMode=RCIM_Cubic, leaveTangent=0.8292, tangentMode=RCTM_User, time=0.321, value=1.3453}
          {arriveTangent=0.0529, interpMode=RCIM_Cubic, leaveTangent=0.2021, tangentMode=RCTM_User, time=1, value=1.5}
         ]}}}
      CurveIndex=&Add_Float(A=&Multiply_Float(A=$Particles.Initial.MaterialRandom, B=0.5), B=$Particles.NormalizedAge)
     )
    )
    Index 0 Param 2=&FloatFromCurve(FloatCurve=DataInterface() {propertyValues {Curve {keys [
         {arriveTangent=0.0453, interpMode=RCIM_Cubic, tangentMode=RCTM_Break, time=0, value=-1}
         {arriveTangent=0.0604, interpMode=RCIM_Cubic, leaveTangent=0.0604, tangentMode=RCTM_User, time=1, value=-1}
        ]}}})
    Index 0 Param 3=&FloatFromCurve(FloatCurve=DataInterface() {propertyValues {Curve {keys [
         {arriveTangent=-53.9199, interpMode=RCIM_Cubic, leaveTangent=-8.9727, tangentMode=RCTM_User, time=0, value=1}
         {arriveTangent=0.2347, interpMode=RCIM_Cubic, leaveTangent=0.2723, tangentMode=RCTM_Break, time=0.25, value=0}
        ]}}})
    Index 0 Param 4=&FloatFromCurve(FloatCurve=DataInterface() {propertyValues {Curve {keys [<at 0.5 1>,<at 1 0>]}}})
    Index 1 Param 1=&FloatFromCurve(
     FloatCurve=DataInterface() {propertyValues {Curve {keys [
         {interpMode=RCIM_Cubic, leaveTangent=4.7088, tangentMode=RCTM_Break, time=0, value=0}
         {arriveTangent=4.2257, interpMode=RCIM_Cubic, leaveTangent=4.2257, tangentMode=RCTM_User, time=0.25, value=1}
        ]}}}
     Scale Curve=&RandomRangeFloat(Minimum=1.5, Maximum=1.5)
    )
    Index 1 Param 2=&FloatFromCurve(
     FloatCurve=DataInterface() {propertyValues {Curve {keys [<at 0 0.2>, {leaveTangent=0.9066, time=1, value=2}]}}}
     Scale Curve=&RandomRangeFloat(Minimum=0.5, Maximum=0.6)
    )
    Index 1 Param 3=&FloatFromCurve(FloatCurve=DataInterface() {propertyValues {Curve {keys [
         {interpMode=RCIM_Cubic, leaveTangent=6.3917, tangentMode=RCTM_Break, time=0, value=0.2}
         {arriveTangent=1.7549, interpMode=RCIM_Cubic, leaveTangent=10.6767, tangentMode=RCTM_User, time=0.25, value=1}
        ]}}})
    Index 1 Param 4=&FloatFromCurve(
     FloatCurve=DataInterface() {propertyValues {Curve {keys [{arriveTangent=-0.136, interpMode=RCIM_Cubic, leaveTangent=-0.136, tangentMode=RCTM_User, time=0, value=1}, <at 1 0.95>]}}}
     Scale Curve=&RandomRangeFloat(Minimum=1.1, Maximum=1.4)
    )
    Index 2 Param 1=&FloatFromCurve(FloatCurve=DataInterface() {propertyValues {Curve {keys [
         {interpMode=RCIM_Cubic, leaveTangent=4.2485, tangentMode=RCTM_Break, time=0, value=0}
         {arriveTangent=1.0548, interpMode=RCIM_Cubic, leaveTangent=1.0548, tangentMode=RCTM_User, time=0.17, value=0.5715}
         <at 1 1>
        ]}}})
    Index 2 Param 2=&FloatFromCurve(
     FloatCurve=DataInterface() {propertyValues {Curve {keys [
         {arriveTangent=1.9056, interpMode=RCIM_Cubic, leaveTangent=1.9056, tangentMode=RCTM_User, time=0, value=0.2}
         {arriveTangent=0.0259, interpMode=RCIM_Cubic, leaveTangent=0.0259, tangentMode=RCTM_User, time=1, value=1.4}
        ]}}}
     Scale Curve=&RandomRangeFloat(Minimum=0.2, Maximum=0.7)
    )
    Index 2 Param 3=&FloatFromCurve(FloatCurve=DataInterface() {propertyValues {Curve {keys [
         {interpMode=RCIM_Cubic, leaveTangent=2.3064, tangentMode=RCTM_Break, time=0, value=1}
         {interpMode=RCIM_Cubic, leaveTangent=11.1233, tangentMode=RCTM_Break, time=1, value=2}
        ]}}})
    Index 2 Param 4=&FloatFromCurve(
     FloatCurve=DataInterface() {propertyValues {Curve {keys [
         {arriveTangent=-0.6867, interpMode=RCIM_Cubic, leaveTangent=-0.6867, tangentMode=RCTM_User, time=0, value=1}
         {interpMode=RCIM_Cubic, leaveTangent=-1.2956, tangentMode=RCTM_Break, time=1, value=0.75}
        ]}}}
     Scale Curve=1.6
    )
    Module.Param0WriteEnabled=true
    Module.Param0WriteEnabled1=true
    Module.Param0WriteEnabled2=true
    Module.Param1WriteEnabled=false
    Module.Param1WriteEnabled1=true
    Module.Param1WriteEnabled2=true
    Module.Param2WriteEnabled=true
    Module.Param2WriteEnabled1=true
    Module.Param2WriteEnabled2=true
    Module.Param3WriteEnabled=true
    Module.Param3WriteEnabled1=true
    Module.Param3WriteEnabled2=true
    Write Parameter Index 1=true
    Write Parameter Index 2=true
   )
   &CameraOffset(Camera Offset Amount=35) {enabled=false}
  ]
  rendererClasses [&NiagaraSpriteRendererProperties]
  renderers [
   {rendererClass=&NiagaraSpriteRendererProperties}
   {rendererClass=&NiagaraSpriteRendererProperties, rendererIndex=1}
  ]
 }
 WindPlanes_Shockwave {
  emitterUpdateScript [
   &EmitterState(Life Cycle Mode=::NewEnumerator1, Loop Behavior=::NewEnumerator1, Loop Duration=1, Scalability Mode=::NewEnumerator1, UseLoopDelay=false)
   &SpawnBurst_Instantaneous(Spawn Group=1, Use Spawn Probability=false, UseLoopCountLimit=false) {module=SpawnBurst_Instantaneous002}
   &SpawnBurst_Instantaneous(Use Spawn Probability=false, UseLoopCountLimit=false) {enabled=false}
   &SpawnBurst_Instantaneous(Use Spawn Probability=false, UseLoopCountLimit=false) {module=SpawnBurst_Instantaneous001}
  ]
  particleSpawnScript [
   /SetVariables(Particles.SpawnGroup=$Emitter.SpawnGroup) {isSetParameters=true}
   /SetVariables(Particles.VisibilityTag=&MakeCustomIntFromBool(Bool=$User.IsWater, True INT=$Particles.SpawnGroup, False INT=0)) {isSetParameters=true}
   &InitializeParticle(
    Color=<rgba 1 0.3773 0.1084 1>
    Lifetime=&MergeHelper_FloatSwitch(0=0.758, 1=3, 2=0, 3=0, 4=0, Index=$Particles.SpawnGroup, Num=::NewEnumerator3)
    Mesh Scale=<6 6 11>
    Mesh Scale Mode=::NewEnumerator4
    Position Offset=&AddVectorToPosition(Position=<0 0 -30>, Vector=$System.FXPosition)
    Position Offset Coordinate Space=::World
    UsePositionOffset=true
    Write Material Random=true
   )
   /SetVariables(Particles.Lifetime=&MergeHelper_FloatSwitch(0=0.758, 1=1.5, 2=0, 3=0, 4=0, Index=$Particles.SpawnGroup, Num=::NewEnumerator3)) {isSetParameters=true}
   /SetVariables(
    Particles.Mesh_01_Color:LinearColor=&MergeHelper_ColorSwitch(0=<rgba 1 1 1 2>, 1=<rgba 1 1 1 2>, 2=?, 3=?, Index=$System.ImpactType, Num=::NewEnumerator0)
    Particles.Mesh_01_Orientation:Quat4f=&OrientMeshToVector(Look At Direction=<1 0 0>, Reference Axis Vector=<1 0 0>)
    Particles.Mesh_01_Position:NiagaraPosition=<0 0 -30>
    Particles.Mesh_01_Size:Vector3f=&VectorFromFloat(Value=40)
   ) {isSetParameters=true}
   &CameraOffset(Camera Offset Amount=100)
   &InitialMeshOrientation(Mesh Orientation Mode=::NewEnumerator2, Rotation=<0.5 0 0>, Use Rotation Vector=true)
  ]
  particleUpdateScript [
   &ParticleState()
   &ScaleMeshSize(Scale Factor=&MakeVector(
     X=&FloatFromCurve(
      FloatCurve=DataInterface() {propertyValues {Curve {keys [
          {arriveTangent=3.3046, interpMode=RCIM_Cubic, leaveTangent=3.3046, tangentMode=RCTM_User, time=0, value=0}
          {arriveTangent=1, interpMode=RCIM_Cubic, leaveTangent=1, time=0.2, value=0.45}
          {arriveTangent=1.0813, interpMode=RCIM_Cubic, leaveTangent=1.0813, tangentMode=RCTM_User, time=1, value=1}
         ]}}}
      Scale Curve=1.5
     )
     Y=&FloatFromCurve(
      FloatCurve=DataInterface() {propertyValues {Curve {keys [
          {arriveTangent=3.3046, interpMode=RCIM_Cubic, leaveTangent=3.3046, tangentMode=RCTM_User, time=0, value=0}
          {arriveTangent=1, interpMode=RCIM_Cubic, leaveTangent=1, time=0.2, value=0.45}
          {arriveTangent=1.0813, interpMode=RCIM_Cubic, leaveTangent=1.0813, tangentMode=RCTM_User, time=1, value=1}
         ]}}}
      Scale Curve=1.5
     )
     Z=&FloatFromCurve(
      FloatCurve=DataInterface() {propertyValues {Curve {keys [
          {interpMode=RCIM_Cubic, tangentMode=RCTM_User, time=0, value=0}
          {arriveTangent=2.0031, interpMode=RCIM_Cubic, leaveTangent=2.0031, tangentMode=RCTM_User, time=1, value=1}
         ]}}}
      Scale Curve=2
     )
    )) {enabled=false}
   &ScaleColor(
    Linear Color Curve=DataInterface() {propertyValues {
      AlphaCurve {keys [<at 0 1>,<at 0.4544 0.5>,<at 0.9945 0>]}
      BlueCurve {keys [<at 0 1>,<at 1 1>]}
      GreenCurve {keys [<at 0 1>,<at 1 1>]}
      RedCurve {keys [<at 0 1>,<at 1 1>]}
     }}
    Scale Mode=::NewEnumerator2
   )
   &GravityForce(Gravity=&Multiply_VectorByFloat(
     Vector=<0 0 980>
     Float=&MergeHelper_FloatSwitch(0=1, 1=0, 2=0, 3=0, 4=0, Index=$Particles.SpawnGroup, Num=::NewEnumerator3)
    ))
   /SetVariables(
    Particles.Mesh_01_Color:LinearColor=&ScaleLinearColorByCurve(
     InitialValue=$Particles.Initial.Mesh_01_Color
     ScaleCurve=DataInterface() {propertyValues {
       AlphaCurve {keys [{arriveTangent=0.25, interpMode=RCIM_Cubic, time=0, value=0}, <at 0.3 0.15>, {interpMode=RCIM_Cubic, leaveTangent=-0.1071, time=1, value=0}]}
       BlueCurve {keys [<at 0 0.9263 RCIM_Cubic>]}
       GreenCurve {keys [<at 0 0.9816 RCIM_Cubic>]}
       RedCurve {keys [<at 0 1 RCIM_Cubic>]}
      }}
     CurveIndex=&Divide_Float(A=$Particles.NormalizedAge, B=0.4)
    )
    Particles.Mesh_01_Size:Vector3f=&Multiply_Vector(
     A=$Particles.Initial.Mesh_01_Size
     B=&VectorFromFloat(Value=&FloatFromCurve(
       FloatCurve=DataInterface() {propertyValues {Curve {keys [
           {arriveTangent=2.1062, interpMode=RCIM_Cubic, leaveTangent=2.1062, tangentMode=RCTM_User, time=0, value=0}
           {arriveTangent=1.0293, interpMode=RCIM_Cubic, leaveTangent=1.0293, tangentMode=RCTM_User, time=1, value=1}
          ]}}}
       CurveIndex=&Divide_Float(A=$Particles.NormalizedAge, B=0.4)
       Scale Curve=0.5
      ))
    )
   ) {isSetParameters=true}
   &SolveForcesAndVelocity(Clamp Velocity=false, Limit Acceleration=false)
   &ScaleMeshSize(Scale Factor=&Multiply_VectorByFloat(
     Vector=<1 1 1>
     Float=&FloatFromCurve(
      FloatCurve=DataInterface() {propertyValues {Curve {keys [
          {arriveTangent=7.2197, interpMode=RCIM_Cubic, leaveTangent=7.2197, tangentMode=RCTM_User, time=0, value=0}
          {arriveTangent=1.9952, interpMode=RCIM_Cubic, leaveTangent=1.9952, time=0.4917, value=1.9432}
          {arriveTangent=-0.0405, interpMode=RCIM_Cubic, leaveTangent=-0.0405, tangentMode=RCTM_User, time=1.0024, value=2}
         ]}}}
      CurveIndex=&Divide_Float(
       A=$Particles.NormalizedAge
       B=&SelectFloatFromArray(Direct Array Index=$Particles.SpawnGroup, Float Selection Array=DataInterface() {propertyValues {FloatData [0.9,1.25]}})
      )
      Scale Curve=&SelectFloatFromArray(Direct Array Index=$Particles.SpawnGroup, Float Selection Array=DataInterface() {propertyValues {FloatData [1,0.5]}})
     )
    )) {module=ScaleMeshSize001}
   &DynamicMaterialParameters(
    Index 0 Param 1=&RandomRangeFloat(Minimum=-1, Maximum=1)
    Index 0 Param 2=&FloatFromCurve(FloatCurve=DataInterface() {propertyValues {Curve {keys [<at 0 0.33>,<at 1 1.25>]}}})
    Index 0 Param 3=&FloatFromCurve(FloatCurve=DataInterface() {propertyValues {Curve {keys [{tangentMode=RCTM_User, time=0, value=1}, {tangentMode=RCTM_User, time=1, value=0}]}}})
    Index 1 Param 1=&MakeFloatFromVector(Vector=Linked() {linkedVariable {name=Particles.Position, type=NiagaraPosition}}, Channel=::NewEnumerator0)
    Index 1 Param 2=&MakeFloatFromVector(Vector=Linked() {linkedVariable {name=Particles.Position, type=NiagaraPosition}}, Channel=::NewEnumerator1)
    Index 1 Param 3=&FloatFromCurve(FloatCurve=DataInterface() {propertyValues {Curve {keys [<at 0 1 RCIM_Cubic>]}}}, Scale Curve=200)
    Index 2 Param 1=&FloatFromCurve(
     FloatCurve=DataInterface() {propertyValues {Curve {keys [
         <at 0 1 RCIM_Cubic>
         {arriveTangent=-9.0783, interpMode=RCIM_Cubic, leaveTangent=1.3508, tangentMode=RCTM_Break, time=0.15, value=0}
         {tangentMode=RCTM_User, time=1, value=1}
        ]}}}
     Scale Curve=1.75
    )
    Index 2 Param 2=&FloatFromCurve(FloatCurve=DataInterface() {propertyValues {Curve {keys [
         <at 0 0.5 RCIM_Cubic>
         {arriveTangent=-3.91, interpMode=RCIM_Cubic, leaveTangent=-3.91, tangentMode=RCTM_User, time=0.15, value=2}
         <at 1 0.5 RCIM_Cubic>
        ]}}})
    Index 2 Param 3=2
    Index 2 Param 4=&FloatFromCurve(FloatCurve=DataInterface() {propertyValues {Curve {keys [<at 0 0.1>,<at 0.15 1 RCIM_Cubic>,<at 1 0.1 RCIM_Cubic>]}}}, Scale Curve=0.5)
    Module.Param0WriteEnabled=true
    Module.Param0WriteEnabled1=true
    Module.Param0WriteEnabled2=true
    Module.Param1WriteEnabled=true
    Module.Param1WriteEnabled1=true
    Module.Param1WriteEnabled2=true
    Module.Param2WriteEnabled=true
    Module.Param2WriteEnabled1=true
    Module.Param2WriteEnabled2=true
    Module.Param3WriteEnabled=true
    Module.Param3WriteEnabled1=false
    Module.Param3WriteEnabled2=true
    Write Parameter Index 1=true
    Write Parameter Index 2=true
   ) {module=DynamicMaterialParameters001}
   &UpdateMeshOrientation(
    Rotation Rate=&FloatFromCurve(
     FloatCurve=DataInterface() {propertyValues {Curve {keys [{tangentMode=RCTM_User, time=0, value=-1}, <at 0.25 1>, <at 0.5 -1>, <at 0.75 1>, {tangentMode=RCTM_User, time=1, value=-1}]}}}
     CurveIndex=&ReturnNormalizedExecIndex(Normalized Index Includes One=false)
     Scale Curve=&SelectFloatFromArray(Direct Array Index=$Particles.SpawnGroup, Float Selection Array=DataInterface() {propertyValues {FloatData [-1,1]}})
    )
    Rotation Vector=<0 0 1>
   )
   &ScaleMeshSize(Scale Factor=&MakeVector(
     X=1
     Y=1
     Z=&SelectFloatFromArray(Direct Array Index=$Particles.SpawnGroup, Float Selection Array=DataInterface() {propertyValues {FloatData [2.5,1.5]}})
    )) {module=ScaleMeshSize002}
   &ScaleColor(
    Scale Alpha=&FloatFromCurve(
     FloatCurve=DataInterface() {propertyValues {Curve {keys [
         {arriveTangent=6.2301, interpMode=RCIM_Cubic, leaveTangent=6.2301, tangentMode=RCTM_User, time=0, value=0.5}
         {arriveTangent=-0.7357, interpMode=RCIM_Cubic, leaveTangent=-0.7357, tangentMode=RCTM_User, time=0.1, value=2}
         <at 0.4 0.2>
         {arriveTangent=0.4048, interpMode=RCIM_Cubic, leaveTangent=0.4048, tangentMode=RCTM_User, time=0.7, value=0}
        ]}}}
     Scale Curve=0.5
    )
    ScaleA=true
    ScaleRGB=true
   ) {module=ScaleColor001}
   &FN_Mesh_Attributes(
    Color=&MultiplyLinearColor(
     Linear Color=<rgba 1 1 1 1>
     Scale Factor=&MakeFloat4FromVectorAndFloat(
      "Vector (RGB)"=<1 1 1>
      "Float (Alpha)"=&FloatFromCurve(FloatCurve=DataInterface() {propertyValues {Curve {keys [<at 0 1>,<at 1 0>]}}})
     )
    )
    Create Color Attribute=true
    Create Orientation Attribute=true
    Create Position Attribute=true
    Create Scale Attribute=true
    Position:NiagaraPosition=$User.WaterPlaneLocation
    Scale=&VectorFromFloat(Value=&Multiply_Float(
      A=2000
      B=&FloatFromCurve(
       FloatCurve=DataInterface() {propertyValues {Curve {keys [
           {arriveTangent=4.1785, interpMode=RCIM_Cubic, leaveTangent=4.1785, tangentMode=RCTM_User, time=0, value=0}
           {arriveTangent=1.1179, interpMode=RCIM_Cubic, leaveTangent=1.1179, tangentMode=RCTM_User, time=0.2042, value=0.6228}
           {arriveTangent=0.0224, interpMode=RCIM_Cubic, leaveTangent=0.0224, tangentMode=RCTM_User, time=1, value=1}
          ]}}}
       Scale Curve=4
      )
     ))
    Write to Module Namespace=true
   )
   /SetVariables(Particles.WaterPosition:NiagaraPosition=&TransformPosition(Position=$User.WaterPlaneLocation, Source Space=::World, Destination Space=::Simulation)) {isSetParameters=true}
   &CameraOffset(Camera Offset Amount=&MakeCustomFloatFromBool(Bool=$User.IsWater, True Float=22, False Float=0)) {enabled=false, module=CameraOffset001}
  ]
  rendererClasses [&NiagaraMeshRendererProperties]
  renderers [{rendererClass=&NiagaraMeshRendererProperties, rendererIndex=0}]
 }
}
summary {
 systemName=NS_TentacleAttackImpact
 emitters [
  {emitterName=Bolt, rendererClasses [&NiagaraMeshRendererProperties]}
  {
   emitterName=Light001
   rendererClasses [
    &NiagaraSpriteRendererProperties
    &NiagaraLightRendererProperties
    &NiagaraComponentRendererProperties
   ]
  }
  {emitterName=Empty003, rendererClasses [&NiagaraMeshRendererProperties]}
  {emitterName=Radial_Plasma_Energy, rendererClasses [&NiagaraSpriteRendererProperties]}
  {emitterName=Shimmers001, simTarget=GPUComputeSim, rendererClasses [&NiagaraMeshRendererProperties]}
  {emitterName=Smoke_Sphere, rendererClasses [&NiagaraSpriteRendererProperties]}
  {bEnabled=false, emitterName=Shockwave_Ground, rendererClasses [&NiagaraMeshRendererProperties]}
  {emitterName=WindPlanes_Shockwave, rendererClasses [&NiagaraMeshRendererProperties]}
  {emitterName=Comet_Impact, rendererClasses [&NiagaraMeshRendererProperties]}
  {emitterName=Dirt, simTarget=GPUComputeSim, rendererClasses [&NiagaraSpriteRendererProperties]}
  {emitterName=Lightning_Primary001, rendererClasses [&NiagaraMeshRendererProperties]}
  {bEnabled=false, emitterName=BlastPostProcess, rendererClasses [&NiagaraComponentRendererProperties]}
 ]
 userVariables [
  User.Color {type=LinearColor, description="", defaultValue {struct=LinearColor, value=<rgba 0 0.161 0.5 1>}}
  User.ColorMult {type=NiagaraFloat, defaultValue=1, description=""}
  User.ElementSizeScale {type=NiagaraFloat, defaultValue=1, description=""}
  User.IsWater {type=NiagaraBool}
  User.LightningRadius {type=NiagaraFloat, defaultValue=1500, description=""}
  User.PostNoiseStrength {type=NiagaraFloat, defaultValue=1, description=""}
  User.PostSizeScale {type=NiagaraFloat, defaultValue=1, description=""}
  User.PostStrength {type=NiagaraFloat, defaultValue=1, description=""}
  User.PrimaryColor {type=LinearColor, description="", defaultValue {struct=LinearColor, value=<rgba 0.755 3.9375 5 1>}}
  User.Scaling_SpawnRadius {type=NiagaraFloat, defaultValue=1, description=""}
  User.SecondaryColor {type=LinearColor, description="", defaultValue {struct=LinearColor, value=<rgba 1 0.05 0.1292 3>}}
  User.WaterPlaneLocation {type=NiagaraPosition, description="", defaultValue {}}
  User.WidthMult {type=NiagaraFloat, defaultValue=1, description=""}
 ]
}
