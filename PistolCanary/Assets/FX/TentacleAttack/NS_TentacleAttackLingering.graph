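# Niagara system for the lingering area effect of a tentacle attack: a wide, flattened field of electrical plasma.
# Mesh lightning, arc and fire sprites, and GPU shimmers layer over a glow mesh, flickering light and expanding radial decal.
# Arc and fire spawn rates fall with LOD distance, while shimmers use distance culling.
# User parameters scale lightning width and brightness, fire radius and light radius; Color, Offset and GlowRadius appear unused.
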NiagaraSystem /PistolCanary/Assets/FX/TentacleAttack/NS_TentacleAttackLingering
dependencies {
 usedDataInterfaces [
  &NiagaraDataInterfaceCurve
  &NiagaraDataInterfaceArrayColor
  &NiagaraDataInterfaceArrayFloat
  &NiagaraDataInterfaceColorCurve
  &NiagaraDataInterfaceSkeletalMesh
 ]
 usedDynamicInputs []
 usedModules [
  &SystemState
  &self:SystemSpawnScript.SystemScriptSource.SystemScriptGraph.NiagaraNodeAssignment_1.SetVariables_#1
 ]
}
emitters {
 Arcs {
  emitterUpdateScript [
   &EmitterState()
   &SpawnRate(
    SpawnRate=&Lerp_Float(
     A=25
     B=6
     Alpha=&FN_LOD_Distance_Remap(Start Distance=2250, End Distance=30000, Minimum Output=0, Maximum Output=1)
    )
    Use Spawn Probability=false
   )
  ]
  particleSpawnScript [
   &InitializeParticle(
    Lifetime Max=0.4
    Lifetime Min=0.35
    Lifetime Mode=::NewEnumerator1
    Sprite Rotation Mode=::NewEnumerator2
    Sprite Size Mode=::NewEnumerator1
    Uniform Sprite Size Max=400
    Uniform Sprite Size Min=200
    UsePositionOffset=false
    Write Material Random=true
   )
   &SubUVAnimation(End Frame=3, SubUV Animation Mode:ENiagara_SubUVLookupMode=::NewEnumerator2)
   &SphereLocation(
    MaskSpawnGroup=false
    Non Uniform Scale=<1 1 0.25>
    Override Seed=false
    Sphere Radius=1000
    Surface Only=false
    UseNonUniformScale=true
    UseOffset=false
    UseSphereOrientation=false
   )
  ]
  particleUpdateScript [
   &ParticleState()
   &CameraOffset(Camera Offset Amount=25)
   &ScaleColor(
    Scale RGB=&MakeVectorFromLinearColorRGB(Color=&LerpLinearColors(
      StartColor=&RandomRangeLinearColor(
       Minimum=&ColorFromCurve(ColorCurve=DataInterface() {propertyValues {
          AlphaCurve {keys [<at 0.5586 1 RCIM_Cubic>,<at 1 0 RCIM_Cubic>]}
          BlueCurve {keys [<at 0 1 RCIM_Cubic>,<at 1 1 RCIM_Cubic>]}
          GreenCurve {keys [
            {arriveTangent=-0.0536, interpMode=RCIM_Cubic, leaveTangent=-0.0536, time=0, value=0.2093}
            {arriveTangent=-0.0536, interpMode=RCIM_Cubic, leaveTangent=-0.0536, time=1, value=0.1557}
           ]}
          RedCurve {keys [
            {arriveTangent=0.0676, interpMode=RCIM_Cubic, leaveTangent=0.0676, time=0, value=0.0396}
            {arriveTangent=0.0676, interpMode=RCIM_Cubic, leaveTangent=0.0676, time=1, value=0.1073}
           ]}
         }})
       Maximum=&ColorFromCurve(ColorCurve=DataInterface() {propertyValues {
          AlphaCurve {keys [<at 0.5586 1 RCIM_Cubic>,<at 1 0 RCIM_Cubic>]}
          BlueCurve {keys [<at 0 1 RCIM_Cubic>,<at 1 1 RCIM_Cubic>]}
          GreenCurve {keys [{arriveTangent=-0.0536, interpMode=RCIM_Cubic, time=0, value=0.1868}, {interpMode=RCIM_Cubic, leaveTangent=-0.0536, time=1, value=0.2107}]}
          RedCurve {keys [{arriveTangent=0.0676, interpMode=RCIM_Cubic, time=0, value=0.8161}, {interpMode=RCIM_Cubic, leaveTangent=0.0676, time=1, value=0.9236}]}
         }})
      )
      EndColor=&ColorFromCurve(ColorCurve=DataInterface() {propertyValues {
         AlphaCurve {keys [<at 0.5586 1 RCIM_Cubic>,<at 1.0091 0 RCIM_Cubic>]}
         BlueCurve {keys [<at 0 1.2408 RCIM_Cubic>,<at 0.982 0>]}
         GreenCurve {keys [<at 0 1.7901 RCIM_Cubic>,<at 0.982 0.2738>]}
         RedCurve {keys [<at 0 5 RCIM_Cubic>,<at 0.982 5>]}
        }})
      LerpFactor=&SelectFloatFromArray(Direct Array Index=0, Float Selection Array=DataInterface() {propertyValues {FloatData [0,0,0,1]}})
     ))
    ScaleA=false
    ScaleRGB=true
   )
   &ScaleColor(
    Scale RGB=&VectorFromFloat(Value=&FloatFromCurve(FloatCurve=DataInterface() {propertyValues {Curve {keys [
          {interpMode=RCIM_Cubic, tangentMode=RCTM_User, time=0, value=1}
          {arriveTangent=106.638, interpMode=RCIM_Cubic, leaveTangent=-47.9454, tangentMode=RCTM_Break, time=0.2, value=10}
          {interpMode=RCIM_Cubic, tangentMode=RCTM_User, time=0.8, value=1}
         ]}}}))
    ScaleA=false
    ScaleRGB=true
   ) {module=ScaleColor001}
   &ScaleSpriteSize(Scale Factor=&Vector2DFromFloat(Value=&FloatFromCurve(
      FloatCurve=DataInterface() {propertyValues {Curve {keys [
          {arriveTangent=-0.0604, interpMode=RCIM_Cubic, leaveTangent=-0.0604, tangentMode=RCTM_User, time=0, value=0.6}
          {arriveTangent=1.6386, interpMode=RCIM_Cubic, leaveTangent=1.6386, tangentMode=RCTM_User, time=1, value=1}
         ]}}}
      Scale Curve=&SelectFloatFromArray(Direct Array Index=0, Float Selection Array=DataInterface() {propertyValues {FloatData [1,1,1,1.5]}})
     )))
   &SpriteRotationRate(Rotation Rate=25)
   &ScaleSpriteSize(
    Scale Factor=&MakeVector2D(
     X=&FloatFromCurve(FloatCurve=DataInterface() {propertyValues {Curve {keys [
          {arriveTangent=-3.4808, interpMode=RCIM_Cubic, leaveTangent=-3.4808, tangentMode=RCTM_User, time=0.4, value=1}
          {arriveTangent=0.0049, interpMode=RCIM_Cubic, leaveTangent=0.0049, tangentMode=RCTM_User, time=0.8, value=0.5}
         ]}}})
     Y=&FloatFromCurve(FloatCurve=DataInterface() {propertyValues {Curve {keys [
          {arriveTangent=0.0124, interpMode=RCIM_Cubic, leaveTangent=0.0124, tangentMode=RCTM_User, time=0.4, value=0.85}
          {arriveTangent=2.3888, interpMode=RCIM_Cubic, leaveTangent=2.3888, tangentMode=RCTM_User, time=0.8, value=1.2}
         ]}}})
    )
    Scale Sprite Size Mode=::NewEnumerator2
   ) {module=ScaleSpriteSize001}
   &DynamicMaterialParameters(
    Index 0 Param 1=1
    Index 0 Param 2=1
    Index 0 Param 3=&FloatFromCurve(FloatCurve=DataInterface() {propertyValues {Curve {keys [
         {arriveTangent=0.1411, interpMode=RCIM_Cubic, leaveTangent=0.1411, tangentMode=RCTM_User, time=0.2, value=0}
         {arriveTangent=1.8523, interpMode=RCIM_Cubic, leaveTangent=1.8523, tangentMode=RCTM_User, time=1, value=0.5}
        ]}}})
    Module.Param0WriteEnabled=true
    Module.Param1WriteEnabled=true
    Module.Param2WriteEnabled=true
    Module.Param3WriteEnabled=false
   )
  ]
  rendererClasses [&NiagaraSpriteRendererProperties]
  renderers [{rendererClass=&NiagaraSpriteRendererProperties}]
 }
 Big_Glow_Mesh {
  emitterUpdateScript [
   &EmitterState()
   &SpawnBurst_Instantaneous(Use Spawn Probability=false, UseLoopCountLimit=false)
  ]
  particleSpawnScript [/Niagara/Modules/Spawn/Initialization/InitializeParticle(
    Color=<rgba 0.071 0.5308 1 0.5>
    Lifetime=2.25
    Mesh Scale=&VectorFromFloat(Value=16)
    Module.Write Color=true
    Module.Write Lifetime=true
    Module.Write Mass=true
    Module.Write Material Random=true
    Module.Write Position=true
    Module.Write Scale=true
    Module.Write SpriteRotation=false
    Module.Write SpriteSize=false
    Position=&SimulationPosition()
    Sprite Size=<25 25>
   )]
  particleUpdateScript [
   &ParticleState()
   &ScaleColor(
    Scale Alpha=0.25
    Scale RGB=&Multiply_VectorByFloat(
     Vector=<0.25 0.25 0.25>
     Float=&FloatFromCurve(FloatCurve=DataInterface() {propertyValues {Curve {keys [
          {arriveTangent=2.2898, interpMode=RCIM_Cubic, leaveTangent=2.2898, time=0, value=0}
          {arriveTangent=0.9444, interpMode=RCIM_Cubic, leaveTangent=0.9444, time=0.4367, value=1}
          <at 0.9 0.85>
          {arriveTangent=-8.3841, interpMode=RCIM_Cubic, leaveTangent=-8.3841, time=1.0011, value=0.002}
         ]}}})
    )
    ScaleA=true
    ScaleRGB=true
   )
  ]
  rendererClasses [&NiagaraMeshRendererProperties]
  renderers [{rendererClass=&NiagaraMeshRendererProperties, rendererIndex=0}]
 }
 Fire {
  emitterUpdateScript [
   &EmitterState()
   &SpawnRate(
    SpawnRate=&Lerp_Float(
     A=25
     B=0
     Alpha=&FN_LOD_Distance_Remap(Start Distance=1500, End Distance=9000, Minimum Output=0, Maximum Output=1)
    )
    Use Spawn Probability=false
   )
  ]
  particleSpawnScript [
   /Niagara/Modules/Spawn/Initialization/V2/InitializeParticle(
    Color=&RandomRangeLinearColor(Minimum=<rgba 0.2773 2.708 10 1>, Maximum=<rgba 6.6189 1.1315 10 1>)
    Lifetime Mode=::NewEnumerator1
    Sprite Rotation Mode=::NewEnumerator2
    Sprite Size Mode=::NewEnumerator1
    Uniform Sprite Size Max=500
    Uniform Sprite Size Min=400
    UsePositionOffset=false
    Write Material Random=true
   )
   /SetVariables(Particles.RandomFrame=&RandomRangeInt(Minimum=0, Maximum=63)) {isSetParameters=true}
   &SphereLocation(
    MaskSpawnGroup=false
    Non Uniform Scale=<1 1 0.1>
    Override Seed=false
    Sphere Radius=&Multiply_Float(A=1800, B=$User.Scaling_SpawnRadius)
    Surface Only=false
    UseNonUniformScale=true
    UseOffset=false
    UseSphereOrientation=false
   )
  ]
  particleUpdateScript [
   &ParticleState()
   &CameraOffset(Camera Offset Amount=25)
   &FN_SubUVAnimation(End Frame=&Add_Integer(A=$Particles.RandomFrame, B=63), Start Frame=$Particles.RandomFrame)
   &ScaleSpriteSize(Scale Factor=&Vector2DFromFloat(Value=&FloatFromCurve(
      FloatCurve=DataInterface() {propertyValues {Curve {keys [
          {arriveTangent=1.5383, interpMode=RCIM_Cubic, leaveTangent=1.5383, tangentMode=RCTM_User, time=0, value=0.475}
          {arriveTangent=0.0897, interpMode=RCIM_Cubic, leaveTangent=0.0897, tangentMode=RCTM_User, time=1, value=1}
         ]}}}
      Scale Curve=&SelectFloatFromArray(Direct Array Index=0, Float Selection Array=DataInterface() {propertyValues {FloatData [1,1,1,2]}})
     )))
   &ScaleColor(
    Scale Alpha=&FloatFromCurve(
     FloatCurve=DataInterface() {propertyValues {Curve {keys [
         {arriveTangent=8.6533, interpMode=RCIM_Cubic, leaveTangent=8.6533, tangentMode=RCTM_User, time=0, value=0}
         {interpMode=RCIM_Cubic, tangentMode=RCTM_User, time=0.2, value=1}
         {arriveTangent=-0.5344, interpMode=RCIM_Cubic, leaveTangent=-0.5344, tangentMode=RCTM_User, time=1, value=0}
        ]}}}
     Scale Curve=0.4
    )
    ScaleA=true
    ScaleRGB=false
   )
   &AccelerationForce()
   &SolveForcesAndVelocity(Clamp Velocity=false, Limit Acceleration=false)
  ]
  rendererClasses [&NiagaraSpriteRendererProperties]
  renderers [{rendererClass=&NiagaraSpriteRendererProperties}]
 }
 Light {
  emitterUpdateScript [
   &EmitterState()
   &SpawnBurst_Instantaneous(Use Spawn Probability=false, UseLoopCountLimit=false)
  ]
  particleSpawnScript [
   &InitializeParticle(
    Color=<rgba 0 2.1453 15 1>
    Lifetime=4
    Sprite Size Mode=::NewEnumerator3
    Uniform Sprite Size=&Multiply_Float(A=600, B=$User.Scaling_LightRadius)
    UsePositionOffset=false
    Write Material Random=true
   )
   &SystemLocation(MaskSpawnGroup=false, Offset=<0 0 500>)
  ]
  particleUpdateScript [
   &ParticleState()
   &FN_Radial_Material(Base Sprite Tile U=2, Base Sprite Tile V=2, Depth Fade=250, Radial Mask Power=3, Radial Mask Strength=0.5, Sprite Extrusion Speed=1)
   &ScaleColor(
    Scale Alpha=&FloatFromCurve(FloatCurve=DataInterface() {propertyValues {Curve {keys [
         {interpMode=RCIM_Cubic, tangentMode=RCTM_User, time=0, value=0}
         {interpMode=RCIM_Cubic, tangentMode=RCTM_User, time=0.5, value=1}
         {tangentMode=RCTM_User, time=1, value=0}
        ]}}})
    Scale RGB=&VectorFromFloat(Value=&FloatFromCurve(
      FloatCurve=DataInterface() {propertyValues {Curve {keys [
          {interpMode=RCIM_Cubic, tangentMode=RCTM_User, time=0, value=0}
          {interpMode=RCIM_Cubic, tangentMode=RCTM_User, time=0.0604, value=0.3273}
          {interpMode=RCIM_Cubic, tangentMode=RCTM_User, time=0.1, value=0.45}
          {interpMode=RCIM_Cubic, tangentMode=RCTM_User, time=0.1517, value=0.4071}
          {interpMode=RCIM_Cubic, tangentMode=RCTM_User, time=0.2, value=0.6}
          {interpMode=RCIM_Cubic, tangentMode=RCTM_User, time=0.3, value=0.55}
          {interpMode=RCIM_Cubic, tangentMode=RCTM_User, time=0.3499, value=0.4364}
          {interpMode=RCIM_Cubic, tangentMode=RCTM_User, time=0.4, value=0.6}
          {interpMode=RCIM_Cubic, tangentMode=RCTM_User, time=0.5, value=0.7}
          <at 0.526 0.626>
          <at 0.5563 0.4522>
          <at 0.6 0.45>
          <at 0.652 0.3925>
          <at 0.7 0.55>
          <at 0.7124 0.5986>
          <at 0.7452 0.5665>
          {interpMode=RCIM_Cubic, tangentMode=RCTM_User, time=0.8, value=0.6}
          {interpMode=RCIM_Cubic, tangentMode=RCTM_User, time=0.9, value=0.375}
          {tangentMode=RCTM_User, time=1, value=0}
         ]}}}
      Scale Curve=5
     ))
    ScaleA=true
    ScaleRGB=true
   )
  ]
  rendererClasses [&NiagaraLightRendererProperties]
  renderers [{rendererClass=&NiagaraLightRendererProperties, rendererIndex=0}]
 }
 Lightning_Primary001 {
  emitterUpdateScript [
   &EmitterState()
   &SpawnRate(SpawnRate=25, Use Spawn Probability=false) {module=SpawnRate001}
  ]
  particleSpawnScript [
   /SetVariables(Particles.isShort:NiagaraFloat=&FloatFromCurve(
     FloatCurve=DataInterface() {propertyValues {Curve {keys [<at 0 0 RCIM_Constant>,<at 0.8 1 RCIM_Constant>]}}}
     CurveIndex=&RandomRangeFloat(Minimum=0, Maximum=1)
    )) {isSetParameters=true}
   &InitializeParticle(
    Color=&SelectLinearColorFromArray(
     Array Sampling Mode=::NewEnumerator1
     Direct Array Index=&RandomRangeInt(Minimum=0, Maximum=1)
     Color Selection Array=DataInterface() {propertyValues {ColorData [<rgba 0 0.161 0.5 1>, <rgba 0.3139 0.0895 0.5 1>]}}
     Random Seed=?
    )
    Lifetime=&Lerp_Float(
     A=&RandomRangeFloat(Minimum=0.5, Maximum=0.8)
     B=&RandomRangeFloat(Minimum=0.3, Maximum=0.4)
     Alpha=$Particles.isShort
    )
    Mesh Scale Mode=::NewEnumerator3
    Mesh Uniform Scale=&MakeFloatFromInt(Int=&Multiply_Int(A=&RandomRangeInt(Minimum=2, Maximum=3), B=4))
    Position Mode=::NewEnumerator1
    Sprite Size Mode=::NewEnumerator4
    UsePositionOffset=false
    Write Material Random=true
   ) {module=InitializeParticle001}
   &DynamicMaterialParameters(
    Index 0 Param 1=1
    Index 0 Param 2=0.15
    Index 0 Param 3=1
    Index 0 Param 4=1
    Module.Param0WriteEnabled=true
    Module.Param1WriteEnabled=true
    Module.Param2WriteEnabled=true
    Module.Param3WriteEnabled=true
   )
   &ShapeLocation(
    MaskSpawnGroup=false
    Non Uniform Scale=<1 1 0>
    Offset=<0 0 -420>
    Offset Mode=::NewEnumerator0
    Override Seed=false
    Sphere Radius=2000
    Sphere Surface Distribution=1
    UseOwnerScale=true
   ) {module=ShapeLocation002}
   &AddVelocity(Module.Use Falloff Distance=false, Override Seed=false, Velocity Mode=::NewEnumerator1, Velocity Speed=0.25)
   &InitialMeshOrientation(
    Mesh Orientation Mode=::NewEnumerator7
    Rotation=&RandomRangeVector(Minimum=<0 0 0>, Maximum=<0 0 0>)
    Use Rotation Vector=true
   ) {module=InitialMeshOrientation002}
  ]
  particleUpdateScript [
   &SolveForcesAndVelocity(Clamp Velocity=false, Limit Acceleration=false)
   &ParticleState()
   &ScaleMeshSize(Scale Factor=&Multiply_VectorByFloat(
     Vector=&MakeVector(
      X=&Multiply_Float(
       A=&FloatFromCurve(FloatCurve=DataInterface() {propertyValues {Curve {keys [{interpMode=RCIM_Cubic, tangentMode=RCTM_User, time=0, value=1}, {interpMode=RCIM_Cubic, tangentMode=RCTM_User, time=1, value=1.5}]}}})
       B=&RandomRangeFloat(Minimum=0.5, Maximum=0.8)
      )
      Y=1
      Z=1
     )
     Float=&MakeFloatFromVector(
      Vector=&VectorFromFloat(Value=&Multiply_Float(
        A=&RandomRangeFloat(Minimum=0.8, Maximum=1.2)
        B=&Multiply_Float(
         A=0.1
         B=&FloatFromCurve(FloatCurve=DataInterface() {propertyValues {Curve {keys [
              {arriveTangent=0.2102, interpMode=RCIM_Cubic, leaveTangent=0.2102, tangentMode=RCTM_User, time=0, value=0.9}
              {interpMode=RCIM_Cubic, tangentMode=RCTM_User, time=1, value=1}
             ]}}})
        )
       ))
      Channel=::NewEnumerator0
     )
    ))
   &DynamicMaterialParameters(
    Index 0 Param 1=&Lerp_Float(
     A=&FloatFromCurve(FloatCurve=DataInterface() {propertyValues {Curve {keys [{tangentMode=RCTM_User, time=0, value=1}, <at 0.2 0.5>, <at 0.8 0.5>, {tangentMode=RCTM_User, time=1, value=0}]}}})
     B=0.4
     Alpha=$Particles.isShort
    )
    Index 0 Param 2=&FloatFromCurve(
     FloatCurve=DataInterface() {propertyValues {Curve {keys [
         {interpMode=RCIM_Cubic, leaveTangent=-1.1405, tangentMode=RCTM_Break, time=0, value=0.5}
         {arriveTangent=4.4936, interpMode=RCIM_Cubic, leaveTangent=-5.3647, tangentMode=RCTM_Break, time=0.4833, value=1}
         <at 1 0>
        ]}}}
     Scale Curve=&Multiply_Float(A=10, B=$User.WidthMult)
    )
    Index 0 Param 3=120
    Index 1 Param 1=&RandomRangeFloat(Minimum=0, Maximum=1)
    Index 1 Param 2=1
    Index 1 Param 3=1
    Index 1 Param 4=1
    Module.Param0WriteEnabled=true
    Module.Param0WriteEnabled1=true
    Module.Param1WriteEnabled=true
    Module.Param1WriteEnabled1=true
    Module.Param2WriteEnabled=true
    Module.Param2WriteEnabled1=true
    Module.Param3WriteEnabled=false
    Module.Param3WriteEnabled1=true
    Write Parameter Index 1=true
   ) {module=DynamicMaterialParameters001}
   &ScaleColor(
    Scale Alpha=&FloatFromCurve(FloatCurve=DataInterface() {propertyValues {Curve {keys [
         {interpMode=RCIM_Cubic, tangentMode=RCTM_User, time=0, value=0}
         {interpMode=RCIM_Cubic, tangentMode=RCTM_User, time=0.1258, value=1}
         <at 0.341 0>
         {interpMode=RCIM_Cubic, tangentMode=RCTM_User, time=0.496, value=0.5}
         <at 1 0>
        ]}}})
    Scale RGB=&VectorFromFloat(Value=&FloatFromCurve(
      FloatCurve=DataInterface() {propertyValues {Curve {keys [
          {arriveTangent=0.1143, interpMode=RCIM_Cubic, leaveTangent=0.1143, tangentMode=RCTM_User, time=0.018, value=4.9326}
          {arriveTangent=-6.0507, interpMode=RCIM_Cubic, leaveTangent=-6.0507, time=0.1778, value=0.8567}
          {arriveTangent=-0.5645, interpMode=RCIM_Cubic, leaveTangent=-0.5645, tangentMode=RCTM_User, time=0.2493, value=3.533}
          {arriveTangent=-22.9399, interpMode=RCIM_Cubic, leaveTangent=-22.9399, time=0.3224, value=1.7136}
          {arriveTangent=2.8763, interpMode=RCIM_Cubic, leaveTangent=2.8763, time=0.3773, value=0.5981}
          {arriveTangent=-1.8312, interpMode=RCIM_Cubic, leaveTangent=-1.8312, time=0.4693, value=2.1361}
          {arriveTangent=-2.2307, interpMode=RCIM_Cubic, leaveTangent=-2.2307, time=0.5983, value=0.1933}
          {arriveTangent=1.5587, interpMode=RCIM_Cubic, leaveTangent=1.5587, time=0.7391, value=1.5342}
          {arriveTangent=-2.0478, interpMode=RCIM_Cubic, leaveTangent=-2.0478, time=0.8629, value=0.6057}
          {arriveTangent=-0.0659, interpMode=RCIM_Cubic, leaveTangent=-0.0659, tangentMode=RCTM_User, time=1, value=1}
         ]}}}
      Scale Curve=$User.ColorMult
     ))
    ScaleA=true
    ScaleRGB=true
   ) {module=ScaleColor001}
  ]
  rendererClasses [&NiagaraMeshRendererProperties]
  renderers [{rendererClass=&NiagaraMeshRendererProperties, rendererIndex=0}]
 }
 Radial_Plasma_Energy001 {
  emitterUpdateScript [
   &EmitterState(Life Cycle Mode=::NewEnumerator1, Loop Behavior=::NewEnumerator1, Loop Duration=10, UseLoopDelay=false)
   &SpawnBurst_Instantaneous(Spawn Count=1, Use Spawn Probability=false)
  ]
  particleSpawnScript [
   &InitializeParticle(
    Color=<rgba 1 0.6 0.6 1>
    Lifetime=4
    Sprite Size Mode=::NewEnumerator3
    Uniform Sprite Size=1000
    UsePositionOffset=false
    Write Material Random=true
   )
   /SetVariables(Particles.SpriteSize=&Multiply_Vector2D_ByFloat(
     Float=&FN_LOD_Distance_Remap(Start Distance=5000, End Distance=25000, Minimum Output=1, Maximum Output=3)
     Vector2D=<2500 2500>
    )) {isSetParameters=true}
  ]
  particleUpdateScript [
   &ParticleState()
   &FN_Radial_Material(Base Sprite Tile U=0.5, Base Sprite Tile V=3, Depth Fade=2500, Radial Mask Offset X=0.1, Sprite Extrusion Speed=1.25) {enabled=false}
   &ScaleSpriteSize(Scale Factor=&Vector2DFromFloat(Value=&FloatFromCurve(
      FloatCurve=DataInterface() {propertyValues {Curve {keys [
          {arriveTangent=2.2779, interpMode=RCIM_Cubic, leaveTangent=2.2779, tangentMode=RCTM_User, time=0, value=0}
          {interpMode=RCIM_Cubic, tangentMode=RCTM_User, time=1, value=1}
         ]}}}
      Scale Curve=2
     )))
   /Niagara/Modules/Decal/Decal_Attributes(Decal Facing Mode=1, Decal Size:Vector3f=<800 8000 8000>, Orientation=1, Set Alignment=false, Size=1, Visible=1)
   &ScaleColor(
    Scale Alpha=&FloatFromCurve(FloatCurve=DataInterface() {propertyValues {Curve {keys [<at 0 1>,<at 1 0>]}}})
    ScaleA=true
    ScaleRGB=true
   )
   /SetVariables(Particles.DecalSize:Vector3f=&Multiply_Vector(
     A=$Particles.Decal_Attributes.DecalSize
     B=&VectorFromFloat(Value=&FloatFromCurve(FloatCurve=DataInterface() {propertyValues {Curve {keys [
           {arriveTangent=8.9669, interpMode=RCIM_Cubic, leaveTangent=8.9669, tangentMode=RCTM_User, time=0, value=0}
           {interpMode=RCIM_Cubic, tangentMode=RCTM_User, time=0.1, value=1}
          ]}}}))
    )) {isSetParameters=true}
  ]
  rendererClasses [&NiagaraDecalRendererProperties]
  renderers [{rendererClass=&NiagaraDecalRendererProperties, rendererIndex=0}]
 }
 Shimmers {
  simTarget=GPUComputeSim
  emitterUpdateScript [
   &EmitterState(
    Enable Distance Culling=true
    MaxDistance=10000
    Min Distance Response=::SleepAndLetParticlesFinish
    Module.bUseMaxDistance=true
    Module.bUseMinDistance=true
    Scalability Mode=::NewEnumerator1
   )
   &SpawnRate(SpawnRate=50, Use Spawn Probability=true)
  ]
  particleSpawnScript [
   /SetVariables(
    Particles.ColorEnd:LinearColor=&RandomRangeLinearColor(Minimum=<rgba 0.25 0 0.138 1>, Maximum=<rgba 0.0466 0 0.25 1>)
    Particles.ColorStart:LinearColor=&RandomRangeLinearColor(Minimum=<rgba 0.36 0.3 0.45 1>, Maximum=<rgba 0.6 0.68 0.74 1>)
   ) {isSetParameters=true, module=SetVariables_#2}
   &InitializeParticle(
    Lifetime Max=0.5
    Lifetime Min=0.25
    Lifetime Mode=::NewEnumerator1
    Position Mode=::NewEnumerator1
    Sprite Rotation Mode=::NewEnumerator2
    Sprite Size=&RandomRangeVector2D(Minimum=<400 400>, Maximum=<800 800>, Evaluation Type=::NewEnumerator1)
    Sprite Size Mode=::NewEnumerator4
    UsePositionOffset=false
    Write Material Random=true
   )
   &AddVelocity(Coordinate Space=::World, Velocity=&RandomRangeVector(Minimum=<-25 -2 -2>, Maximum=<-15 2 2>))
   &DynamicMaterialParameters(
    Index 0 Param 1=&RandomRangeFloat(Minimum=0, Maximum=1, Evaluation Type=::NewEnumerator1)
    Index 0 Param 2=&RandomRangeFloat(Minimum=0.4, Maximum=0.6, Evaluation Type=::NewEnumerator1)
    Index 0 Param 3=0.2
    Module.Param0WriteEnabled=true
    Module.Param1WriteEnabled=true
    Module.Param2WriteEnabled=true
    Module.Param3WriteEnabled=true
   )
   &SphereLocation(
    MaskSpawnGroup=false
    Non Uniform Scale=<1 1 0.2>
    Override Seed=false
    Sphere Radius=800
    Surface Only=false
    UseNonUniformScale=true
    UseOffset=false
    UseSphereOrientation=false
   )
  ]
  particleUpdateScript [
   &ParticleState()
   &SolveForcesAndVelocity(Clamp Velocity=false, Limit Acceleration=false)
   &SpriteRotationRate(Rotation Rate=&RandomRangeFloat(Minimum=-20, Maximum=20))
   &ScaleColor(
    Scale Alpha=&FloatFromCurve(
     FloatCurve=DataInterface() {propertyValues {Curve {keys [{arriveTangent=1.0833, interpMode=RCIM_Cubic, leaveTangent=1.0833, time=0, value=0}, <at 0.6 0.65 RCIM_Cubic>, <at 1 0>]}}}
     Scale Curve=&RandomRangeFloat(Minimum=0.05, Maximum=0.1)
    )
    Scale RGB=&MakeVectorFromLinearColorRGB(Color=<rgba 0.225 0.5317 1 1>)
    ScaleA=true
    ScaleRGB=true
   ) {module=ScaleColor001}
   &ScaleSpriteSize(Uniform Curve Sprite Scale=DataInterface() {propertyValues {Curve {keys [<at 0 0.7 RCIM_Cubic>, {interpMode=RCIM_Cubic, tangentMode=RCTM_Break, time=1, value=1}]}}})
  ]
  rendererClasses [&NiagaraSpriteRendererProperties]
  renderers [{rendererClass=&NiagaraSpriteRendererProperties}]
 }
 Shimmers001 {
  simTarget=GPUComputeSim
  emitterUpdateScript [
   &EmitterState(
    Enable Distance Culling=true
    MaxDistance=10000
    Min Distance Response=::SleepAndLetParticlesFinish
    Module.bUseMaxDistance=true
    Module.bUseMinDistance=true
    Scalability Mode=::NewEnumerator1
   )
   &SpawnRate(SpawnRate=50, Use Spawn Probability=true)
  ]
  particleSpawnScript [
   /SetVariables(
    Particles.ColorEnd:LinearColor=&RandomRangeLinearColor(Minimum=<rgba 0.25 0 0.138 1>, Maximum=<rgba 0.0466 0 0.25 1>)
    Particles.ColorStart:LinearColor=&RandomRangeLinearColor(Minimum=<rgba 0.36 0.3 0.45 1>, Maximum=<rgba 0.6 0.68 0.74 1>)
   ) {isSetParameters=true, module=SetVariables_#2}
   &InitializeParticle(
    Lifetime Max=0.5
    Lifetime Min=0.25
    Lifetime Mode=::NewEnumerator1
    Mesh Scale Mode=::NewEnumerator1
    Mesh Uniform Scale Max=10
    Mesh Uniform Scale Min=8
    Position Mode=::NewEnumerator1
    Sprite Rotation Mode=::NewEnumerator2
    Sprite Size=&RandomRangeVector2D(Minimum=<400 400>, Maximum=<800 800>, Evaluation Type=::NewEnumerator1)
    Sprite Size Mode=::NewEnumerator4
    UsePositionOffset=false
    Write Material Random=true
   )
   &AddVelocity(Coordinate Space=::World, Velocity=&RandomRangeVector(Minimum=<-25 -2 -2>, Maximum=<-15 2 2>))
   &DynamicMaterialParameters(
    Index 0 Param 1=&RandomRangeFloat(Minimum=0, Maximum=1, Evaluation Type=::NewEnumerator1)
    Index 0 Param 2=&RandomRangeFloat(Minimum=0.4, Maximum=0.6, Evaluation Type=::NewEnumerator1)
    Index 0 Param 3=0.2
    Module.Param0WriteEnabled=true
    Module.Param1WriteEnabled=true
    Module.Param2WriteEnabled=true
    Module.Param3WriteEnabled=true
   )
   &SphereLocation(
    MaskSpawnGroup=false
    Non Uniform Scale=<1 1 0.2>
    Override Seed=false
    Sphere Radius=800
    Surface Only=false
    UseNonUniformScale=true
    UseOffset=false
    UseSphereOrientation=false
   )
   &InitialMeshOrientation(Use Rotation Vector=false)
  ]
  particleUpdateScript [
   &ParticleState()
   &SolveForcesAndVelocity(Clamp Velocity=false, Limit Acceleration=false)
   &SpriteRotationRate(Rotation Rate=&RandomRangeFloat(Minimum=-20, Maximum=20))
   &ScaleColor(
    Scale Alpha=&FloatFromCurve(
     FloatCurve=DataInterface() {propertyValues {Curve {keys [{arriveTangent=1.0833, interpMode=RCIM_Cubic, leaveTangent=1.0833, time=0, value=0}, <at 0.6 0.65 RCIM_Cubic>, <at 1 0>]}}}
     Scale Curve=&RandomRangeFloat(Minimum=0.8, Maximum=1)
    )
    Scale RGB=&MakeVectorFromLinearColorRGB(Color=&RandomRangeLinearColor(Minimum=<rgba 1 2 50 1>, Maximum=<rgba 34.7836 8.0685 50 1>))
    ScaleA=true
    ScaleRGB=true
   ) {module=ScaleColor001}
   &DynamicMaterialParameters(
    Index 0 Param 1=1
    Index 0 Param 2=&FloatFromCurve(FloatCurve=DataInterface() {propertyValues {Curve {keys [
         {interpMode=RCIM_Cubic, tangentMode=RCTM_User, time=0, value=-1}
         {interpMode=RCIM_Cubic, tangentMode=RCTM_User, time=0.5, value=1}
         {tangentMode=RCTM_User, time=1, value=-1}
        ]}}})
    Index 0 Param 3=2
    Index 0 Param 4=1
    Module.Param0WriteEnabled=true
    Module.Param1WriteEnabled=true
    Module.Param2WriteEnabled=true
    Module.Param3WriteEnabled=true
   ) {module=DynamicMaterialParameters001}
   &ScaleMeshSize(Scale Factor=&VectorFromFloat(Value=&FloatFromCurve(
      FloatCurve=DataInterface() {propertyValues {Curve {keys [
          {arriveTangent=3.4712, interpMode=RCIM_Cubic, leaveTangent=3.4712, tangentMode=RCTM_User, time=0, value=0}
          {arriveTangent=-0.0405, interpMode=RCIM_Cubic, leaveTangent=-0.0405, tangentMode=RCTM_User, time=1, value=1.125}
         ]}}}
      Scale Curve=0.5
     )))
  ]
  rendererClasses [&NiagaraMeshRendererProperties]
  renderers [{rendererClass=&NiagaraMeshRendererProperties, rendererIndex=0}]
 }
}
summary {
 systemName=NS_TentacleAttackLingering
 emitters [
  {emitterName=Lightning_Primary001, rendererClasses [&NiagaraMeshRendererProperties]}
  {emitterName=Fire, rendererClasses [&NiagaraSpriteRendererProperties]}
  {emitterName=Arcs, rendererClasses [&NiagaraSpriteRendererProperties]}
  {emitterName=Shimmers, simTarget=GPUComputeSim, rendererClasses [&NiagaraSpriteRendererProperties]}
  {emitterName=Light, rendererClasses [&NiagaraLightRendererProperties]}
  {emitterName=Shimmers001, simTarget=GPUComputeSim, rendererClasses [&NiagaraMeshRendererProperties]}
  {emitterName=Big_Glow_Mesh, rendererClasses [&NiagaraMeshRendererProperties]}
  {emitterName=Radial_Plasma_Energy001, rendererClasses [&NiagaraDecalRendererProperties]}
 ]
 userVariables [
  User.Color {type=LinearColor, description="", defaultValue {struct=LinearColor, value=<rgba 0 0.161 0.5 1>}}
  User.ColorMult {type=NiagaraFloat, defaultValue=1, description=""}
  User.Offset {type=NiagaraPosition, description="", defaultValue {}}
  User.Scaling_GlowRadius {type=NiagaraFloat, defaultValue=1, description=""}
  User.Scaling_LightRadius {type=NiagaraFloat, defaultValue=1500, description=""}
  User.Scaling_SpawnRadius {type=NiagaraFloat, defaultValue=1, description=""}
  User.WidthMult {type=NiagaraFloat, defaultValue=1, description=""}
 ]
}
