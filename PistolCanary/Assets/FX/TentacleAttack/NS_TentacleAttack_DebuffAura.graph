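# Niagara system for the tentacle attack's debuff aura, wrapping a character's skeletal mesh in blue and purple electrical effects.
# Looping sprite, lightning-mesh and pulsing-light emitters combine with bone-attached GPU mist, rising ribbons and BoneFlow trails spiralling along filtered arm, spine and leg bones.
# An arc emitter samples two bones on User.TargetSM, bends ribbons between them with a Bezier point, varies material by bone distance and scales spawn rate by platform quality.
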NiagaraSystem /PistolCanary/Assets/FX/TentacleAttack/NS_TentacleAttack_DebuffAura
dependencies {
 usedDataInterfaces [
  &NiagaraDataInterfaceCurve
  &NiagaraDataInterfaceArrayFloat3
  &NiagaraDataInterfaceColorCurve
  &NiagaraDataInterfacePlatformSet
  &NiagaraDataInterfaceDebugDraw
 ]
 usedDynamicInputs [&RandomVector]
 usedModules [&SystemState]
}
emitters {
 ElectricSprites {
  emitterUpdateScript [
   &EmitterState(Life Cycle Mode=::NewEnumerator1, Loop Behavior=::NewEnumerator1, Loop Delay=0.5, Loop Duration=1.4, UseLoopDelay=true)
   &SpawnRate(SpawnRate=15, Use Spawn Probability=false)
  ]
  particleSpawnScript [
   &FN_SetSpawnGroup(Number of Spawn Groups=11)
   &InitializeParticle(
    Color=<rgba 2.6761 5.7287 10 2>
    Lifetime Max=0.25
    Lifetime Min=0.1
    Lifetime Mode=::NewEnumerator1
    Position Offset=<0 -10 0>
    Position Offset Coordinate Space=::Local
    Sprite Rotation Mode=::NewEnumerator2
    Sprite Size=&UniformRangedVector2D(Minimum=<200 200>, Maximum=<400 400>)
    Sprite Size Mode=::NewEnumerator4
    Sprite UV Mode=::NewEnumerator1
    UsePositionOffset=true
    Write Material Random=true
   ) {module=InitializeParticle001}
   &ShapeLocation(Box Size=<10 10 20>, MaskSpawnGroup=false, Override Seed=false, Shape Primitive=::NewEnumerator2, Surface Only=false, UseOwnerScale=true)
   &DynamicMaterialParameters(
    Index 0 Param 1=$Particles.MaterialRandom
    Index 0 Param 2=&Lerp_Float(A=3, B=5, Alpha=$Particles.MaterialRandom)
    Module.Param0WriteEnabled=true
    Module.Param1WriteEnabled=true
    Module.Param2WriteEnabled=false
    Module.Param3WriteEnabled=false
   )
   &CameraOffset(Camera Offset Amount=&RandomRangeFloat(Minimum=-10, Maximum=10))
   &SubUVAnimation(End Frame=3, SubUV Animation Mode:ENiagara_SubUVLookupMode=::NewEnumerator2) {module=SubUVAnimation001}
  ]
  particleUpdateScript [
   &ParticleState()
   &DynamicMaterialParameters(
    Index 0 Param 3=&FloatFromCurve(FloatCurve=DataInterface() {propertyValues {Curve {keys [<at 0.2 0 RCIM_Cubic>, {interpMode=RCIM_Cubic, leaveTangent=1.5819, time=1, value=1}]}}})
    Module.Param0WriteEnabled=false
    Module.Param1WriteEnabled=false
    Module.Param2WriteEnabled=true
    Module.Param3WriteEnabled=false
   ) {module=DynamicMaterialParameters001}
   &ScaleSpriteSize(Scale Factor=&Vector2DFromFloat(Value=&FloatFromCurve(
      FloatCurve=DataInterface() {propertyValues {Curve {keys [{arriveTangent=0.3147, interpMode=RCIM_Cubic, leaveTangent=0.3147, tangentMode=RCTM_User, time=0, value=0.8}, <at 1 1>]}}}
      Scale Curve=&RandomRangeFloat(Minimum=0.1, Maximum=0.1)
     )))
   &ScaleColor(
    Scale Alpha=&FNTimeOfDayScalar(Night Scale=0.1, Morning Scale=0.1, Evening Scale=0.1)
    ScaleA=true
    ScaleRGB=true
   )
  ]
  rendererClasses [&NiagaraSpriteRendererProperties]
  renderers [{rendererClass=&NiagaraSpriteRendererProperties}]
 }
 Empty001 {
  emitterUpdateScript [
   &EmitterState(Life Cycle Mode=::NewEnumerator1, Loop Behavior=::NewEnumerator1, Loop Duration=1, UseLoopDelay=false)
   &SpawnBurst_Instantaneous(Use Spawn Probability=false, UseLoopCountLimit=false)
  ]
  particleSpawnScript [&InitializeParticle(
    Color=<rgba 0.03 0.12 0.3 1>
    Lifetime=6
    Sprite Size Mode=::NewEnumerator3
    Uniform Sprite Size=201.705
    UsePositionOffset=false
    Write Material Random=true
   )]
  particleUpdateScript [
   &ParticleState()
   &SkeletalMeshLocation(Bone / Socket Index=20, Bone Sampling Mode=::NewEnumerator4, MaskSpawnGroup=false, Skeletal Mesh=$User.TargetSM)
   &ScaleColor(
    Scale Alpha=&FloatFromCurve(FloatCurve=DataInterface() {propertyValues {Curve {keys [
         {arriveTangent=6.7479, interpMode=RCIM_Cubic, leaveTangent=6.7479, tangentMode=RCTM_User, time=0, value=0}
         {interpMode=RCIM_Cubic, tangentMode=RCTM_User, time=0.5, value=1}
         {tangentMode=RCTM_User, time=1, value=0}
        ]}}})
    ScaleA=true
    ScaleRGB=true
   )
   &FN_Light_Attributes(
    Color=<rgba 0 0.6667 2 1>
    Create Color Attribute=true
    Create Position Attribute=true
    Create Radius Attribute=true
    Position=<0 50 50>
    Radius=900
    Scale Color=&FloatFromCurve(
     FloatCurve=DataInterface() {propertyValues {Curve {keys [
         {arriveTangent=8.5788, interpMode=RCIM_Cubic, leaveTangent=8.5788, tangentMode=RCTM_User, time=0, value=0}
         {interpMode=RCIM_Cubic, tangentMode=RCTM_User, time=0.5, value=1}
         {tangentMode=RCTM_User, time=1, value=0}
        ]}}}
     Scale Curve=&RandomRangeFloat(Minimum=30000, Maximum=100000, Evaluation Type=::NewEnumerator1)
    )
   )
  ]
  rendererClasses [&NiagaraSpriteRendererProperties, &NiagaraLightRendererProperties]
  renderers [
   {rendererClass=&NiagaraSpriteRendererProperties}
   {rendererClass=&NiagaraLightRendererProperties, rendererIndex=1}
  ]
 }
 FN_Ribbons {
  emitterUpdateScript [
   &EmitterState(Life Cycle Mode=::NewEnumerator1, Loop Duration=0, UseLoopDelay=false)
   &FN_SpawnRibbon_v2(Number of Ribbons=7, Spawn Mode=::NewEnumerator0, SpawnRate=50)
  ]
  particleSpawnScript [
   &FN_SetSpawnGroup_v2(Number of Spawn Groups=7)
   &FN_InitializeRibbon_v3(
    Color:LinearColor=&MaskLinearColorBySpawnGroup(
     Masked Linear Color=<rgba 0 0 0 2.5>
     Passthrough Linear Color=&MaskLinearColorBySpawnGroup(Masked Linear Color=<rgba 0 0 0 2.5>, Passthrough Linear Color=<rgba 0.401 0.401 0.401 1.5>, Spawn Group Mask=3)
     Spawn Group Mask=2
    )
    Enable Skeletal Position=true
    Lifetime=0.35
    Ribbon Width=35
   )
   &AddVelocity(
    Coordinate Space=::World
    Scale Added Velocity=&Lerp_Vector(A=<1 1 1>, B=<1 1 2>, Alpha=$Emitter.Dive)
    Velocity=&RandomRangeVector(Minimum=<0 0 100>, Maximum=<0 1 100>)
   )
   &DynamicMaterialParameters(
    Index 0 Param 1=1
    Index 0 Param 2=&MaskFloatBySpawnGroup(Masked Float=0, Passthrough Float=0.2, Spawn Group Mask=1)
    Module.Param0WriteEnabled=true
    Module.Param1WriteEnabled=true
    Module.Param2WriteEnabled=false
    Module.Param3WriteEnabled=false
   )
  ]
  particleUpdateScript [
   &SolveForcesAndVelocity(Clamp Velocity=false, Limit Acceleration=false)
   &FN_UpdateRibbon(Detach Locked Particle=true, Enable Skeletal Position=true)
   &ScaleColor(Scale Mode=::NewEnumerator1, Scale RGBA=<3 1 3 1>, ScaleRGBA=true)
  ]
  rendererClasses [&NiagaraRibbonRendererProperties]
  renderers [{rendererClass=&NiagaraRibbonRendererProperties}]
 }
 LeftArm004 {
  emitterUpdateScript [
   &EmitterState(Life Cycle Mode=::NewEnumerator1, Loop Behavior=::NewEnumerator1, Loop Delay=2, Loop Duration=3, UseLoopDelay=true)
   &self:SpawnRibbonRate_02(
    ParticlesPerRibbon=3
    Rate:NiagaraFloat=&FN_PickFloatDependOnPlatform(
     TargetPlatform=DataInterface() {propertyValues {Platforms {qualityLevelMask=-29}}}
     TargetPlatformActive=&FN_PickFloatDependOnPlatform(
      TargetPlatform=DataInterface() {propertyValues {Platforms {qualityLevelMask=-31}}}
      TargetPlatformActive=&Multiply_Float(A=$User.RibbonSpawnRate, B=2)
      TargetPlatformNoActive=&Multiply_Float(A=$User.RibbonSpawnRate, B=1.5)
     )
     TargetPlatformNoActive=&Multiply_Float(A=$User.RibbonSpawnRate, B=1)
    )
   ) {module=SpawnRibbonRate_02}
   &SpawnPerFrame(Spawn Count:NiagaraInt32=$Emitter.SpawnInt, Use Spawn Probability=false)
   &SpawnPerFrame(Spawn Count:NiagaraInt32=$Emitter.SpawnInt, Spawn Group=1, Use Spawn Probability=false) {module=SpawnPerFrame001}
   &SpawnPerFrame(Spawn Count:NiagaraInt32=$Emitter.SpawnInt, Spawn Group=2, Use Spawn Probability=false) {module=SpawnPerFrame002}
   &SpawnPerFrame(Spawn Count:NiagaraInt32=$Emitter.SpawnInt, Spawn Group=3, Use Spawn Probability=false) {module=SpawnPerFrame003}
  ]
  particleSpawnScript [
   &self:DifferentGroup_02(
    Max=5
    Max1=11
    Max2=15
    Max3=20
    Min1=6
    Min2=12
    Min3=16
    RPG:NiagaraInt32=$Emitter.ParticlesPerRibbon
    Simple Counter:NiagaraDataInterfaceSimpleCounter=$Emitter.Simple Counter
   ) {module=DifferentGroup_02}
   &InitializeParticle(
    Color=<rgba 2 2 2 1>
    Lifetime=/Game/Effects/Niagara/DynamicInputs/Random/FN_SeededRandomRangeFloat(Minimum:NiagaraFloat=0.25, Maximum:NiagaraFloat=0.3, Random Seed:NiagaraInt32=$Particles.Group)
    Ribbon Width=0
    Ribbon Width Mode=::NewEnumerator1
    Sprite Size Mode=::NewEnumerator3
    Uniform Sprite Size=&RandomRangeFloat(Minimum=100, Maximum=150)
    UsePositionOffset=false
    Write Material Random=true
   )
   &CameraOffset(Camera Offset Amount=5)
  ]
  particleUpdateScript [
   &ParticleState()
   &ScaleSpriteSize(
    Scale Sprite Size Mode=::NewEnumerator0
    Uniform Scale Factor=&MakeCustomFloatFromBool(
     Bool=&Boolean_OrOperation(
      Bool1=&SetBoolByFloatComparison(A=$Particles.NormalizeP, B=0, Comparison Type=::NewEnumerator4)
      Bool2=&SetBoolByFloatComparison(A=$Particles.NormalizeP, B=1, Comparison Type=::NewEnumerator4)
     )
     True Float=1
     False Float=0
    )
   )
   &SkeletalMeshLocation(
    Bone / Socket Index=$Particles.Random1
    Bone Sampling Mode=::NewEnumerator4
    MaskSpawnGroup=false
    Orientation Sampling=::NewEnumerator1
    Position Sampling=::NewEnumerator2
    Sampled Position Offset=&RandomRangeVector(Minimum=<-2 -2 -5>, Maximum=<2 2 5>)
    Skeletal Mesh=$User.TargetSM
    Use Sampled Mesh Attribute as Ribbon ID=false
   )
   &SkeletalMeshLocation(
    Bone / Socket Index=$Particles.Random2
    Bone Sampling Mode=::NewEnumerator4
    MaskSpawnGroup=false
    Orientation Sampling=::NewEnumerator1
    Position Sampling=::NewEnumerator2
    Sampled Position Offset=&RandomRangeVector(Minimum=<-2 -2 -5>, Maximum=<2 2 5>)
    Skeletal Mesh=$User.TargetSM
    Use Sampled Mesh Attribute as Ribbon ID=false
   ) {module=SkeletalMeshLocation001}
   /SetVariables(Particles.Position=&LerpPosition(A=$Particles.SkeletalMeshLocation.SampledPosition, B=$Particles.SkeletalMeshLocation001.SampledPosition, Alpha=$Particles.NormalizeP)) {isSetParameters=true}
   &self:GetThirdPoint_02(
    BoneO1:Quat4f=$Particles.SkeletalMeshLocation.SampledRotation
    BoneO2:Quat4f=$Particles.SkeletalMeshLocation001.SampledRotation
    P1:NiagaraPosition=$Output.SkeletalMeshLocation.SampledPositionWithoutOffset
    P2:NiagaraPosition=$Output.SkeletalMeshLocation001.SampledPositionWithoutOffset
    Rotate:NiagaraFloat=&FN_SeededRandomRangeFloat(Minimum:NiagaraFloat=0, Maximum:NiagaraFloat=360, Random Seed:NiagaraInt32=$Particles.Group)
    Se:NiagaraFloat=&MakeCustomFloatFromBool(
     Bool=&SetBoolByFloatComparison(
      A=&FN_SeededRandomRangeFloat(Minimum:NiagaraFloat=0, Maximum:NiagaraFloat=1, Random Seed:NiagaraInt32=$Particles.Group)
      B=0.5
     )
     True Float=0.3
     False Float=-0.3
    )
    float001:NiagaraFloat=&FloatFromCurve(FloatCurve=DataInterface() {propertyValues {Curve {keys [<at 0 1.2>,<at 1 0.8>]}}})
   ) {module=GetThirdPoint_02}
   &self:BezCurve_02(
    NormalizedExeindex:NiagaraFloat=$Particles.NormalizeP
    P1:NiagaraPosition=$Particles.SkeletalMeshLocation.SampledPosition
    P2:NiagaraPosition=$Particles.BezCThirdPoint
    P3:NiagaraPosition=$Particles.SkeletalMeshLocation001.SampledPosition
   ) {module=BezCurve_02}
   &DynamicMaterialParameters(
    Index 0 Param 1=1
    Index 0 Param 3=&FN_SeededRandomRangeFloat(Minimum:NiagaraFloat=75, Maximum:NiagaraFloat=125, Random Seed:NiagaraInt32=$Particles.Group)
    Index 1 Param 3=&FloatFromCurve(FloatCurve=DataInterface() {propertyValues {Curve {keys [<at 0.5 -0.5>,<at 1 0.3>]}}})
    Index 1 Param 4=10
    Index 2 Param 1=&FloatFromCurve(
     FloatCurve=DataInterface() {propertyValues {Curve {keys [
         <at 0 5 RCIM_Cubic>
         {arriveTangent=-0.061, interpMode=RCIM_Cubic, leaveTangent=-0.061, tangentMode=RCTM_User, time=63.8, value=3}
         <at 63.9 0 RCIM_Constant>
        ]}}}
     CurveIndex=&DistanceBetweenPositions(A=$Particles.SkeletalMeshLocation.SampledPosition, B=$Particles.SkeletalMeshLocation001.SampledPosition)
     Scale Curve=2
    )
    Index 3 Param 3=&FloatFromCurve(FloatCurve=DataInterface() {propertyValues {Curve {keys [<at 0 0>,<at 0.8 1>]}}})
    Index 3 Param 4=-1
    Module.Index0Param0Write=true
    Module.Index0Param1Write=true
    Module.Index0Param2Write=true
    Module.Index0Param3Write=true
    Module.Index1Param0Write=true
    Module.Index1Param1Write=true
    Module.Index1Param2Write=true
    Module.Index1Param3Write=true
    Module.Index2Param0Write=true
    Module.Index2Param1Write=true
    Module.Index2Param2Write=true
    Module.Index2Param3Write=true
    Module.Index3Param0Write=true
    Module.Index3Param1Write=true
    Module.Index3Param2Write=true
    Module.Index3Param3Write=true
    Write Parameter Index 1=true
    Write Parameter Index 2=true
    Write Parameter Index 3=true
   )
   &ScaleColor(ScaleA=true, ScaleRGB=true)
   &ScaleColor(
    Scale Alpha=&RandomRangeFloat(Minimum=-0.5, Maximum=1, Evaluation Type=::NewEnumerator1)
    Scale RGB=&VectorFromFloat(Value=&RandomRangeFloat(Minimum=1, Maximum=2, Evaluation Type=::NewEnumerator1))
    ScaleA=true
    ScaleRGB=true
   ) {module=ScaleColor001}
   /SetVariables(Particles.RColor:LinearColor=$Particles.Color) {isSetParameters=true}
   &ScaleColor(
    Scale Alpha=&Multiply_Float(
     A=&MakeCustomFloatFromBool(
      Bool=&SetBoolByFloatComparison(A=$Particles.NormalizeP, B=0.5)
      True Float=&FloatFromCurve(FloatCurve=DataInterface() {propertyValues {Curve {keys [<at 0.4 0>,<at 0.6 1>]}}})
      False Float=1
     )
     B=&FloatFromCurve(FloatCurve=DataInterface() {propertyValues {Curve {keys [
          {arriveTangent=7.595, interpMode=RCIM_Cubic, leaveTangent=7.595, tangentMode=RCTM_User, time=0, value=0}
          {interpMode=RCIM_Cubic, tangentMode=RCTM_User, time=0.3, value=1}
          {arriveTangent=0.1069, tangentMode=RCTM_User, time=0.6, value=0}
         ]}}})
    )
    Scale RGB=&MakeVectorFromLinearColorRGB(Color=&MultiplyLinearColor(
      Linear Color=<rgba 0.0551 0.0197 0.3 0.1>
      Scale Factor=&MakeVector4_FromFloat(Float=&RandomRangeFloat(Minimum=0, Maximum=2))
     ))
    ScaleA=true
    ScaleRGB=true
   ) {module=ScaleColor002}
  ]
  rendererClasses [&NiagaraSpriteRendererProperties, &NiagaraRibbonRendererProperties]
  renderers [
   {rendererClass=&NiagaraSpriteRendererProperties}
   {rendererClass=&NiagaraRibbonRendererProperties, rendererIndex=1}
  ]
 }
 LightningMeshes {
  emitterUpdateScript [
   &EmitterState(
    Inactive Response=::NewEnumerator2
    Life Cycle Mode=::NewEnumerator1
    Loop Duration=&Lerp_Float(A=0.3, B=&RandomRangeFloat(Minimum=0.0666, Maximum=0.1666), Alpha=0.5)
    Recalculate Duration Each Loop=true
    UseLoopDelay=false
   )
   &SpawnBurst_Instantaneous(Use Spawn Probability=false, UseLoopCountLimit=false)
  ]
  particleSpawnScript [
   /SetVariables(
    Particles.MeshIndex=&SelectIntFromArray(Direct Array Index=0, Direct Array Mode=::NewEnumerator1, Int Selection Array=DataInterface() {propertyValues {IntData [0,1]}})
    Particles.bIsTravelingBolt:NiagaraBool=&RandomBool()
   ) {isSetParameters=true}
   &InitializeParticle(Color=<rgba 0 48.0999 222 1>, UsePositionOffset=false, Write Material Random=true)
   &DynamicMaterialParameters(
    Index 0 Param 1=1.333
    Index 0 Param 4=0.5
    Module.Param0WriteEnabled=true
    Module.Param1WriteEnabled=true
    Module.Param2WriteEnabled=true
    Module.Param3WriteEnabled=true
   )
   &InitialMeshOrientation(
    Mesh Orientation Mode=::NewEnumerator7
    Rotation=&SelectVectorFromArray(
     Direct Array Index=/Niagara/DynamicInputs/Divide/Divide_Int(A:NiagaraInt32=0, B:NiagaraInt32=4)
     Direct Array Mode=::NewEnumerator1
     Vector Selection Array=DataInterface() {propertyValues {FloatData [<0 0 0.5>, <0 0 0>]}}
    )
    Use Rotation Vector=true
   )
  ]
  particleUpdateScript [
   &ParticleState()
   &DynamicMaterialParameters(
    Index 0 Param 1=&MakeCustomFloatFromBool(
     Bool=$Particles.bIsTravelingBolt
     True Float=1.333
     False Float=&FloatFromCurve(FloatCurve=DataInterface() {propertyValues {Curve {keys [<at 0.666 1>,<at 1 0>]}}}, Scale Curve=1.333)
    )
    Index 0 Param 3=&MakeCustomFloatFromBool(
     Bool=$Particles.bIsTravelingBolt
     True Float=$Particles.NormalizedAge
     False Float=&FloatFromCurve(FloatCurve=DataInterface() {propertyValues {Curve {keys [<at 0 0>,<at 0.333 0.5>]}}})
    )
    Index 0 Param 4=&MakeCustomFloatFromBool(
     Bool=$Particles.bIsTravelingBolt
     True Float=0.5
     False Float=&FloatFromCurve(FloatCurve=DataInterface() {propertyValues {Curve {keys [{tangentMode=RCTM_User, time=0, value=0.333}, {tangentMode=RCTM_User, time=1, value=0.666}]}}})
    )
    Module.Param0WriteEnabled=true
    Module.Param1WriteEnabled=false
    Module.Param2WriteEnabled=true
    Module.Param3WriteEnabled=true
   ) {module=DynamicMaterialParameters001}
   /SetVariables() {enabled=false, isSetParameters=true}
   &ScaleColor(
    Scale RGB=&Multiply_Vector(
     A=&VectorFromFloat(Value=&Lerp_Float(A=0.133, B=1, Alpha=0.5))
     B=&VectorFromFloat(Value=1)
    )
    ScaleA=true
    ScaleRGB=true
   )
  ]
  rendererClasses [&NiagaraMeshRendererProperties]
  renderers [{rendererClass=&NiagaraMeshRendererProperties, rendererIndex=0}]
 }
 Lightning_shape04 {
  emitterUpdateScript [
   &EmitterState()
   &SpawnRate(SpawnRate=40, Use Spawn Probability=false)
  ]
  particleSpawnScript [
   &InitializeParticle(Color=<rgba 6.1875 0 10 1>, Lifetime=0.3, Mesh Scale Mode=::NewEnumerator4, UsePositionOffset=true, Write Material Random=true)
   &InitialMeshOrientation(Mesh Orientation Mode=::NewEnumerator2, Rotation=<0 0 0.4>, Use Rotation Vector=true)
   &CameraOffset(Camera Offset Amount=20)
  ]
  particleUpdateScript [
   &ParticleState()
   &ScaleMeshSize(Scale Factor=&Multiply_VectorByFloat(
     Vector=&Multiply_VectorByFloat(Vector=<1 1 1>, Float=0.9)
     Float=&RandomRangeFloat(Minimum=0.7, Maximum=0.8)
    ))
   &DynamicMaterialParameters(
    Index 0 Param 1=&FloatFromCurve(FloatCurve=DataInterface() {propertyValues {Curve {keys [<at 0 0>,<at 0.7 1>]}}})
    Index 0 Param 2=2
    Index 0 Param 3=&FloatFromCurve(FloatCurve=DataInterface() {propertyValues {Curve {keys [
         {arriveTangent=0.2581, interpMode=RCIM_Cubic, leaveTangent=0.2581, tangentMode=RCTM_User, time=0, value=0}
         <at 0.5 0>
         {arriveTangent=1.6703, interpMode=RCIM_Cubic, leaveTangent=1.6703, tangentMode=RCTM_User, time=1, value=1}
        ]}}})
    Index 0 Param 4=1
    Module.Param0WriteEnabled=true
    Module.Param1WriteEnabled=true
    Module.Param2WriteEnabled=true
    Module.Param3WriteEnabled=true
   )
   &ScaleColor(
    Scale Alpha=&Multiply_Float(
     A=1
     B=&FloatFromCurve(FloatCurve=DataInterface() {propertyValues {Curve {keys [
          {tangentMode=RCTM_User, time=0, value=1}
          {arriveTangent=-0.0405, interpMode=RCIM_Cubic, leaveTangent=-0.0405, tangentMode=RCTM_User, time=0.7009, value=1}
          {arriveTangent=-8.0089, interpMode=RCIM_Cubic, leaveTangent=-8.0089, tangentMode=RCTM_User, time=1, value=0}
         ]}}})
    )
    Scale RGB=&Multiply_VectorByFloat(
     Vector=<0.1 0 25>
     Float=&Multiply_Float(A=&RandomRangeFloat(Minimum=0, Maximum=2, Evaluation Type=::NewEnumerator1), B=1)
    )
    ScaleA=true
    ScaleRGB=true
   )
   /SetVariables(Particles.Glow Color=&MakeLinearColorFromVectorAndFloat(
     "Vector (RGB)"=<1 1 250>
     "Float (Alpha)"=&Multiply_Float(
      A=&FloatFromCurve(
       FloatCurve=DataInterface() {propertyValues {Curve {keys [
           {interpMode=RCIM_Cubic, tangentMode=RCTM_User, time=0, value=1}
           {arriveTangent=-0.1842, interpMode=RCIM_Cubic, leaveTangent=-0.1842, tangentMode=RCTM_User, time=1, value=0}
          ]}}}
       Scale Curve=1.5
      )
      B=1
     )
    )) {isSetParameters=true}
   /SetVariables(Particles.Glow Dy02:Vector4f=&MakeVector4(
     X=&FloatFromCurve(FloatCurve=DataInterface() {propertyValues {Curve {keys [<at 0 0>,<at 1 1>]}}})
     Y=2
     Z=0
     W=0.5
    )) {isSetParameters=true}
  ]
  rendererClasses [&NiagaraMeshRendererProperties]
  renderers [
   {rendererClass=&NiagaraMeshRendererProperties, rendererIndex=0}
   {rendererClass=&NiagaraMeshRendererProperties, rendererIndex=1}
   {rendererClass=&NiagaraMeshRendererProperties, rendererIndex=2}
  ]
 }
 Mist {
  simTarget=GPUComputeSim
  emitterUpdateScript [
   &EmitterState(Inactive Response=::NewEnumerator1, Life Cycle Mode=::NewEnumerator1, Loop Duration=1, UseLoopDelay=false) {module=EmitterState001}
   &SpawnRate(SpawnRate=25, Use Spawn Probability=false)
  ]
  particleSpawnScript [
   &FN_SetSpawnGroup_v2(Number of Spawn Groups=7)
   /Niagara/Modules/Spawn/Initialization/InitializeParticle(
    Color=<rgba 0.7855 0.01 1 0.5>
    Lifetime=&UniformRangedFloat(Minimum=0.5, Maximum=0.65)
    Module.Write Color=true
    Module.Write Lifetime=true
    Module.Write Mass=false
    Module.Write Material Random=true
    Module.Write Position=false
    Module.Write Scale=false
    Module.Write SpriteRotation=true
    Module.Write SpriteSize=true
    Position=&SimulationPosition()
    Sprite Rotation=&UniformRangedFloat(Minimum=0, Maximum=360)
    Sprite Size=&Vector2DFromFloat(Value=&Multiply_Float(
      A=&RandomRangeFloat(Minimum=35, Maximum=45)
      B=&MakeCustomFloatFromBool(Bool=&RandomBool(), True Float=-1, False Float=1)
     ))
   )
   &DynamicMaterialParameters(
    Index 0 Param 1=&UniformRangedFloat(Minimum=0, Maximum=1)
    Index 0 Param 2=&UniformRangedFloat(Minimum=1, Maximum=2)
    Index 0 Param 4=0.5
    Module.Param0WriteEnabled=true
    Module.Param1WriteEnabled=true
    Module.Param2WriteEnabled=false
    Module.Param3WriteEnabled=true
   )
   &CameraOffset(Camera Offset Amount=&RandomRangeFloat(Minimum=-30, Maximum=30)) {module=CameraOffset001}
   &SphereLocation(
    MaskSpawnGroup=false
    Override Seed=false
    Sphere Radius=10
    Surface Only=false
    UseNonUniformScale=false
    UseOffset=true
    UseSphereOrientation=false
   )
  ]
  particleUpdateScript [
   &ParticleState()
   &SkeletalMeshLocation(
    Bone / Socket Index=$Particles.SpawnGroup
    Bone Sampling Mode=::NewEnumerator4
    Inherit Velocity=false
    "Inherit Velocity (Scale)"=0
    MaskSpawnGroup=false
    Sampled Position Offset=&Lerp_Vector(
     A=Linked() {linkedVariable {name=Particles.Initial.Position, type=NiagaraPosition}}
     B=&Add_Vector(
      A=Linked() {linkedVariable {name=Particles.Initial.Position, type=NiagaraPosition}}
      B=&RandomRangeVector(Minimum=<-5 -5 0>, Maximum=<5 5 25>)
     )
     Alpha=&FloatFromCurve(FloatCurve=DataInterface() {propertyValues {Curve {keys [
          {arriveTangent=0.7341, interpMode=RCIM_Cubic, leaveTangent=0.7341, tangentMode=RCTM_User, time=0, value=0}
          {arriveTangent=1.7377, interpMode=RCIM_Cubic, leaveTangent=1.7377, tangentMode=RCTM_User, time=1, value=1}
         ]}}})
    )
    Sampled Position Offset Coordinate Space:ENiagaraMeshTransforms=::NewEnumerator2
    Skeletal Mesh=$User.SkeletalMesh
   )
   /SetVariables(Particles.SpriteRotation:NiagaraFloat=&Add_Float(A=$Particles.Initial.SpriteRotation, B=&RandomRangeFloat(Minimum=1, Maximum=1))) {isSetParameters=true}
   &SubUVAnimation(
    End Frame:NiagaraInt32=&RandomRangeInt(Minimum=45, Maximum=60)
    Start Frame:NiagaraInt32=&RandomRangeInt(Minimum=4, Maximum=12)
    SubUV Loop Count=0.3
   )
   &ScaleSpriteSize(Scale Factor=&Vector2DFromFloat(Value=&FloatFromCurve(FloatCurve=DataInterface() {propertyValues {Curve {keys [{tangentMode=RCTM_User, time=0, value=1}, {tangentMode=RCTM_User, time=1, value=1}]}}})))
   &ScaleColor(
    Scale Alpha=&FloatFromCurve(
     FloatCurve=DataInterface() {propertyValues {Curve {keys [
         {arriveTangent=1.0333, interpMode=RCIM_Cubic, leaveTangent=1.0333, tangentMode=RCTM_User, time=0, value=0}
         {arriveTangent=0.4682, interpMode=RCIM_Cubic, leaveTangent=0.4682, tangentMode=RCTM_User, time=0.2534, value=0.4966}
         {arriveTangent=0.0439, interpMode=RCIM_Cubic, leaveTangent=0.0439, tangentMode=RCTM_User, time=1, value=0}
        ]}}}
     Scale Curve=0.5
    )
    ScaleA=true
    ScaleRGB=true
   ) {module=ScaleColor001}
  ]
  rendererClasses [&NiagaraSpriteRendererProperties]
  renderers [{rendererClass=&NiagaraSpriteRendererProperties}]
 }
 Trail_BoneFlow_Arm_01 {
  emitterUpdateScript [
   &EmitterState(
    Life Cycle Mode=::NewEnumerator1
    Loop Duration=&RandomRangeFloat(Minimum=1, Maximum=3)
    Recalculate Duration Each Loop=true
    UseLoopDelay=false
   )
   &SpawnRate(SpawnRate=40, Use Spawn Probability=false) {enabled=false}
   &SpawnBurst_Instantaneous(Spawn Count=40, Use Spawn Probability=false, UseLoopCountLimit=false)
  ]
  particleSpawnScript [
   &FN_SetSpawnGroup(Inclusivity=true, Normalize Spawn Group=true, Particles Per Group=20, Pattern=::NewEnumerator0, Set Ribbon ID and Ribbon Link Order=true)
   &InitializeRibbon(Lifetime=0.6, Ribbon Width=3, UsePositionOffset=false, Write Material Random=true)
   &CameraOffset(Camera Offset Amount=10) {enabled=false}
  ]
  particleUpdateScript [
   &ParticleState()
   &FN_BoneFlow(
    Angle:NiagaraFloat=&Add_Float(A=&Multiply_Float(A=$Particles.Normalized Spawn Group, B=1.5), B=$Particles.Age)
    BoneID:NiagaraInt32=&ModuloInt(A=$Particles.SpawnGroup, B=2)
    Enable Position Offsets=true
    Height:NiagaraFloat=&Multiply_Float(A=$Particles.Normalized Spawn Group, B=1.25)
    Offset Start:Vector3f=&SelectVectorFromArray(
     Direct Array Index=&ModuloInt(A=$Particles.SpawnGroup, B=2)
     Vector Selection Array=DataInterface() {propertyValues {FloatData [<15 0 0>, <-15 0 0>]}}
    )
    Radius:NiagaraFloat=&FloatFromCurve(
     FloatCurve=DataInterface() {propertyValues {Curve {keys [
         {arriveTangent=27.7757, interpMode=RCIM_Cubic, leaveTangent=27.7757, tangentMode=RCTM_User, time=0, value=2}
         {interpMode=RCIM_Cubic, tangentMode=RCTM_User, time=0.3, value=7}
         {arriveTangent=-6.6061, interpMode=RCIM_Cubic, leaveTangent=-6.6061, tangentMode=RCTM_User, time=0.7, value=7}
         {arriveTangent=0.5728, interpMode=RCIM_Cubic, leaveTangent=0.5728, tangentMode=RCTM_User, time=1, value=0}
        ]}}}
     CurveIndex=$Particles.Normalized Spawn Group
     Scale Curve=1.6
    )
    SkeletalMesh=DataInterface() {propertyValues {
      DefaultMesh=null
      ExcludeBoneName=root
      OptionalComponentClass=null
      PreviewMesh=/BRCosmetics/Characters/Player/Male/Medium/Bodies/M_MED_SunBurst_Alt/Meshes/M_MED_SunBurst_Alt
      SkinningMode=SkinOnTheFly
      SoftSourceActor=null
      SourceMode=Default
      UvSetIndex=0
      WholeMeshLOD=-1
      bExcludeBone=false
      bReadDeformedGeometry=true
      bRequireCurrentFrameData=true
      ComponentTags []
      FilteredBones [upperarm_l,spine_02,spine_03,spine_04,hand_r]
      FilteredSockets []
      MeshUserParameter {parameter {name=None, typeDefHandle {registeredTypeIndex=1}, varData []}}
      SamplingRegions []
     }}
   )
   /SetVariables(Particles.RibbonFacing=$Output.FN_BoneFlow.SplineNormal) {isSetParameters=true, module=SetVariables_#1}
   &ScaleRibbonWidth(Ribbon Width Scale=&FloatFromCurve(FloatCurve=DataInterface() {propertyValues {Curve {keys [<at 0 1.5>,<at 1 0.5>]}}}, CurveIndex=$Particles.RibbonLinkOrder))
   /SetVariables(Particles.Ribbon B Width:NiagaraFloat=&Multiply_Float(A=$Particles.RibbonWidth, B=1)) {isSetParameters=true, module=SetVariables_#2}
   &ScaleColor(
    Scale Alpha=&FloatFromCurve(
     FloatCurve=DataInterface() {propertyValues {Curve {keys [
         {interpMode=RCIM_Cubic, tangentMode=RCTM_User, time=0, value=0}
         {arriveTangent=6.3194, interpMode=RCIM_Cubic, leaveTangent=-3.732, tangentMode=RCTM_Break, time=0.2, value=0.9}
         {arriveTangent=-0.5344, interpMode=RCIM_Cubic, leaveTangent=-0.5344, tangentMode=RCTM_User, time=0.9, value=0}
        ]}}}
     CurveIndex=$Particles.RibbonLinkOrder
    )
    Scale RGB=&MakeVectorFromLinearColorRGB(Color=&ColorFromCurve(
      ColorCurve=DataInterface() {propertyValues {
        AlphaCurve {keys [<at 0 1 RCIM_Cubic>]}
        BlueCurve {keys [<at 0.002 1 RCIM_Cubic>]}
        GreenCurve {keys [<at 0.002 0.8012 RCIM_Cubic>]}
        RedCurve {keys [<at 0.002 0 RCIM_Cubic>]}
       }}
      CurveIndex=$StackContext.RibbonLinkOrder
     ))
    ScaleA=true
    ScaleRGB=true
   ) {enabled=false}
   &ScaleColor(
    Color Value To Scale=$Particles.Color
    Scale Alpha=&Multiply_Float(A=1, B=&FN_FrontEndFloatModifier(Gameplay=1, Front End=1.5))
    Scale RGB=&Multiply_VectorByFloat(
     Vector=&Multiply_VectorByFloat(
      Vector=&VectorFromFloat(Value=&FNTimeOfDayScalar(Night Scale=0.5, Morning Scale=0.25, Evening Scale=0.25))
      Float=&FN_FrontEndFloatModifier(Gameplay=1, Front End=1.5)
     )
     Float=&Multiply_Float(
      A=&FloatFromCurve(FloatCurve=DataInterface() {propertyValues {Curve {keys [<at 0 1>,<at 1 10>]}}}, CurveIndex=1)
      B=1
     )
    )
    ScaleA=true
    ScaleRGB=true
   ) {enabled=false, module=ScaleColor001}
   &SolveForcesAndVelocity(Clamp Velocity=false, Limit Acceleration=false)
   &DynamicMaterialParameters(
    Index 0 Param 2=1
    Index 0 Param 3=1
    Index 0 Param 4=1
    Module.Param0WriteEnabled=true
    Module.Param1WriteEnabled=true
    Module.Param2WriteEnabled=true
    Module.Param3WriteEnabled=true
   )
   &ScaleColor(
    Scale Alpha=&FloatFromCurve(
     FloatCurve=DataInterface() {propertyValues {Curve {keys [
         {interpMode=RCIM_Cubic, tangentMode=RCTM_User, time=0, value=0}
         {interpMode=RCIM_Cubic, tangentMode=RCTM_User, time=0.5, value=1}
         {tangentMode=RCTM_User, time=1, value=0}
        ]}}}
     CurveIndex=$Particles.RibbonLinkOrder
     Scale Curve=&FloatFromCurve(FloatCurve=DataInterface() {propertyValues {Curve {keys [
          {interpMode=RCIM_Cubic, tangentMode=RCTM_User, time=0, value=0}
          {interpMode=RCIM_Cubic, tangentMode=RCTM_User, time=0.5, value=1}
          {tangentMode=RCTM_User, time=1, value=0}
         ]}}})
    )
    Scale RGB=<25 5 25>
    ScaleA=true
    ScaleRGB=true
   ) {module=ScaleColor002}
   &CameraOffset(Camera Offset Amount=25) {module=CameraOffset001}
  ]
  rendererClasses [&NiagaraRibbonRendererProperties]
  renderers [
   {rendererClass=&NiagaraRibbonRendererProperties}
   {rendererClass=&NiagaraRibbonRendererProperties, rendererIndex=1}
  ]
 }
 Trail_BoneFlow_Arm_01001 {
  emitterUpdateScript [
   &EmitterState(
    Life Cycle Mode=::NewEnumerator1
    Loop Duration=&RandomRangeFloat(Minimum=1, Maximum=3)
    Recalculate Duration Each Loop=true
    UseLoopDelay=false
   )
   &SpawnRate(SpawnRate=40, Use Spawn Probability=false) {enabled=false}
   &SpawnBurst_Instantaneous(Spawn Count=40, Use Spawn Probability=false, UseLoopCountLimit=false)
  ]
  particleSpawnScript [
   &FN_SetSpawnGroup(Inclusivity=true, Normalize Spawn Group=true, Particles Per Group=20, Pattern=::NewEnumerator0, Set Ribbon ID and Ribbon Link Order=true)
   &InitializeRibbon(Lifetime=0.35, Ribbon Width=3, UsePositionOffset=false, Write Material Random=true)
   &CameraOffset(Camera Offset Amount=10) {enabled=false}
  ]
  particleUpdateScript [
   &ParticleState()
   &FN_BoneFlow(
    Angle:NiagaraFloat=&Add_Float(A=&Multiply_Float(A=$Particles.Normalized Spawn Group, B=1.5), B=$Particles.Age)
    BoneID:NiagaraInt32=&ModuloInt(A=$Particles.SpawnGroup, B=2)
    Enable Position Offsets=true
    Height:NiagaraFloat=&Multiply_Float(A=$Particles.Normalized Spawn Group, B=1.25)
    Offset Start:Vector3f=&SelectVectorFromArray(
     Direct Array Index=&ModuloInt(A=$Particles.SpawnGroup, B=2)
     Vector Selection Array=DataInterface() {propertyValues {FloatData [<15 0 0>, <-15 0 0>]}}
    )
    Radius:NiagaraFloat=&FloatFromCurve(
     FloatCurve=DataInterface() {propertyValues {Curve {keys [
         {arriveTangent=27.7757, interpMode=RCIM_Cubic, leaveTangent=27.7757, tangentMode=RCTM_User, time=0, value=2}
         {interpMode=RCIM_Cubic, tangentMode=RCTM_User, time=0.3, value=7}
         {arriveTangent=-6.6061, interpMode=RCIM_Cubic, leaveTangent=-6.6061, tangentMode=RCTM_User, time=0.7, value=7}
         {arriveTangent=0.5728, interpMode=RCIM_Cubic, leaveTangent=0.5728, tangentMode=RCTM_User, time=1, value=0}
        ]}}}
     CurveIndex=$Particles.Normalized Spawn Group
     Scale Curve=1.6
    )
    SkeletalMesh=DataInterface() {propertyValues {
      DefaultMesh=null
      ExcludeBoneName=root
      OptionalComponentClass=null
      PreviewMesh=&M_MED_SunBurst_Alt
      SkinningMode=SkinOnTheFly
      SoftSourceActor=null
      SourceMode=Default
      UvSetIndex=0
      WholeMeshLOD=-1
      bExcludeBone=false
      bReadDeformedGeometry=true
      bRequireCurrentFrameData=true
      ComponentTags []
      FilteredBones [thigh_l,calf_l,foot_l,spine_01]
      FilteredSockets []
      MeshUserParameter {parameter {name=None, typeDefHandle {registeredTypeIndex=1}, varData []}}
      SamplingRegions []
     }}
   )
   /SetVariables(Particles.RibbonFacing=$Output.FN_BoneFlow.SplineNormal) {isSetParameters=true, module=SetVariables_#1}
   &ScaleRibbonWidth(Ribbon Width Scale=&FloatFromCurve(FloatCurve=DataInterface() {propertyValues {Curve {keys [<at 0 1.5>,<at 1 0.5>]}}}, CurveIndex=$Particles.RibbonLinkOrder))
   /SetVariables(Particles.Ribbon B Width:NiagaraFloat=&Multiply_Float(A=$Particles.RibbonWidth, B=3)) {isSetParameters=true, module=SetVariables_#2}
   &ScaleColor(
    Scale Alpha=&FloatFromCurve(
     FloatCurve=DataInterface() {propertyValues {Curve {keys [
         {interpMode=RCIM_Cubic, tangentMode=RCTM_User, time=0, value=0}
         {arriveTangent=6.3194, interpMode=RCIM_Cubic, leaveTangent=-3.732, tangentMode=RCTM_Break, time=0.2, value=0.9}
         {arriveTangent=-0.5344, interpMode=RCIM_Cubic, leaveTangent=-0.5344, tangentMode=RCTM_User, time=0.9, value=0}
        ]}}}
     CurveIndex=$Particles.RibbonLinkOrder
    )
    Scale RGB=&MakeVectorFromLinearColorRGB(Color=&ColorFromCurve(
      ColorCurve=DataInterface() {propertyValues {
        AlphaCurve {keys [<at 0 1 RCIM_Cubic>]}
        BlueCurve {keys [<at 0.002 1 RCIM_Cubic>]}
        GreenCurve {keys [<at 0.002 0.8012 RCIM_Cubic>]}
        RedCurve {keys [<at 0.002 0 RCIM_Cubic>]}
       }}
      CurveIndex=$StackContext.RibbonLinkOrder
     ))
    ScaleA=true
    ScaleRGB=true
   ) {enabled=false}
   &ScaleColor(
    Color Value To Scale=$Particles.Color
    Scale Alpha=&Multiply_Float(A=1, B=&FN_FrontEndFloatModifier(Gameplay=1, Front End=1.5))
    Scale RGB=&Multiply_VectorByFloat(
     Vector=&Multiply_VectorByFloat(
      Vector=&VectorFromFloat(Value=&FNTimeOfDayScalar(Night Scale=0.5, Morning Scale=0.25, Evening Scale=0.25))
      Float=&FN_FrontEndFloatModifier(Gameplay=1, Front End=1.5)
     )
     Float=&Multiply_Float(
      A=&FloatFromCurve(FloatCurve=DataInterface() {propertyValues {Curve {keys [<at 0 1>,<at 1 10>]}}}, CurveIndex=1)
      B=1
     )
    )
    ScaleA=true
    ScaleRGB=true
   ) {enabled=false, module=ScaleColor001}
   &SolveForcesAndVelocity(Clamp Velocity=false, Limit Acceleration=false)
   &DynamicMaterialParameters(
    Index 0 Param 2=1
    Index 0 Param 3=1
    Index 0 Param 4=1
    Module.Param0WriteEnabled=true
    Module.Param1WriteEnabled=true
    Module.Param2WriteEnabled=true
    Module.Param3WriteEnabled=true
   )
   &ScaleColor(
    Scale Alpha=&FloatFromCurve(
     FloatCurve=DataInterface() {propertyValues {Curve {keys [
         {interpMode=RCIM_Cubic, tangentMode=RCTM_User, time=0, value=0}
         {interpMode=RCIM_Cubic, tangentMode=RCTM_User, time=0.5, value=1}
         {tangentMode=RCTM_User, time=1, value=0}
        ]}}}
     CurveIndex=$Particles.RibbonLinkOrder
     Scale Curve=&FloatFromCurve(FloatCurve=DataInterface() {propertyValues {Curve {keys [
          {interpMode=RCIM_Cubic, tangentMode=RCTM_User, time=0, value=0}
          {interpMode=RCIM_Cubic, tangentMode=RCTM_User, time=0.5, value=1}
          {tangentMode=RCTM_User, time=1, value=0}
         ]}}})
    )
    Scale RGB=<25 5 25>
    ScaleA=true
    ScaleRGB=true
   ) {module=ScaleColor002}
   &CameraOffset(Camera Offset Amount=25) {module=CameraOffset001}
  ]
  rendererClasses [&NiagaraRibbonRendererProperties]
  renderers [
   {rendererClass=&NiagaraRibbonRendererProperties}
   {rendererClass=&NiagaraRibbonRendererProperties, rendererIndex=1}
  ]
 }
 Trail_BoneFlow_Arm_01002 {
  emitterUpdateScript [
   &EmitterState(
    Life Cycle Mode=::NewEnumerator1
    Loop Duration=&RandomRangeFloat(Minimum=1, Maximum=3)
    Recalculate Duration Each Loop=true
    UseLoopDelay=false
   )
   &SpawnRate(SpawnRate=40, Use Spawn Probability=false) {enabled=false}
   &SpawnBurst_Instantaneous(Spawn Count=40, Use Spawn Probability=false, UseLoopCountLimit=false)
  ]
  particleSpawnScript [
   &FN_SetSpawnGroup(Inclusivity=true, Normalize Spawn Group=true, Particles Per Group=20, Pattern=::NewEnumerator0, Set Ribbon ID and Ribbon Link Order=true)
   &InitializeRibbon(Lifetime=0.35, Ribbon Width=3, UsePositionOffset=false, Write Material Random=true)
   &CameraOffset(Camera Offset Amount=10) {enabled=false}
  ]
  particleUpdateScript [
   &ParticleState()
   &FN_BoneFlow(
    Angle:NiagaraFloat=&Add_Float(
     A=&Multiply_Float(A=$Particles.Normalized Spawn Group, B=&RandomRangeFloat(Minimum=0.25, Maximum=1))
     B=$Particles.Age
    )
    BoneID:NiagaraInt32=&ModuloInt(A=$Particles.SpawnGroup, B=2)
    Enable Position Offsets=true
    Height:NiagaraFloat=&Multiply_Float(A=$Particles.Normalized Spawn Group, B=&RandomRangeFloat(Minimum=1, Maximum=1.25))
    Offset Start:Vector3f=&SelectVectorFromArray(
     Direct Array Index=&ModuloInt(A=$Particles.SpawnGroup, B=2)
     Vector Selection Array=DataInterface() {propertyValues {FloatData [<15 0 25>, <-15 0 -25>]}}
    )
    Radius:NiagaraFloat=&FloatFromCurve(
     FloatCurve=DataInterface() {propertyValues {Curve {keys [
         {arriveTangent=27.7757, interpMode=RCIM_Cubic, leaveTangent=27.7757, tangentMode=RCTM_User, time=0, value=2}
         {interpMode=RCIM_Cubic, tangentMode=RCTM_User, time=0.3, value=7}
         {arriveTangent=-6.6061, interpMode=RCIM_Cubic, leaveTangent=-6.6061, tangentMode=RCTM_User, time=0.7, value=7}
         {arriveTangent=0.5728, interpMode=RCIM_Cubic, leaveTangent=0.5728, tangentMode=RCTM_User, time=1, value=0}
        ]}}}
     CurveIndex=$Particles.Normalized Spawn Group
     Scale Curve=1.6
    )
    SkeletalMesh=DataInterface() {propertyValues {
      DefaultMesh=null
      ExcludeBoneName=root
      OptionalComponentClass=null
      PreviewMesh=&M_MED_SunBurst_Alt
      SkinningMode=SkinOnTheFly
      SoftSourceActor=null
      SourceMode=Default
      UvSetIndex=0
      WholeMeshLOD=-1
      bExcludeBone=false
      bReadDeformedGeometry=true
      bRequireCurrentFrameData=true
      ComponentTags []
      FilteredBones [upperarm_l,lowerarm_l,hand_l,spine_01]
      FilteredSockets []
      MeshUserParameter {parameter {name=None, typeDefHandle {registeredTypeIndex=1}, varData []}}
      SamplingRegions []
     }}
   )
   /SetVariables(Particles.RibbonFacing=$Output.FN_BoneFlow.SplineNormal) {isSetParameters=true, module=SetVariables_#1}
   &ScaleRibbonWidth(Ribbon Width Scale=&FloatFromCurve(FloatCurve=DataInterface() {propertyValues {Curve {keys [<at 0 1.5>,<at 1 0.5>]}}}, CurveIndex=$Particles.RibbonLinkOrder, Scale Curve=2))
   /SetVariables(Particles.Ribbon B Width:NiagaraFloat=&Multiply_Float(A=$Particles.RibbonWidth, B=1)) {isSetParameters=true, module=SetVariables_#2}
   &ScaleColor(
    Scale Alpha=&FloatFromCurve(
     FloatCurve=DataInterface() {propertyValues {Curve {keys [
         {interpMode=RCIM_Cubic, tangentMode=RCTM_User, time=0, value=0}
         {arriveTangent=6.3194, interpMode=RCIM_Cubic, leaveTangent=-3.732, tangentMode=RCTM_Break, time=0.2, value=0.9}
         {arriveTangent=-0.5344, interpMode=RCIM_Cubic, leaveTangent=-0.5344, tangentMode=RCTM_User, time=0.9, value=0}
        ]}}}
     CurveIndex=$Particles.RibbonLinkOrder
    )
    Scale RGB=&MakeVectorFromLinearColorRGB(Color=&ColorFromCurve(
      ColorCurve=DataInterface() {propertyValues {
        AlphaCurve {keys [<at 0 1 RCIM_Cubic>]}
        BlueCurve {keys [<at 0.002 1 RCIM_Cubic>]}
        GreenCurve {keys [<at 0.002 0.8012 RCIM_Cubic>]}
        RedCurve {keys [<at 0.002 0 RCIM_Cubic>]}
       }}
      CurveIndex=$StackContext.RibbonLinkOrder
     ))
    ScaleA=true
    ScaleRGB=true
   ) {enabled=false}
   &ScaleColor(
    Color Value To Scale=$Particles.Color
    Scale Alpha=&Multiply_Float(A=1, B=&FN_FrontEndFloatModifier(Gameplay=1, Front End=1.5))
    Scale RGB=&Multiply_VectorByFloat(
     Vector=&Multiply_VectorByFloat(
      Vector=&VectorFromFloat(Value=&FNTimeOfDayScalar(Night Scale=0.5, Morning Scale=0.25, Evening Scale=0.25))
      Float=&FN_FrontEndFloatModifier(Gameplay=1, Front End=1.5)
     )
     Float=&Multiply_Float(
      A=&FloatFromCurve(FloatCurve=DataInterface() {propertyValues {Curve {keys [<at 0 1>,<at 1 10>]}}}, CurveIndex=1)
      B=1
     )
    )
    ScaleA=true
    ScaleRGB=true
   ) {enabled=false, module=ScaleColor001}
   &SolveForcesAndVelocity(Clamp Velocity=false, Limit Acceleration=false)
   &DynamicMaterialParameters(
    Index 0 Param 2=1
    Index 0 Param 3=1
    Index 0 Param 4=1
    Module.Param0WriteEnabled=true
    Module.Param1WriteEnabled=true
    Module.Param2WriteEnabled=true
    Module.Param3WriteEnabled=true
   )
   &ScaleColor(
    Scale Alpha=&FloatFromCurve(
     FloatCurve=DataInterface() {propertyValues {Curve {keys [
         {interpMode=RCIM_Cubic, tangentMode=RCTM_User, time=0, value=0}
         {interpMode=RCIM_Cubic, tangentMode=RCTM_User, time=0.5, value=1}
         {tangentMode=RCTM_User, time=1, value=0}
        ]}}}
     CurveIndex=$Particles.RibbonLinkOrder
     Scale Curve=&FloatFromCurve(FloatCurve=DataInterface() {propertyValues {Curve {keys [
          {interpMode=RCIM_Cubic, tangentMode=RCTM_User, time=0, value=0}
          {interpMode=RCIM_Cubic, tangentMode=RCTM_User, time=0.5, value=1}
          {tangentMode=RCTM_User, time=1, value=0}
         ]}}})
    )
    Scale RGB=<15 2 15>
    ScaleA=true
    ScaleRGB=true
   ) {module=ScaleColor002}
   &CameraOffset(Camera Offset Amount=25) {module=CameraOffset001}
  ]
  rendererClasses [&NiagaraRibbonRendererProperties]
  renderers [
   {rendererClass=&NiagaraRibbonRendererProperties}
   {rendererClass=&NiagaraRibbonRendererProperties, rendererIndex=1}
  ]
 }
}
summary {
 systemName=NS_TentacleAttack_DebuffAura
 emitters [
  {emitterName=Mist, simTarget=GPUComputeSim, rendererClasses [&NiagaraSpriteRendererProperties]}
  {emitterName=FN_Ribbons, rendererClasses [&NiagaraRibbonRendererProperties]}
  {emitterName=Trail_BoneFlow_Arm_01, rendererClasses [&NiagaraRibbonRendererProperties]}
  {emitterName=Trail_BoneFlow_Arm_01001, rendererClasses [&NiagaraRibbonRendererProperties]}
  {emitterName=Trail_BoneFlow_Arm_01002, rendererClasses [&NiagaraRibbonRendererProperties]}
  {emitterName=Lightning_shape04, rendererClasses [&NiagaraMeshRendererProperties]}
  {emitterName=ElectricSprites, rendererClasses [&NiagaraSpriteRendererProperties]}
  {emitterName=LightningMeshes, rendererClasses [&NiagaraMeshRendererProperties]}
  {emitterName=Empty001, rendererClasses [&NiagaraSpriteRendererProperties, &NiagaraLightRendererProperties]}
  {emitterName=LeftArm004, rendererClasses [&NiagaraSpriteRendererProperties, &NiagaraRibbonRendererProperties]}
 ]
 userVariables [
  User.LightningOrder {
   type=NiagaraDataInterfaceArrayInt32
   description=""
   defaultValue {
    struct=NiagaraExt_VariableValue_DataInterface
    value {
     dataInterface=&self:NiagaraDataInterfaceArrayInt32_0
     dataInterfaceClass=&NiagaraDataInterfaceArrayInt32
    }
   }
  }
  User.RibbonIDOffset {type=NiagaraInt32}
  User.RibbonSpawnRate {type=NiagaraFloat, defaultValue=0.1, description=""}
  User.SK {
   type=NiagaraDataInterfaceSkeletalMesh
   defaultValue {
    struct=NiagaraExt_VariableValue_DataInterface
    value {dataInterface=&self:NiagaraDataInterfaceSkeletalMesh_1}
   }
  }
  User.SkeletalMesh {
   type=NiagaraDataInterfaceSkeletalMesh
   defaultValue {
    struct=NiagaraExt_VariableValue_DataInterface
    value {dataInterface=&self:NiagaraDataInterfaceSkeletalMesh_0}
   }
  }
  User.TargetSM {
   type=NiagaraDataInterfaceSkeletalMesh
   defaultValue {
    struct=NiagaraExt_VariableValue_DataInterface
    value {dataInterface=&self:NiagaraDataInterfaceSkeletalMesh_2}
   }
  }
 ]
}
